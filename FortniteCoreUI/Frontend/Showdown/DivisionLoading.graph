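# Arena division loading screen widget, showing division icon, league name and division text over faint, vertically tiled badge columns.
# UpdateBackground selects one of ten division icon materials by DivisionRank and creates two dynamic instances with opposite Speed values.
# One instance feeds the three right-hand columns; the other, which also sets MaskSpeed, feeds the three left-hand ones.
# PreConstruct calls UpdateBackground with PreviewRank.
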WidgetBlueprint /FortniteCoreUI/Frontend/Showdown/DivisionLoading
graphs [EventGraph {nodes [
   Event PreConstruct()
   > UpdateBackground!(execute, self:Fort Arena Loading Screen Object Reference, DivisionRank:int=Get PreviewRank() -> (PreviewRank:int=0)) -> (then)
   n1 = Event UpdateBackground!() -> (OutputDelegate, then, DivisionRank:int=0) {is=title K2Node_Event}
   n2 = Select!(
    Option 0:Material Interface Object Reference=/FortniteCoreUI/Frontend/Showdown/Materials/league/Loading/M_UI_LeagueDivision_IconMaster_Loading_Icon1
    Option 1:Material Interface Object Reference=/FortniteCoreUI/Frontend/Showdown/Materials/league/Loading/M_UI_LeagueDivision_IconMaster_Loading_Icon2
    Option 2:Material Interface Object Reference=/FortniteCoreUI/Frontend/Showdown/Materials/league/Loading/M_UI_LeagueDivision_IconMaster_Loading_Icon3
    Option 3:Material Interface Object Reference=/FortniteCoreUI/Frontend/Showdown/Materials/league/Loading/M_UI_LeagueDivision_IconMaster_Loading_Icon4
    Option 4:Material Interface Object Reference=/FortniteCoreUI/Frontend/Showdown/Materials/league/Loading/M_UI_LeagueDivision_IconMaster_Loading_Icon5
    Option 5:Material Interface Object Reference=/FortniteCoreUI/Frontend/Showdown/Materials/league/Loading/M_UI_LeagueDivision_IconMaster_Loading_Icon6
    Option 6:Material Interface Object Reference=/FortniteCoreUI/Frontend/Showdown/Materials/league/Loading/M_UI_LeagueDivision_IconMaster_Loading_Icon7
    Option 7:Material Interface Object Reference=/FortniteCoreUI/Frontend/Showdown/Materials/league/Loading/M_UI_LeagueDivision_IconMaster_Loading_Icon8
    Option 8:Material Interface Object Reference=/FortniteCoreUI/Frontend/Showdown/Materials/league/Loading/M_UI_LeagueDivision_IconMaster_Loading_Icon9
    Option 9:Material Interface Object Reference=/FortniteCoreUI/Frontend/Showdown/Materials/league/Loading/M_UI_LeagueDivision_IconMaster_Loading_Icon10
    Option 10:Material Interface Object Reference
    Index:int=@n1.DivisionRank
   ) -> (ReturnValue:Material Interface Object Reference)
   > n3 = CreateDynamicMaterialInstance!(execute, Parent=@n2, OptionalName=OuterMat, CreationFlags=None) -> (then, ReturnValue)
   > SetScalarParameterValue(self=@n3, ParameterName=Speed, Value=0.25)
   > n4 = CreateDynamicMaterialInstance(Parent=@n2, OptionalName=CenterMat, CreationFlags=None)
   > SetScalarParameterValue(self=@n4, ParameterName=Speed, Value=-0.25)
   > SetScalarParameterValue(self=@n4, ParameterName=MaskSpeed, Value=-0.124)
   > SetBrushFromMaterial(
    self=[
     Get Image_BGBadge4() -> (Image_BGBadge4:Image Object Reference)
     Get Image_BGBadge6() -> (Image_BGBadge6:Image Object Reference)
     Get Image_BGBadge5() -> (Image_BGBadge5:Image Object Reference)
    ]
    Material=@n3
   )
   > SetBrushFromMaterial(
    self=[
     Get Image_BGBadge3() -> (Image_BGBadge3:Image Object Reference)
     Get Image_BGBadge1() -> (Image_BGBadge1:Image Object Reference)
     Get Image_BGBadge2() -> (Image_BGBadge2:Image Object Reference)
    ]
    Material=@n4
   )
  ]}]
variables [PreviewRank {type=integer, default=5}]
widgets {
 info {inheritedWidgetCount=3, parentClass=/Script/FortniteUI.FortArenaLoadingScreen, rootWidgetClass=&Overlay, widgetCount=16}
 widgets [
  Overlay_0 {widgetClassPath=&Overlay, widget=Overlay_0}
  Image_Background {
   widgetClassPath=&Image
   bIsVariable=true
   parent=Overlay_0
   widget=Image_Background
   properties {brush {
     bIsDynamicallyLoaded=false
     drawAs=Image
     imageType=NoImage
     mirroring=NoMirror
     resourceName=None
     resourceObject=/FortniteCoreUI/Frontend/Showdown/Materials/league/M_LeaguePanelBG_Loading
     tiling=NoTile
     imageSize=<32 32>
     margin=<ltrb 0 0 0 0>
     outlineSettings {
      bUseBrushTransparency=false
      roundingType=HalfHeightRadius
      width=0
      color {colorUseRule=UseColor_Specified, specifiedColor=<rgba 0 0 0 0>}
      cornerRadii=<0 0 0 0>
     }
     tintColor {colorUseRule=UseColor_Specified, specifiedColor=<rgba 1 1 1 1>}
     uVRegion {bIsValid=false, max=<0 0>, min=<0 0>}
    }}
   slot {class=OverlaySlot, properties {horizontalAlignment=HAlign_Fill, verticalAlignment=VAlign_Fill}}
  }
  Overlay_1 {
   widgetClassPath=&Overlay
   parent=Overlay_0
   widget=Overlay_1
   slot {class=OverlaySlot, properties {horizontalAlignment=HAlign_Fill, verticalAlignment=VAlign_Fill, padding=<ltrb 80 -100 80 -100>}}
  }
  HorizontalBox_1 {
   widgetClassPath=&HorizontalBox
   parent=Overlay_1
   widget=HorizontalBox_1
   slot {class=OverlaySlot, properties {verticalAlignment=VAlign_Fill}}
  }
  Image_BGBadge1 {
   widgetClassPath=&Image
   bIsVariable=true
   parent=HorizontalBox_1
   widget=Image_BGBadge1
   properties {
    brush {
     bIsDynamicallyLoaded=false
     drawAs=Image
     imageType=NoImage
     mirroring=NoMirror
     resourceName=None
     resourceObject=null
     tiling=Vertical
     imageSize=<120 120>
     margin=<ltrb 0 0 0 0>
     outlineSettings {
      bUseBrushTransparency=false
      roundingType=HalfHeightRadius
      width=0
      color {colorUseRule=UseColor_Specified, specifiedColor=<rgba 0 0 0 0>}
      cornerRadii=<0 0 0 0>
     }
     tintColor {colorUseRule=UseColor_Specified, specifiedColor=<rgba 1 1 1 1>}
     uVRegion {bIsValid=false, max=<0 0>, min=<0 0>}
    }
    colorAndOpacity=<rgba 0.0252 0.0252 0.2307 0.1524>
   }
   slot {class=HorizontalBoxSlot}
  }
  Image_BGBadge2 {
   widgetClassPath=&Image
   bIsVariable=true
   parent=HorizontalBox_1
   widget=Image_BGBadge2
   properties {
    brush {
     bIsDynamicallyLoaded=false
     drawAs=Image
     imageType=NoImage
     mirroring=NoMirror
     resourceName=None
     resourceObject=null
     tiling=Vertical
     imageSize=<120 120>
     margin=<ltrb 0 0 0 0>
     outlineSettings {
      bUseBrushTransparency=false
      roundingType=HalfHeightRadius
      width=0
      color {colorUseRule=UseColor_Specified, specifiedColor=<rgba 0 0 0 0>}
      cornerRadii=<0 0 0 0>
     }
     tintColor {colorUseRule=UseColor_Specified, specifiedColor=<rgba 1 1 1 1>}
     uVRegion {bIsValid=false, max=<0 0>, min=<0 0>}
    }
    colorAndOpacity=<rgba 0.0252 0.0252 0.2307 0.1524>
   }
   slot {class=HorizontalBoxSlot, properties {padding=<ltrb -30 -120 -30 0>}}
  }
  Image_BGBadge3 {
   widgetClassPath=&Image
   bIsVariable=true
   parent=HorizontalBox_1
   widget=Image_BGBadge3
   properties {
    brush {
     bIsDynamicallyLoaded=false
     drawAs=Image
     imageType=NoImage
     mirroring=NoMirror
     resourceName=None
     resourceObject=null
     tiling=Vertical
     imageSize=<120 120>
     margin=<ltrb 0 0 0 0>
     outlineSettings {
      bUseBrushTransparency=false
      roundingType=HalfHeightRadius
      width=0
      color {colorUseRule=UseColor_Specified, specifiedColor=<rgba 0 0 0 0>}
      cornerRadii=<0 0 0 0>
     }
     tintColor {colorUseRule=UseColor_Specified, specifiedColor=<rgba 1 1 1 1>}
     uVRegion {bIsValid=false, max=<0 0>, min=<0 0>}
    }
    colorAndOpacity=<rgba 0.0252 0.0252 0.2307 0.1524>
   }
   slot {class=HorizontalBoxSlot}
  }
  HorizontalBox_2 {
   widgetClassPath=&HorizontalBox
   parent=Overlay_1
   widget=HorizontalBox_2
   slot {class=OverlaySlot, properties {horizontalAlignment=HAlign_Right, verticalAlignment=VAlign_Fill}}
  }
  Image_BGBadge4 {
   widgetClassPath=&Image
   bIsVariable=true
   parent=HorizontalBox_2
   widget=Image_BGBadge4
   properties {
    brush {
     bIsDynamicallyLoaded=false
     drawAs=Image
     imageType=NoImage
     mirroring=NoMirror
     resourceName=None
     resourceObject=null
     tiling=Vertical
     imageSize=<120 120>
     margin=<ltrb 0 0 0 0>
     outlineSettings {
      bUseBrushTransparency=false
      roundingType=HalfHeightRadius
      width=0
      color {colorUseRule=UseColor_Specified, specifiedColor=<rgba 0 0 0 0>}
      cornerRadii=<0 0 0 0>
     }
     tintColor {colorUseRule=UseColor_Specified, specifiedColor=<rgba 1 1 1 1>}
     uVRegion {bIsValid=false, max=<0 0>, min=<0 0>}
    }
    colorAndOpacity=<rgba 0.0252 0.0252 0.2307 0.1524>
   }
   slot {class=HorizontalBoxSlot}
  }
  Image_BGBadge5 {
   widgetClassPath=&Image
   bIsVariable=true
   parent=HorizontalBox_2
   widget=Image_BGBadge5
   properties {
    brush {
     bIsDynamicallyLoaded=false
     drawAs=Image
     imageType=NoImage
     mirroring=NoMirror
     resourceName=None
     resourceObject=null
     tiling=Vertical
     imageSize=<120 120>
     margin=<ltrb 0 0 0 0>
     outlineSettings {
      bUseBrushTransparency=false
      roundingType=HalfHeightRadius
      width=0
      color {colorUseRule=UseColor_Specified, specifiedColor=<rgba 0 0 0 0>}
      cornerRadii=<0 0 0 0>
     }
     tintColor {colorUseRule=UseColor_Specified, specifiedColor=<rgba 1 1 1 1>}
     uVRegion {bIsValid=false, max=<0 0>, min=<0 0>}
    }
    colorAndOpacity=<rgba 0.0252 0.0252 0.2307 0.1524>
   }
   slot {class=HorizontalBoxSlot, properties {padding=<ltrb -30 -120 -30 0>}}
  }
  Image_BGBadge6 {
   widgetClassPath=&Image
   bIsVariable=true
   parent=HorizontalBox_2
   widget=Image_BGBadge6
   properties {
    brush {
     bIsDynamicallyLoaded=false
     drawAs=Image
     imageType=NoImage
     mirroring=NoMirror
     resourceName=None
     resourceObject=null
     tiling=Vertical
     imageSize=<120 120>
     margin=<ltrb 0 0 0 0>
     outlineSettings {
      bUseBrushTransparency=false
      roundingType=HalfHeightRadius
      width=0
      color {colorUseRule=UseColor_Specified, specifiedColor=<rgba 0 0 0 0>}
      cornerRadii=<0 0 0 0>
     }
     tintColor {colorUseRule=UseColor_Specified, specifiedColor=<rgba 1 1 1 1>}
     uVRegion {bIsValid=false, max=<0 0>, min=<0 0>}
    }
    colorAndOpacity=<rgba 0.0252 0.0252 0.2307 0.1524>
   }
   slot {class=HorizontalBoxSlot}
  }
  SafeZone_0 {
   widgetClassPath=&SafeZone
   bIsVariable=true
   parent=Overlay_0
   widget=SafeZone_0
   slot {class=OverlaySlot, properties {horizontalAlignment=HAlign_Fill, verticalAlignment=VAlign_Fill}}
  }
  VerticalBox_3 {
   widgetClassPath=&VerticalBox
   parent=SafeZone_0
   widget=VerticalBox_3
   properties {renderTransform {angle=0, scale=<1 1>, shear=<0 0>, translation=<0 -50>}}
   slot {class=SafeZoneSlot, properties {vAlign=VAlign_Center}}
  }
  Image_CurrentDivision {
   widgetClassPath=&FortLazyImage
   bInherited=true
   bIsVariable=true
   parent=VerticalBox_3
   widget=Image_CurrentDivision
   properties {brush {
     bIsDynamicallyLoaded=false
     drawAs=Image
     imageType=NoImage
     mirroring=NoMirror
     resourceName=None
     resourceObject=/FortniteCoreUI/Frontend/Showdown/Materials/league/T_UI_LeagueDivision_FancyIcon1
     tiling=NoTile
     imageSize=<256 256>
     margin=<ltrb 0 0 0 0>
     outlineSettings {
      bUseBrushTransparency=false
      roundingType=HalfHeightRadius
      width=0
      color {colorUseRule=UseColor_Specified, specifiedColor=<rgba 0 0 0 0>}
      cornerRadii=<0 0 0 0>
     }
     tintColor {colorUseRule=UseColor_Specified, specifiedColor=<rgba 1 1 1 1>}
     uVRegion {bIsValid=false, max=<0 0>, min=<0 0>}
    }}
   slot {class=VerticalBoxSlot, properties {horizontalAlignment=HAlign_Center, verticalAlignment=VAlign_Center, padding=<ltrb 0 -45 0 -25>}}
  }
  Text_LeagueName {
   widgetClassPath=/Script/UMG.RichTextBlock
   bInherited=true
   bIsVariable=true
   parent=VerticalBox_3
   widget=Text_LeagueName
   properties {
    bOverrideDefaultStyle=true
    justification=Center
    text=Open League
    textTransformPolicy=ToUpper
    defaultTextStyleOverride {
     overflowPolicy=Clip
     transformPolicy=None
     colorAndOpacity {colorUseRule=UseColor_Specified, specifiedColor=<rgba 1 1 1 1>}
     font {
      bForceMonospaced=false
      bMaterialIsStencil=false
      fontMaterial=null
      fontObject=&Burbank
      letterSpacing=0
      monospacedWidth=1
      size=90
      skewAmount=0
      typefaceFontName=Burbank-Black
      outlineSettings {
       bApplyOutlineToDropShadows=false
       bMiteredCorners=false
       bSeparateFillAlpha=false
       outlineMaterial=null
       outlineSize=0
       outlineColor=<rgba 0 0 0 1>
      }
     }
     highlightColor {colorUseRule=UseColor_Specified, specifiedColor=<rgba 0 0 0 1>}
     highlightShape {
      bIsDynamicallyLoaded=false
      drawAs=Image
      imageType=NoImage
      mirroring=NoMirror
      resourceName=None
      resourceObject=null
      tiling=NoTile
      imageSize=<32 32>
      margin=<ltrb 0 0 0 0>
      outlineSettings {
       bUseBrushTransparency=false
       roundingType=HalfHeightRadius
       width=0
       color {colorUseRule=UseColor_Specified, specifiedColor=<rgba 0 0 0 0>}
       cornerRadii=<0 0 0 0>
      }
      tintColor {colorUseRule=UseColor_Specified, specifiedColor=<rgba 1 1 1 1>}
      uVRegion {bIsValid=false, max=<0 0>, min=<0 0>}
     }
     selectedBackgroundColor {colorUseRule=UseColor_Foreground, specifiedColor=<rgba 1 0 1 1>}
     shadowColorAndOpacity=<rgba 0.0273 0.0012 0.1221 1>
     shadowOffset=<18 18>
     strikeBrush {
      bIsDynamicallyLoaded=false
      drawAs=Image
      imageType=NoImage
      mirroring=NoMirror
      resourceName=None
      resourceObject=null
      tiling=NoTile
      imageSize=<32 32>
      margin=<ltrb 0 0 0 0>
      outlineSettings {
       bUseBrushTransparency=false
       roundingType=HalfHeightRadius
       width=0
       color {colorUseRule=UseColor_Specified, specifiedColor=<rgba 0 0 0 0>}
       cornerRadii=<0 0 0 0>
      }
      tintColor {colorUseRule=UseColor_Specified, specifiedColor=<rgba 1 1 1 1>}
      uVRegion {bIsValid=false, max=<0 0>, min=<0 0>}
     }
     underlineBrush {
      bIsDynamicallyLoaded=false
      drawAs=Image
      imageType=NoImage
      mirroring=NoMirror
      resourceName=None
      resourceObject=null
      tiling=NoTile
      imageSize=<32 32>
      margin=<ltrb 0 0 0 0>
      outlineSettings {
       bUseBrushTransparency=false
       roundingType=HalfHeightRadius
       width=0
       color {colorUseRule=UseColor_Specified, specifiedColor=<rgba 0 0 0 0>}
       cornerRadii=<0 0 0 0>
      }
      tintColor {colorUseRule=UseColor_Specified, specifiedColor=<rgba 1 1 1 1>}
      uVRegion {bIsValid=false, max=<0 0>, min=<0 0>}
     }
    }
   }
   slot {class=VerticalBoxSlot}
  }
  Text_Division {
   widgetClassPath=&CommonTextBlock
   bInherited=true
   bIsVariable=true
   parent=VerticalBox_3
   widget=Text_Division
   properties {
    justification=Center
    text=DIVISION 1
    colorAndOpacity {colorUseRule=UseColor_Specified, specifiedColor=<rgba 0 1 0.9823 1>}
    font {
     bForceMonospaced=false
     bMaterialIsStencil=false
     fontMaterial=null
     fontObject=&BurbankSmall
     letterSpacing=0
     monospacedWidth=1
     size=24
     skewAmount=0
     typefaceFontName=Black
     outlineSettings {
      bApplyOutlineToDropShadows=false
      bMiteredCorners=false
      bSeparateFillAlpha=false
      outlineMaterial=null
      outlineSize=0
      outlineColor=<rgba 0 0 0 1>
     }
    }
   }
   slot {class=VerticalBoxSlot, properties {horizontalAlignment=HAlign_Center, verticalAlignment=VAlign_Center}}
  }
 ]
}
