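# UI material that tints the text of a simple button on the Showdown events frontend according to hover and press state.
# It blends BaseColor toward HoverColor by the Hover scalar, then toward PressedColor by Pressed.
# A last blend toward PressedHoverColor, weighted by Hover times Pressed, gives the combined state its own color.
# Only the RGB of each color parameter feeds the blend chain.
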Material /FortniteCoreUI/Frontend/Showdown/Materials/M_EventsSimpleButtonText
graph {
 properties {bUsedWithStaticMesh=false, materialDomain=MD_UI, pixelDepthOffsetMode=PDOM_Legacy, usedShadingModels=Default Lit}
 nodes [
  n1 = VectorParameter() {parameterName=BaseColor, defaultValue=<rgba 1 1 1 1>}
  n2 = VectorParameter() {parameterName=HoverColor, defaultValue=<rgba 1 1 1 0>}
  n3 = ScalarParameter() {parameterName=Hover}
  n4 = ScalarParameter() {parameterName=Pressed}
  LinearInterpolate(
   A=LinearInterpolate(
    A=LinearInterpolate(A=@n1.RGB, B=@n2.RGB, Alpha=@n3)
    B=VectorParameter() {parameterName=PressedColor, defaultValue=<rgba 1 1 1 0>}.RGB
    Alpha=@n4
   )
   B=VectorParameter() {parameterName=PressedHoverColor, defaultValue=<rgba 1 1 1 0>}.RGB
   Alpha=Multiply(A=ScalarParameter() {parameterName=Hover}, B=@n4)
  )
 ]
}
