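# Translucent UI master material for Showdown frontend badge silhouettes, built from base, bottom bracket, wings-and-star and laurel layers.
# Channels of two tier textures pass through antialiased thresholds, with radial gradients and bump scalars driving per-layer reveals.
# An offset wings-and-star threshold cuts a gap into the other layers, limited by vertical position on the base.
# A higher base threshold separates shape from outline color, each blended by Focus and Hover.
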Material /FortniteCoreUI/Frontend/Showdown/Materials/M_UI_Badge_Silhouette_Master
graph {
 properties {
  bUsedWithStaticMesh=false
  blendMode=BLEND_Translucent
  materialDomain=MD_UI
  pixelDepthOffsetMode=PDOM_Legacy
  usedShadingModels=Default Lit
 }
 nodes [
  n1 = Multiply(A=ScalarParameter() {defaultValue=0.1, parameterName=EdgeBase}, B=ScalarParameter() {defaultValue=1, group=Reveals, parameterName=BaseBump})
  n2 = ScalarParameter() {defaultValue=0.4, parameterName=OutlineSize}
  n3 = TextureCoordinate()
  n4 = Multiply(A=MaterialFunctionCall!(
    Threshold=Add(A=@n1, B=@n2)
    Gradient=TextureSample(UVs=@n3) {texture=&/T_UI_Badge_Silhouette_1}.R
    Filter Width Override=ScalarParameter() {defaultValue=0.05, parameterName=EdgeSoftness_Outline}
    Filter Width Fudge Factor
   ) {materialFunction=/Game/UI/Materials/AAThreshold}.Result)
  n5 = TextureSampleParameter2D(UVs=@n3) {parameterName=Tier, texture=&T_UI_Badge_Silhouette_1}
  n6 = MaterialFunctionCall(
   CenterPosition=Constant2Vector() {g=0.5, r=0.5}
   Radius=LinearInterpolate(Alpha=ScalarParameter() {defaultValue=1, group=Reveals, parameterName=BaseReveal, sliderMax=1}) {constB=2}
  ) {materialFunction=&RadialGradientExponential}
  n7 = Multiply(A=MaterialFunctionCall(
    Threshold=@n1
    Gradient=Multiply(A=@n5.R, B=@n6.RadialGradientExponential)
    Filter Width Override=ScalarParameter() {defaultValue=0.02, parameterName=EdgeSoftness_Base}
   ) {materialFunction=&AAThreshold}.Result)
  n8 = Add(
   A=ScalarParameter() {defaultValue=0.175, parameterName=EdgeWingsStar}
   B=ScalarParameter() {defaultValue=-0.0667, parameterName=WingsStarOutline}
  )
  n9 = ScalarParameter() {defaultValue=1, group=Reveals, parameterName=WingStarBump}
  n10 = TextureSampleParameter2D() {parameterName=Tier_1, texture=&/T_UI_Badge_Silhouette_2}
  n11 = ScalarParameter() {defaultValue=1, group=Reveals, parameterName=WingStarReveal, sliderMax=1}
  n12 = Multiply(
   A=@n10.R
   B=MaterialFunctionCall(CenterPosition=Constant2Vector() {g=0.8, r=0.5}, Radius=LinearInterpolate(Alpha=@n11) {constB=2}) {materialFunction=&RadialGradientExponential}.RadialGradientExponential
  )
  n13 = Multiply(
   A=MaterialFunctionCall(Threshold=Multiply(A=@n8, B=@n9), Gradient=@n12) {materialFunction=&AAThreshold}.Result
   B=ScalarParameter() {parameterName=Wings + Star}
  )
  n14 = ComponentMask(None=TextureCoordinate()) {g=true}
  n15 = Constant() {r=0.75}
  n16 = OneMinus(None=Multiply(
    A=@n13
    B=MaterialFunctionCall(Alpha=@n14, Min=@n15, Max=@n15) {materialFunction=&SmoothStep}.Result
   ))
  n17 = ScalarParameter() {defaultValue=0.15, parameterName=EdgeBracket}
  n18 = ScalarParameter() {defaultValue=1, group=Reveals, parameterName=BracketBump}
  n19 = MaterialFunctionCall(
   CenterPosition=Constant2Vector() {g=1, r=0.5}
   Radius=LinearInterpolate(Alpha=ScalarParameter() {defaultValue=1, group=Reveals, parameterName=BracketReveal, sliderMax=1}) {constB=2}
  ) {materialFunction=&RadialGradientExponential}
  n20 = Multiply(
   A=MaterialFunctionCall(
    Threshold=Multiply(A=@n17, B=@n18)
    Gradient=Multiply(A=@n5.G, B=@n19.RadialGradientExponential)
    Filter Width Override=ScalarParameter() {defaultValue=0.05, parameterName=EdgeSoftness_Bracket}
   ) {materialFunction=&AAThreshold}.Result
   B=ScalarParameter() {parameterName=Bottom Bracket}
  )
  n21 = OneMinus(None=@n13)
  n22 = Multiply(
   A=ScalarParameter() {defaultValue=0.175, parameterName=EdgeWingsStar}
   B=ScalarParameter() {defaultValue=1, group=Reveals, parameterName=WingStarBump}
  )
  n23 = ScalarParameter() {defaultValue=0.04, parameterName=EdgeSoftness_WingsStar}
  n24 = Add(
   A=Add(A=Multiply(A=@n7, B=@n16), B=Multiply(A=@n20, B=@n21))
   B=Multiply(
    A=MaterialFunctionCall(Threshold=@n22, Gradient=@n12, Filter Width Override=@n23) {materialFunction=&AAThreshold}.Result
    B=ScalarParameter() {parameterName=Wings + Star}
   )
  )
  n25 = Multiply(
   A=ScalarParameter() {defaultValue=0.18, parameterName=EdgeLaurel}
   B=ScalarParameter() {defaultValue=1, group=Reveals, parameterName=LaurelBump}
  )
  n26 = ScalarParameter() {defaultValue=1, group=Reveals, parameterName=LaurelReveal, sliderMax=1}
  n27 = Multiply(
   A=@n10.G
   B=MaterialFunctionCall(CenterPosition=Constant2Vector() {g=1, r=0.5}, Radius=LinearInterpolate(Alpha=@n26) {constB=2}) {materialFunction=&RadialGradientExponential}.RadialGradientExponential
  )
  n28 = ScalarParameter() {defaultValue=0.07, parameterName=EdgeSoftness_Laurel}
  n29 = Multiply(
   A=Multiply(
    A=MaterialFunctionCall(Threshold=@n25, Gradient=@n27, Filter Width Override=@n28) {materialFunction=&AAThreshold}.Result
    B=@n21
   )
   B=ScalarParameter() {parameterName=Laurel}
  )
  DotProduct(A=OneMinus(None=@n4), B=Clamp(None=Add(A=@n24, B=@n29)))
  n30 = VectorParameter() {parameterName=OutlineColor, defaultValue=<rgba 0.2016 0.9823 1 1>}
  n31 = VectorParameter() {parameterName=OutlineColorFocus, defaultValue=<rgba 0.5208 1 0.9491 1>}
  n32 = ScalarParameter() {parameterName=Focus}
  n33 = LinearInterpolate(
   A=LinearInterpolate(
    A=VectorParameter() {parameterName=ShapeColor, defaultValue=<rgba 0.0009 0.8632 0.8388 1>}.RGB
    B=VectorParameter() {parameterName=ShapeColorFocus, defaultValue=<rgba 0 0.2086 1 1>}.RGB
    Alpha=ScalarParameter() {parameterName=Focus}
   )
   B=VectorParameter() {parameterName=ShapeColorHover, defaultValue=<rgba 0.0009 0.8632 0.8388 1>}.RGB
   Alpha=ScalarParameter() {parameterName=Hover}
  )
  LinearInterpolate(
   A=LinearInterpolate(
    A=LinearInterpolate(A=@n30.RGB, B=@n31.RGB, Alpha=@n32)
    B=VectorParameter() {parameterName=OutlineColorHover, defaultValue=<rgba 0.359 0.9858 1 1>}.RGB
    Alpha=ScalarParameter() {parameterName=Hover}
   )
   B=LinearInterpolate(A=@n33, B=@n33, Alpha=@n4)
   Alpha=@n4
  )
 ]
}
