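# Translucent UI material for Showdown frontend screens that builds an antialiased, zag-shaped mask from two angled edges.
# Each edge subtracts a straight vertical-gradient threshold from a rotated UV threshold, with parameters for position, rotation and softness.
# The Max of both edges is inverted, then multiplied by a thresholded exponential radial gradient centered on a UV corner, sized by RevealRadius.
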Material /FortniteCoreUI/Frontend/Showdown/Materials/M_UI_HZag
graph {
 properties {
  bUsedWithStaticMesh=false
  blendMode=BLEND_Translucent
  materialDomain=MD_UI
  pixelDepthOffsetMode=PDOM_Legacy
  usedShadingModels=Default Lit
 }
 nodes [
  Constant3Vector() {constant=<rgba 1 1 1 0>}
  n1 = TextureCoordinate()
  MaterialFunctionCall(UVs=@n1) {materialFunction=&CustomRotator}
  n2 = TextureCoordinate()
  MaterialFunctionCall(UVs=@n2) {materialFunction=&CustomRotator}
  n3 = MaterialFunctionCall(UVs=TextureCoordinate(), "Rotation Angle (0-1)"=ScalarParameter() {defaultValue=0.25, parameterName=Edge1Rotation}) {materialFunction=&CustomRotator}
  n4 = ComponentMask(None=@n1) {g=true}
  n5 = Subtract(
   A=MaterialFunctionCall!(
    Threshold=ScalarParameter() {defaultValue=0.5, parameterName=Edge1Angle}
    Gradient=ComponentMask(None=@n3.Rotated Values) {g=true}
    Filter Width Override=ScalarParameter() {defaultValue=0.01, parameterName=EdgeSoftness}
    Filter Width Fudge Factor
   ) {materialFunction=/Game/UI/Materials/AAThreshold}.Result
   B=MaterialFunctionCall(
    Threshold=ScalarParameter() {defaultValue=0.5, parameterName=Edge1}
    Gradient=OneMinus(None=@n4)
    Filter Width Override=ScalarParameter() {defaultValue=0.01, parameterName=EdgeSoftness}
   ) {materialFunction=&AAThreshold}.Result
  )
  n6 = MaterialFunctionCall(UVs=TextureCoordinate(), "Rotation Angle (0-1)"=ScalarParameter() {defaultValue=-0.25, parameterName=Edge2Rotation}) {materialFunction=&CustomRotator}
  n7 = ComponentMask(None=@n2) {g=true}
  n8 = Subtract(
   A=MaterialFunctionCall(
    Threshold=ScalarParameter() {defaultValue=0.5, parameterName=Edge2Angle}
    Gradient=ComponentMask(None=@n6.Rotated Values) {g=true}
    Filter Width Override=ScalarParameter() {defaultValue=0.01, parameterName=EdgeSoftness}
   ) {materialFunction=&AAThreshold}.Result
   B=MaterialFunctionCall(
    Threshold=ScalarParameter() {defaultValue=0.5, parameterName=Edge2}
    Gradient=@n7
    Filter Width Override=ScalarParameter() {defaultValue=0.01, parameterName=EdgeSoftness}
   ) {materialFunction=&AAThreshold}.Result
  )
  n9 = Constant2Vector() {r=1}
  n10 = ScalarParameter() {defaultValue=2, parameterName=RevealRadius}
  Multiply(
   A=OneMinus(None=Max(A=@n5, B=@n8))
   B=MaterialFunctionCall(
    Threshold=Constant() {r=0.1}
    Gradient=MaterialFunctionCall(CenterPosition=@n9, Radius=@n10) {materialFunction=&RadialGradientExponential}.RadialGradientExponential
   ) {materialFunction=&AAThreshold}.Result
  )
  OneMinus(None=@n7)
 ]
}
