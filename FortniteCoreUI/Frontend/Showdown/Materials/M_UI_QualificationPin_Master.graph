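# UI master material for the Showdown qualification pin, compositing a badge texture over a two-color gradient-filled silhouette.
# AAThreshold turns a silhouette texture into an anti-aliased outer shape, plus a rescaled inner-edge mask with adjustable alpha that attenuates the badge layer.
# A radial gradient blends the two color parameters inside the silhouette, and the badge texture is layered on top.
# A ShineSwipe function, limited by that badge mask, is color-dodged over the result.
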Material /FortniteCoreUI/Frontend/Showdown/Materials/M_UI_QualificationPin_Master
graph {
 properties {
  bUsedWithStaticMesh=false
  blendMode=BLEND_Translucent
  materialDomain=MD_UI
  pixelDepthOffsetMode=PDOM_Legacy
  usedShadingModels=Default Lit
 }
 nodes [
  n1 = TextureCoordinate()
  n2 = ScalarParameter() {defaultValue=0.01, parameterName=EdgeSoftness}
  n3 = Multiply(A=MaterialFunctionCall!(
    Threshold=ScalarParameter() {defaultValue=0.2, parameterName=Edge}
    Gradient=TextureSample(UVs=@n1) {texture=&/T_UI_Badge_Silhouettes}.R
    Filter Width Override=@n2
    Filter Width Fudge Factor
   ) {materialFunction=/Game/UI/Materials/AAThreshold}.Result)
  n4 = TextureSample(UVs=TextureCoordinate()) {texture=&/T_UI_Qualification_Badge_256x}
  Add(A=@n3, B=@n4.A)
  n5 = MaterialFunctionCall!(In_Uv) {materialFunction=/Game/UI/Foundation/Materials/M-Function-ShineSwipe}
  n6 = ScalarParameter() {defaultValue=0.879, group=InnerEdge, parameterName=InnerEdgeScale}
  n7 = VectorParameter() {group=InnerEdge, parameterName=InnerEdgeCenter, defaultValue=<rgba 128 116.2839 256 1>}
  n8 = MaterialFunctionCall(
   Threshold=ScalarParameter() {defaultValue=0.6, group=InnerEdge, parameterName=InnerEdge}
   Gradient=TextureSample(UVs=MaterialFunctionCall!(Texture Scale=@n6, UVs, Offset XY Texturesize Z=@n7.RGB) {materialFunction=/Engine/Functions/Engine_MaterialFunctions01/Texturing/ScaleUVsAroundPoint}.UVs) {samplerSource=SSM_Clamp_WorldGroupSettings, texture=&T_UI_Badge_Silhouettes}.R
   Filter Width Override=@n2
  ) {materialFunction=&AAThreshold}
  n9 = OneMinus(None=Multiply(A=Multiply(A=@n8.Result), B=ScalarParameter() {group=InnerEdge, parameterName=InnerEdgeAlpha}))
  n10 = Multiply(A=@n4.A, B=@n9)
  n11 = MaterialFunctionCall(UVs=@n1, CenterPosition=Constant2Vector() {g=0.25, r=0.33}, Radius=Constant() {r=0.625}, Density=Constant() {r=2}) {materialFunction=&RadialGradientExponential}
  n12 = LinearInterpolate(
   B=LinearInterpolate(
    A=VectorParameter() {parameterName=Color2, defaultValue=<rgba 0 0.3049 0.8906 1>}.RGB
    B=VectorParameter() {parameterName=Color1, defaultValue=<rgba 0 1 0.9351 1>}.RGB
    Alpha=Multiply(A=@n11.RadialGradientExponential)
   )
   Alpha=@n3
  )
  n13 = Multiply(A=@n4.A, B=@n9)
  Multiply(A=MaterialFunctionCall!(Blend=LinearInterpolate(B=@n5.Alpha, Alpha=@n10), Base=LinearInterpolate(A=@n12, B=@n4.RGB, Alpha=@n13)) {materialFunction=/Engine/Functions/Engine_MaterialFunctions03/Blends/Blend_ColorDodge}.Result)
 ]
}
