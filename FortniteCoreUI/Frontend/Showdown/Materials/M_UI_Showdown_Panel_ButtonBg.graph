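# Translucent UI material for a Showdown panel button background in the frontend, with a two-color gradient, hover tint and rounded corners.
# Normalized UI UVs pass through CustomRotator, and the red channel blends Color1 into Color2 along an adjustable angle.
# A Hover scalar then blends that gradient toward HoverColor.
# A CornerRadius vector with named per-corner channels feeds the shared CoreUI rounded-mask function, with radii in pixels.
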Material /FortniteCoreUI/Frontend/Showdown/Materials/M_UI_Showdown_Panel_ButtonBg
graph {
 properties {
  bUsedWithStaticMesh=false
  blendMode=BLEND_Translucent
  materialDomain=MD_UI
  pixelDepthOffsetMode=PDOM_Legacy
  usedShadingModels=Default Lit
 }
 nodes [
  n1 = MaterialFunctionCall(
   UVs=MaterialFunctionCall() {materialFunction=&GetUserInterfaceUV}.Normalized UV
   "Rotation Angle (0-1)"=ScalarParameter() {defaultValue=0.2, parameterName=Rotation, sliderMax=1}
  ) {materialFunction=&CustomRotator}
  LinearInterpolate(
   A=LinearInterpolate(
    A=VectorParameter() {parameterName=Color1, defaultValue=<rgba 1 1 1 1>}.RGB
    B=VectorParameter() {parameterName=Color2, defaultValue=<rgba 1 1 1 1>}.RGB
    Alpha=ComponentMask(None=@n1.Rotated Values) {r=true}
   )
   B=VectorParameter() {parameterName=HoverColor, defaultValue=<rgba 1 1 1 1>}.RGB
   Alpha=ScalarParameter() {parameterName=Hover}
  )
  n2 = VectorParameter() {parameterName=CornerRadius, channelNames {a=Bottom Right, b=Top Right, g=Bottom Left, r=Top Left}, defaultValue=<rgba 0 20 0 20>}
  MaterialFunctionCall(Corner Radius=AppendVector(A=@n2, B=@n2), Radius in Pixels=Constant() {r=1}) {materialFunction=&MF_UI_CoreUI_RoundedMask}
  MaterialFunctionCall() {materialFunction=&GetUserInterfaceUV}
 ]
}
