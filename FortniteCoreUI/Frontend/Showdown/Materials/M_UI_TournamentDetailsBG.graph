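# Translucent UI material that draws a two-color gradient, named as the background for the Showdown tournament details screen.
# CustomRotator turns the texture coordinates by a fixed angle, and the red channel of the result becomes the blend factor.
# That factor lerps from the Color2 parameter to Color1, both blue by default, giving a static tilted gradient without any texture.
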Material /FortniteCoreUI/Frontend/Showdown/Materials/M_UI_TournamentDetailsBG
graph {
 properties {
  bUsedWithStaticMesh=false
  blendMode=BLEND_Translucent
  materialDomain=MD_UI
  pixelDepthOffsetMode=PDOM_Legacy
  usedShadingModels=Default Lit
 }
 nodes [
  n1 = TextureCoordinate()
  n2 = Constant() {r=-0.1}
  LinearInterpolate(
   A=VectorParameter() {parameterName=Color2, defaultValue=<rgba 0.0097 0.0232 0.15 1>}.RGB
   B=VectorParameter() {parameterName=Color1, defaultValue=<rgba 0.0319 0.1144 0.5711 1>}.RGB
   Alpha=ComponentMask(None=MaterialFunctionCall(UVs=@n1, "Rotation Angle (0-1)"=@n2) {materialFunction=&CustomRotator}.Rotated Values) {r=true}
  )
 ]
}
