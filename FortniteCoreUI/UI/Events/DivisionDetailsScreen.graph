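# Arena division details modal screen, built on FortDivisionDetailsScreen, with an About page and a Scoring payout page in a widget switcher.
# The Next button cycles pages and updates the title and button label, while Back pops the panel from the modal layer.
# Shoulder button functions change division only on the Scoring page.
# Activation plays an intro that scales the panel in and ramps up background blur.
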WidgetBlueprint /FortniteCoreUI/UI/Events/DivisionDetailsScreen
animations [Intro {
  binding_count=2
  length_seconds=0.2526
  object=&self:Intro
  sequence {bindings [
    SizeBox_TopContainer {
     id=#1
     possessed_class=SizeBox
     tracks [{
       type=MovieScene2DTransformTrack
       sections [{
         type=MovieScene2DTransformSection
         end_frame=5
         start_frame=0
         channels [
          Translation.X {type=MovieSceneScriptingFloatChannel}
          Translation.Y {type=MovieSceneScriptingFloatChannel}
          Angle {type=MovieSceneScriptingFloatChannel}
          Scale.X {type=MovieSceneScriptingFloatChannel, keys [[0,0.8563],[4.0213,1.0254],[4.6747,0.9945],[5.0527,1]]}
          Scale.Y {type=MovieSceneScriptingFloatChannel, keys [[0,0.8563],[4.0213,1.0254],[4.6747,0.9945],[5.0527,1]]}
          Shear.X {type=MovieSceneScriptingFloatChannel}
          Shear.Y {type=MovieSceneScriptingFloatChannel}
         ]
        }]
      }]
    }
    BackgroundBlur_1 {
     id=#2
     possessed_class=BackgroundBlur
     tracks [{
       type=MovieSceneFloatTrack
       display_name=Blur Strength
       sections [{type=MovieSceneFloatSection, end_frame=4, start_frame=0, channels [None {type=MovieSceneScriptingFloatChannel, keys [[0,0],[4.3837,25]]}]}]
      }]
    }
   ]}
 }]
graphs [
 EmptyFunction {nodes [
   EmptyFunction() {is=title K2Node_FunctionEntry}
   > Return Node!(execute, Handled:bool=true)
  ]}
 EventGraph {nodes [
   n1 = "On Button Clicked (BackButton)"!() -> (OutputDelegate, then, Button:Common Button Legacy Object Reference) {is=title K2Node_ComponentBoundEvent}
   PopActivatablePanelInModalLayer!(execute=[@n6.then, @n1.then], Panel:Common Activatable Panel Legacy Object Reference=Self-Reference()) -> (then)
   "On Button Clicked (NextButton)"!() -> (OutputDelegate, then, Button:Common Button Legacy Object Reference) {is=title K2Node_ComponentBoundEvent}
   n2 = Get WidgetSwitcher_MainContent() -> (WidgetSwitcher_MainContent:Common Widget Switcher Legacy Object Reference)
   n3 = "% (Integer)"(
    A=int + int(A=GetActiveWidgetIndex!(self:Widget Switcher Object Reference=@n2) -> (ReturnValue:int=0), B=1) -> (ReturnValue=0) {is=title K2Node_CommutativeAssociativeBinaryOperator}
    B=GetNumWidgets!(self:Widget Switcher Object Reference=@n2) -> (ReturnValue:int=0)
   ) -> (ReturnValue=0)
   > SetActiveWidgetIndex(self=@n2, Index=@n3)
   > "SetText (Text)"(
    self=Get Text_Title() -> (Text_Title:Common Text Object Reference)
    InText=Select(Option 0:Text="LOCTEXT(\"Arena Details\")", Option 1:Text="LOCTEXT(\"Arena Scoring\")", Index:int=@n3) -> (ReturnValue:Text)
   )
   > "SetText (Text)"(
    self=Get CenterButtonTextWidget!(self:Events Simple Button Object Reference=Get NextButton() -> (NextButton:Events Simple Button Object Reference)) -> (CenterButtonTextWidget:Common Text Object Reference)
    InText=Select(Option 0="LOCTEXT(\"SCORING\")", Option 1="LOCTEXT(\"ABOUT\")", Index=@n3)
   )
   Event On Activated()
   > PlayAnimation(InAnimation=Get Intro())
   n4 = Get WidgetSwitcher_MainContent()
   n5 = GetActiveWidgetIndex(self=@n4) -> (ReturnValue=0)
   n6 = Branch(Condition="Equal (Integer)"(A=@n5, B=0) {is=title K2Node_CallFunction})
   SetActiveWidgetIndex(execute=@n6.else, self=@n4, Index="Max (Integer)"(A=int - int(A=@n5, B=1) {is=title K2Node_CallFunction}, B=0))
  ]}
 LeftShoulderButton {nodes [
   LeftShoulderButton() {is=title K2Node_FunctionEntry}
   n1 = Get WidgetSwitcher_MainContent()
   > Branch(Condition="Equal (Integer)"(A=GetActiveWidgetIndex(self=@n1) -> (ReturnValue=0), B=1))
   > HandleLeftDivisionClicked(self:Fort Division Details Screen Object Reference)
   > Return Node(Handled=true)
  ]}
 RightShoulderButton {nodes [
   RightShoulderButton() {is=title K2Node_FunctionEntry}
   n1 = Get WidgetSwitcher_MainContent()
   > Branch(Condition="Equal (Integer)"(A=GetActiveWidgetIndex(self=@n1) -> (ReturnValue=0), B=1))
   > HandleRightDivisionClicked(self:Fort Division Details Screen Object Reference)
   > Return Node(Handled=true)
  ]}
]
variables []
widgets {
 info {inheritedWidgetCount=4, parentClass=/Script/FortniteUI.FortDivisionDetailsScreen, rootWidgetClass=&Overlay, widgetCount=20}
 widgets [
  Overlay_0 {widgetClassPath=&Overlay, widget=Overlay_0}
  BackgroundBlur_1 {
   widgetClassPath=&BackgroundBlur
   parent=Overlay_0
   widget=BackgroundBlur_1
   properties {
    blurStrength=25
    visibility=Visible
    lowQualityFallbackBrush {
     bIsDynamicallyLoaded=false
     drawAs=Image
     imageType=NoImage
     mirroring=NoMirror
     resourceName=None
     resourceObject=null
     tiling=NoTile
     imageSize=<0 0>
     margin=<ltrb 0 0 0 0>
     outlineSettings {
      bUseBrushTransparency=false
      roundingType=HalfHeightRadius
      width=0
      color {colorUseRule=UseColor_Specified, specifiedColor=<rgba 0 0 0 0>}
      cornerRadii=<0 0 0 0>
     }
     tintColor {colorUseRule=UseColor_Specified, specifiedColor=<rgba 0 0 0 0.8>}
     uVRegion {bIsValid=false, max=<0 0>, min=<0 0>}
    }
   }
   slot {class=OverlaySlot, properties {horizontalAlignment=HAlign_Fill, verticalAlignment=VAlign_Fill}}
  }
  SafeZone_1 {
   widgetClassPath=&SafeZone
   bIsVariable=true
   parent=BackgroundBlur_1
   widget=SafeZone_1
   slot {class=BackgroundBlurSlot, properties {padding=<ltrb 0 0 0 0>}}
  }
  SizeBox_TopContainer {
   widgetClassPath=&SizeBox
   bIsVariable=true
   parent=SafeZone_1
   widget=SizeBox_TopContainer
   properties {bOverride_HeightOverride=true, bOverride_WidthOverride=true, heightOverride=814, widthOverride=1743}
   slot {class=SafeZoneSlot, properties {hAlign=HAlign_Center, vAlign=VAlign_Center}}
  }
  Overlay_1 {widgetClassPath=&Overlay, parent=SizeBox_TopContainer, widget=Overlay_1, slot {class=SizeBoxSlot}}
  Image_12 {
   widgetClassPath=&Image
   bIsVariable=true
   parent=Overlay_1
   widget=Image_12
   properties {
    brush {
     bIsDynamicallyLoaded=false
     drawAs=Image
     imageType=NoImage
     mirroring=NoMirror
     resourceName=None
     resourceObject=/FortniteCoreUI/Frontend/Showdown/Materials/league/M_LeaguePanelBG_About
     tiling=NoTile
     imageSize=<32 256>
     margin=<ltrb 0 0.17 0 0.17>
     outlineSettings {
      bUseBrushTransparency=false
      roundingType=HalfHeightRadius
      width=0
      color {colorUseRule=UseColor_Specified, specifiedColor=<rgba 0 0 0 0>}
      cornerRadii=<0 0 0 0>
     }
     tintColor {colorUseRule=UseColor_Specified, specifiedColor=<rgba 1 1 1 1>}
     uVRegion {bIsValid=false, max=<0 0>, min=<0 0>}
    }
    colorAndOpacity=<rgba 0.0052 0.0052 0.0056 0.3>
    renderTransform {angle=0, scale=<1 1>, shear=<0 0>, translation=<26 26>}
   }
   slot {class=OverlaySlot, properties {horizontalAlignment=HAlign_Fill, verticalAlignment=VAlign_Fill, padding=<ltrb -10 -13 -10 -12>}}
  }
  Image_PanelBG {
   widgetClassPath=&Image
   bIsVariable=true
   parent=Overlay_1
   widget=Image_PanelBG
   properties {brush {
     bIsDynamicallyLoaded=false
     drawAs=Image
     imageType=NoImage
     mirroring=NoMirror
     resourceName=None
     resourceObject=&M_LeaguePanelBG_About
     tiling=NoTile
     imageSize=<32 256>
     margin=<ltrb 0 0.17 0 0.17>
     outlineSettings {
      bUseBrushTransparency=false
      roundingType=HalfHeightRadius
      width=0
      color {colorUseRule=UseColor_Specified, specifiedColor=<rgba 0 0 0 0>}
      cornerRadii=<0 0 0 0>
     }
     tintColor {colorUseRule=UseColor_Specified, specifiedColor=<rgba 1 1 1 1>}
     uVRegion {bIsValid=false, max=<0 0>, min=<0 0>}
    }}
   slot {class=OverlaySlot, properties {horizontalAlignment=HAlign_Fill, verticalAlignment=VAlign_Fill, padding=<ltrb -10 -13 -10 -12>}}
  }
  Overlay_3 {
   widgetClassPath=&Overlay
   parent=Overlay_1
   widget=Overlay_3
   slot {class=OverlaySlot, properties {horizontalAlignment=HAlign_Fill, verticalAlignment=VAlign_Fill}}
  }
  CommonBorder_0 {
   widgetClassPath=&CommonBorder
   parent=Overlay_3
   widget=CommonBorder_0
   properties {padding=<ltrb 64 128 128 50>}
   slot {class=OverlaySlot, properties {horizontalAlignment=HAlign_Fill, verticalAlignment=VAlign_Fill}}
  }
  WidgetSwitcher_MainContent {
   widgetClassPath=&CommonWidgetSwitcherLegacy
   bInherited=true
   bIsVariable=true
   parent=CommonBorder_0
   widget=WidgetSwitcher_MainContent
   properties {navigation=&self:WidgetTree.WidgetSwitcher_MainContent.WidgetNavigation_0}
   slot {class=BorderSlot, properties {padding=<ltrb 64 128 128 50>}}
  }
  GeneralDivisionInfo {
   widgetClassPath=&VerticalBox
   bIsVariable=true
   parent=WidgetSwitcher_MainContent
   widget=GeneralDivisionInfo
   slot {class=WidgetSwitcherSlot}
  }
  DetailsInfo {
   widgetClassPath=/FortniteCoreUI/Frontend/Showdown/ShowdownDivisionDetailsInfo
   bInherited=true
   bIsVariable=true
   parent=GeneralDivisionInfo
   widget=DetailsInfo
   slot {class=VerticalBoxSlot, properties {size {sizeRule=Fill, value=1}}}
  }
  ScoringPayoutInfo {
   widgetClassPath=&VerticalBox
   bIsVariable=true
   parent=WidgetSwitcher_MainContent
   widget=ScoringPayoutInfo
   slot {class=WidgetSwitcherSlot}
  }
  PayoutInfo {
   widgetClassPath=/FortniteCoreUI/Frontend/Showdown/ShowdownDivisionPayoutInfo
   bInherited=true
   bIsVariable=true
   parent=ScoringPayoutInfo
   widget=PayoutInfo
   slot {class=VerticalBoxSlot, properties {size {sizeRule=Fill, value=1}}}
  }
  CommonBorder_2 {
   widgetClassPath=&CommonBorder
   parent=Overlay_1
   widget=CommonBorder_2
   properties {clipping=ClipToBounds, padding=<ltrb 0 0 -10 0>}
   slot {class=OverlaySlot, properties {horizontalAlignment=HAlign_Right, verticalAlignment=VAlign_Bottom, padding=<ltrb 0 0 68 -2>}}
  }
  BackButton {
   widgetClassPath=/FortniteCoreUI/Frontend/Showdown/EventsSimpleButton
   bIsVariable=true
   parent=CommonBorder_2
   widget=BackButton
   properties {
    buttonNameText=Back
    clickMethod=DownAndUp
    onButtonBaseClicked=(&self:WidgetTree.BackButton.HandleOnButtonClicked)
    onButtonBaseDoubleClicked=(&self:WidgetTree.BackButton.HandleOnButtonDoubleClicked)
    onButtonBaseHovered=(&self:WidgetTree.BackButton.HandleOnButtonHovered)
    onButtonBaseUnhovered=(&self:WidgetTree.BackButton.HandleOnButtonUnhovered)
    onSelectedChangedBase=(&self:WidgetTree.BackButton.HandleOnSelectedChanged)
    pressMethod=DownAndUp
    visibility=Visible
    xL=false
    buttonPadding=<70 20>
    color {
     ButtonColor1=<rgba 0 1 0.9823 1>
     ButtonColor2=<rgba 0 1 0.9823 1>
     ButtonHover=<rgba 1 1 1 1>
     TextColor=<rgba 0.0052 0.0052 0.0056 1>
     TextHover=<rgba 0 0 0 1>
     TextPressed=<rgba 0 0 0 1>
     TextPressedHover=<rgba 0 0 0 1>
    }
    triggeringInputAction {dataTable=&InputActionDataTable, rowName=Back}
   }
   slot {class=BorderSlot, properties {padding=<ltrb 0 0 -10 0>}}
  }
  CommonBorder_1 {
   widgetClassPath=&CommonBorder
   parent=Overlay_1
   widget=CommonBorder_1
   properties {clipping=ClipToBounds, padding=<ltrb -10 0 0 0>}
   slot {class=OverlaySlot, properties {verticalAlignment=VAlign_Bottom, padding=<ltrb -4 0 50 11>}}
  }
  NextButton {
   widgetClassPath=&EventsSimpleButton
   bIsVariable=true
   parent=CommonBorder_1
   widget=NextButton
   properties {
    buttonNameText=Scoring
    clickMethod=DownAndUp
    navigation=&self:WidgetTree.NextButton.WidgetNavigation_0
    onButtonBaseClicked=(&self:WidgetTree.NextButton.HandleOnButtonClicked)
    onButtonBaseDoubleClicked=(&self:WidgetTree.NextButton.HandleOnButtonDoubleClicked)
    onButtonBaseHovered=(&self:WidgetTree.NextButton.HandleOnButtonHovered)
    onButtonBaseUnhovered=(&self:WidgetTree.NextButton.HandleOnButtonUnhovered)
    onSelectedChangedBase=(&self:WidgetTree.NextButton.HandleOnSelectedChanged)
    pressMethod=DownAndUp
    visibility=Visible
    xL=false
    buttonPadding=<70 20>
    color {
     ButtonColor1=<rgba 0 1 0.9823 1>
     ButtonColor2=<rgba 0 1 0.9823 1>
     ButtonHover=<rgba 1 1 1 1>
     TextColor=<rgba 0.0052 0.0052 0.0056 1>
     TextHover=<rgba 0 0 0 1>
     TextPressed=<rgba 0 0 0 1>
     TextPressedHover=<rgba 0 0 0 1>
    }
    triggeringInputAction {dataTable=&InputActionDataTable, rowName=GenericFaceButtonLeft}
   }
   slot {class=BorderSlot, properties {padding=<ltrb -10 0 0 0>}}
  }
  Image_PanelBorder {
   widgetClassPath=&Image
   bIsVariable=true
   parent=Overlay_1
   widget=Image_PanelBorder
   properties {
    visibility=HitTestInvisible
    brush {
     bIsDynamicallyLoaded=false
     drawAs=Image
     imageType=NoImage
     mirroring=NoMirror
     resourceName=None
     resourceObject=/FortniteCoreUI/Frontend/Showdown/Materials/league/M_UI_SkewBoxBorder_LeagueAbout
     tiling=NoTile
     imageSize=<520.659 300>
     margin=<ltrb 0.035 0.3 0.035 0.3>
     outlineSettings {
      bUseBrushTransparency=false
      roundingType=HalfHeightRadius
      width=0
      color {colorUseRule=UseColor_Specified, specifiedColor=<rgba 0 0 0 0>}
      cornerRadii=<0 0 0 0>
     }
     tintColor {colorUseRule=UseColor_Specified, specifiedColor=<rgba 1 1 1 1>}
     uVRegion {bIsValid=false, max=<0 0>, min=<0 0>}
    }
   }
   slot {class=OverlaySlot, properties {horizontalAlignment=HAlign_Fill, verticalAlignment=VAlign_Fill, padding=<ltrb -10 -13 -10 -12>}}
  }
  Text_Title {
   widgetClassPath=&CommonTextBlock
   bInherited=true
   bIsVariable=true
   parent=Overlay_1
   widget=Text_Title
   properties {
    navigation=&self:WidgetTree.Text_Title.WidgetNavigation_0
    text=Arena Details
    textTransformPolicy=ToUpper
    font {
     bForceMonospaced=false
     bMaterialIsStencil=false
     fontMaterial=null
     fontObject=&Burbank
     letterSpacing=0
     monospacedWidth=1
     size=60
     skewAmount=0
     typefaceFontName=Burbank-Black
     outlineSettings {
      bApplyOutlineToDropShadows=false
      bMiteredCorners=false
      bSeparateFillAlpha=false
      outlineMaterial=null
      outlineSize=2
      outlineColor=<rgba 0.0319 0.0052 0.162 1>
     }
    }
    shadowColorAndOpacity=<rgba 0.0273 0.0003 0.117 1>
    shadowOffset=<9 9>
   }
   slot {class=OverlaySlot, properties {padding=<ltrb 50 -20 0 0>}}
  }
 ]
}
