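# Showdown lobby panel widget, built on FortDivisionLobbyPanel, showing league title, division rank, hype points and rank-up progress.
# Refresh updates division data, swaps the progress bar for a points total at top division, and plays the intro once an event ID exists.
# Nested switchers show two-factor or match-cap warnings; the crossplay warning branch is unwired.
# The details button animates on hover, adjusts padding for gamepad and opens DivisionDetailsScreen as a modal.
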WidgetBlueprint /FortniteCoreUI/Frontend/Showdown/ShowdownDivisionLobbyPanel
animations [
 intro {
  binding_count=3
  display_label=""
  length_seconds=1.853
  object=&self:intro
  sequence {bindings [
    SizeBox_0 {
     id=#1
     possessed_class=SizeBox
     tracks [
      {
       type=MovieSceneFloatTrack
       display_name=Render Opacity
       sections [{type=MovieSceneFloatSection, end_frame=36, start_frame=0, channels [None {type=MovieSceneScriptingFloatChannel, keys [[0,0],[9.6777,1]]}]}]
      }
      {
       type=MovieScene2DTransformTrack
       sections [{
         type=MovieScene2DTransformSection
         end_frame=36
         start_frame=0
         channels [
          Translation.X {type=MovieSceneScriptingFloatChannel}
          Translation.Y {type=MovieSceneScriptingFloatChannel}
          Angle {type=MovieSceneScriptingFloatChannel}
          Scale.X {type=MovieSceneScriptingFloatChannel, keys [[0,1.1],[7.0467,0.9944],[8.7,1.0041],[9.6777,1]]}
          Scale.Y {type=MovieSceneScriptingFloatChannel, keys [[0,1.1],[7.0467,0.9944],[8.7,1.0041],[9.6777,1]]}
          Shear.X {type=MovieSceneScriptingFloatChannel}
          Shear.Y {type=MovieSceneScriptingFloatChannel}
         ]
        }]
      }
     ]
    }
    Image_Zag2 {
     id=#2
     possessed_class=Image
     tracks [{
       type=MovieSceneWidgetMaterialTrack
       sections [{
         type=MovieSceneParameterSection
         end_frame=37
         start_frame=0
         channels [RevealRadius {type=MovieSceneScriptingFloatChannel, keys [[0,0],[3.027,0],[22.6613,3]]}]
        }]
      }]
    }
    Image_Zag {
     id=#3
     possessed_class=Image
     tracks [{
       type=MovieSceneWidgetMaterialTrack
       sections [{
         type=MovieSceneParameterSection
         end_frame=37
         start_frame=0
         channels [RevealRadius {type=MovieSceneScriptingFloatChannel, keys [[0,0],[3.027,0],[22.6613,3]]}]
        }]
      }]
    }
   ]}
 }
 EventButtonHover {
  binding_count=1
  display_label=""
  length_seconds=0.7595
  object=&self:EventButtonHover
  sequence {bindings [Image_DetailsButtonBG {
     id=#4
     possessed_class=Image
     tracks [{
       type=MovieSceneWidgetMaterialTrack
       sections [{
         type=MovieSceneParameterSection
         end_frame=14
         start_frame=-1
         channels [Hover {type=MovieSceneScriptingFloatChannel, keys [[-0.07,0],[14.8303,1]]}]
        }]
      }]
    }]}
 }
]
graphs [
 EventGraph {nodes [
   "On Button Hovered (Button_EventDetails)"!() -> (OutputDelegate, then, Button:Common Button Legacy Object Reference) {is=title K2Node_ComponentBoundEvent}
   n1 = Self-Reference()
   n2 = Get EventButtonHover() -> (EventButtonHover:Widget Animation Object Reference)
   > Anim Transition!(Exec:Exec, In:bool=true, Animation:Widget Animation Object Reference=@n2, PlaySpeed:double=2, Widget:User Widget Object Reference=@n1) -> (ExecOut:Exec) {is=title K2Node_MacroInstance}
   "On Button Unhovered (Button_EventDetails)"!() -> (OutputDelegate, then, Button:Common Button Legacy Object Reference) {is=title K2Node_ComponentBoundEvent}
   > Anim Transition(Animation=@n2, PlaySpeed=1, Widget=@n1)
   "On Button Clicked (Button_EventDetails)"!() -> (OutputDelegate, then, Button:Common Button Legacy Object Reference) {is=title K2Node_ComponentBoundEvent}
   > n3 = Create Widget!(
    execute:Exec
    Class:User Widget Class Reference=/FortniteCoreUI/UI/Events/DivisionDetailsScreen.DivisionDetailsScreen_C
    OwningPlayer:Player Controller Object Reference=GetOwningPlayer()
   ) -> (then:Exec, ReturnValue:Division Details Screen Object Reference)
   > SetTournamentAndEventId!(
    execute
    self:Fort Division Details Screen Object Reference=@n3
    InTournamentId:String=GetTournamentId!(self:Fort Showdown Detail View Object Reference) -> (ReturnValue:String)
    InEventId:String=GetEventId!(self:Fort Showdown Detail View Object Reference) -> (ReturnValue:String)
   ) -> (then)
   > TryPushWidgetAsModalPanel!(execute, PushWidget:User Widget Object Reference=@n3) -> (then)
   EventColorize() {is=title K2Node_CustomEvent}
   > n4 = GetDynamicMaterial(self=Get Image_DetailsButtonBG() -> (Image_DetailsButtonBG:Image Object Reference))
   > SetVectorParameterValue(self=@n4, ParameterName=Color1, Value=(R=0,G=1,B=0.9823,A=1))
   > SetVectorParameterValue(self=@n4, ParameterName=Color2, Value=(R=0,G=1,B=0.9823,A=1))
   n5 = EventSwitchUI!() -> (OutputDelegate, then, bNewInputType:ECommonInputType Enum) {is=title K2Node_CustomEvent}
   n6 = IsUsingGamepad!(self:Fort Global UIContext Object Reference=Get FortGlobalUIContext!() -> (ReturnValue:Fort Global UIContext Object Reference) {is=title K2Node_GetBlueprintContextEx}) -> (ReturnValue:bool)
   SetPadding!(
    execute=[@n7.then, @n8.then, @n5.then]
    self:Overlay Slot Object Reference=SlotAsOverlaySlot(Widget=Get Text_DetailsButton() -> (Text_DetailsButton:Common Text Object Reference))
    InPadding:Margin Structure=Make Margin(Left=16, Top=0, Right=Select(Option 0:double=16, Option 1:double=48, Index=@n6) -> (ReturnValue:double), Bottom=0)
   ) -> (then)
   Event Construct()
   > EventColorize() {is=title K2Node_CallFunction}
   > SetVisibility(InVisibility=Collapsed)
   > Bind Event to On Crossplay Preference Changed(
    self:Fort Party Context Object Reference=Get FortPartyContext!() -> (ReturnValue:Fort Party Context Object Reference) {is=title K2Node_GetBlueprintContextEx}
    Delegate=Create Event()
   ) {is=title K2Node_AddDelegate}
   > n7 = Bind Event to On Input Method Changed(self=Get CommonInputSubsystem(), Delegate=@n5)
   Event Refresh Data() {is=title K2Node_Event}
   > n8 = Refresh()
  ]}
 HandleCrossplayChanged {nodes [
   HandleCrossplayChanged() {is=title K2Node_FunctionEntry}
   > Refresh()
  ]}
 InputTypeChanged {nodes [InputTypeChanged!() -> (then, isGamepad:bool=~) {is=title K2Node_FunctionEntry}]}
 Refresh {nodes [
   Refresh() {is=title K2Node_FunctionEntry}
   > RefreshDivisionData()
   > "SetText (Text)"(self=Get Text_DetailsButton(), InText=MakeLiteralText(Value="LOCTEXT(\"DIVISION DETAILS\")"))
   > SetVisibility(
    self=Get BorderTimeRemaining() -> (BorderTimeRemaining:Border Object Reference)
    InVisibility=Select(
     Option 0:ESlateVisibility Enum=SelfHitTestInvisible
     Option 1:ESlateVisibility Enum=Collapsed
     Index=IsDivisionEvent!(self:Fort Showdown Division Event View Object Reference) -> (ReturnValue:bool)
    ) -> (ReturnValue:ESlateVisibility Enum)
   )
   > SetText(
    self=Get RichTextBlock_TournamentTitle() -> (RichTextBlock_TournamentTitle:Common Rich Text Block Object Reference)
    InText=TextToUpper!(InText:"Text (by ref)"=GetLeagueDisplayName!(self:Fort Showdown Division Event View Object Reference) -> (ReturnValue:Text)) -> (ReturnValue:Text)
   )
   > SetVisibility(self=Get CentralContentSwitcher() -> (CentralContentSwitcher:Common Widget Switcher Legacy Object Reference), InVisibility=HitTestInvisible)
   > SetActiveWidget(self:Widget Switcher Object Reference=Get CentralContentSwitcher(), Widget=Get NoMessage() -> (NoMessage:Overlay Object Reference))
   n1 = SlotAsVerticalBoxSlot(Widget=Get GP_LeagueHeader() -> (GP_LeagueHeader:Grid Panel Object Reference))
   n2 = Break Margin!(Margin:"Margin Structure (by ref)"=Get Padding!(self:Vertical Box Slot Object Reference=@n1) -> (Padding:Margin Structure)) -> (Left:float=~, Top:float=~, Right:float=~, Bottom:float=~) {is=title K2Node_BreakStruct}
   > SetPadding(
    self:Vertical Box Slot Object Reference=@n1
    InPadding=Make Margin(Left=@n2, Top=Select(Option 0:double=0, Option 1:double=-50, Index=false) -> (ReturnValue:double), Right=@n2.Right, Bottom=@n2.Bottom)
   )
   n3 = MakeLiteralText(Value="LOCTEXT(\"UNTIL START\")")
   n4 = MakeLiteralText(Value="LOCTEXT(\"REMAINING\")")
   > "SetText (Text)"(
    self=Get Text_Remaining() -> (Text_Remaining:Common Text Object Reference)
    InText=Select!(
     Unknown:Text
     FutureTBD:Text=@n3
     FutureScheduled:Text=@n3
     FutureNext:Text=@n3
     Live:Text=@n4
     LiveParticipating:Text
     LiveNotParticipating:Text
     Completed:Text=@n4
     CompletedParticipated:Text
     CompletedNotPartipated:Text
     Cancelled:Text
     Index:EFortShowdownEventState Enum=GetEventTimeState!(self:Fort Showdown Detail View Object Reference) -> (ReturnValue:EFortShowdownEventState Enum=Unknown)
    ) -> (ReturnValue:Text)
   )
   > n5 = Branch(Condition="Equal Exactly (String)"(A=GetEventId()))
   SetVisibility(execute=@n5.else, InVisibility=SelfHitTestInvisible)
   > Anim Transition(In=true, Animation=Get intro() -> (intro:Widget Animation Object Reference), PlaySpeed=1)
   n6 = "integer >= integer"!(
    A:int=GetCurrentEventTeamNumMatchesPlayed!(self:Fort Showdown Detail View Object Reference) -> (ReturnValue:int=0)
    B:int=int - int(A=GetTournamentMatchCap!(self:Fort Showdown Detail View Object Reference) -> (ReturnValue:int=0), B=5) {is=title K2Node_CallFunction}
   ) -> (ReturnValue:bool) {is=title K2Node_CallFunction}
   n7 = "integer > integer"(A=GetTournamentMatchCap() -> (ReturnValue=0), B=0) {is=title K2Node_CallFunction}
   > Branch(Condition=OR Boolean(A=DoCalloutMFARequirement!(self:Fort Showdown Detail View Object Reference) -> (ReturnValue:bool), B=AND Boolean(A=@n6, B=@n7)))
   > SetActiveWidget(self=Get CentralContentSwitcher(), Widget=Get WarningMessage() -> (WarningMessage:Overlay Object Reference))
   > n8 = Branch(Condition=DoCalloutMFARequirement())
   > SetActiveWidget(
    self=Get CommonWidgetSwitcher_TournamentWarnings() -> (CommonWidgetSwitcher_TournamentWarnings:Common Widget Switcher Legacy Object Reference)
    Widget=Get MFA_WarningText() -> (MFA_WarningText:Common Text Object Reference)
   )
   SetActiveWidget(
    execute=@n8.else
    self=Get CommonWidgetSwitcher_TournamentWarnings()
    Widget=Get Text_MatchCapWarning() -> (Text_MatchCapWarning:Common Text Object Reference)
   )
   n9 = GetTournamentMatchCap() -> (ReturnValue=0)
   n10 = GetCurrentEventTeamNumMatchesPlayed() -> (ReturnValue=0)
   > "SetText (Text)"(
    self=Get Text_MatchCapWarning()
    InText=Format Text!(Format="LOCTEXT(\"{NumMatches} matches remaining\")", NumMatches:int=int - int(A=@n9, B=@n10)) -> (Result)
   )
   n11 = Branch(Condition=ShouldShowCrossplayError!(self:Fort Showdown Detail View Object Reference) -> (ReturnValue:bool))
   > SetActiveWidget(self=Get CentralContentSwitcher(), Widget=Get WarningMessage())
   > SetActiveWidget(
    self=Get CommonWidgetSwitcher_TournamentWarnings()
    Widget=Get Text_CrossplayWarning() -> (Text_CrossplayWarning:Common Text Object Reference)
   )
   Reroute Node(InputPin=@n11.else)
   "SetText (Text)"(InText="LOCTEXT(\"REMAINING\")")
  ]}
 RefreshDivisionData {nodes [
   RefreshDivisionData() {is=title K2Node_FunctionEntry}
   > Set TotalHype(TotalHype:int=GetTotalDivisionPoints!(self:Fort Showdown Division Event View Object Reference) -> (ReturnValue:int=0)) -> (Output_Get:int=0)
   > Set TotalHypeForNextDivision(TotalHypeForNextDivision:int=GetNumericalDisplayAmountForNextDivision!(self:Fort Showdown Division Event View Object Reference) -> (ReturnValue:int=0)) -> (Output_Get:int=0)
   > Set DivsionRank(DivsionRank:int=GetDivisionRank!(self:Fort Showdown Division Event View Object Reference) -> (ReturnValue:int=0)) -> (Output_Get:int=0)
   > Set TotalNumberDivisions(TotalNumberDivisions:int=GetTotalNumberOfDivisions!(self:Fort Showdown Division Event View Object Reference) -> (ReturnValue:int=0)) -> (Output_Get:int=0)
   n1 = Get DivsionRank() -> (DivsionRank:int=0)
   > "SetText (Text)"(
    self=Get Text_DivsionRank() -> (Text_DivsionRank:Common Text Object Reference)
    InText=Format Text!(
     Format="LOCTEXT(\"Division {Rank}\")"
     Rank:int=int + int(A=@n1, B=1) -> (ReturnValue=0) {is=title K2Node_CommutativeAssociativeBinaryOperator}
    ) -> (Result)
   )
   n2 = Get TotalHype() -> (TotalHype:int=0)
   n3 = Get TotalHypeForNextDivision() -> (TotalHypeForNextDivision:int=0)
   > "SetText (Text)"(self=Get Text_CurrentHype() -> (Text_CurrentHype:Common Text Object Reference), InText="To Text (Integer)"(Value=int - int(A=@n3, B=@n2)))
   > "SetText (Text)"(self=Get CommonTextBlock_1() -> (CommonTextBlock_1:Common Text Object Reference), InText="To Text (Integer)"(Value=@n2))
   > "SetText (Text)"(
    self=Get Text_DivsionProgress() -> (Text_DivsionProgress:Common Text Object Reference)
    InText=Format Text!(Format="LOCTEXT(\"{CurrentHype} / {RequiredHype}\")", CurrentHype:int=@n2, RequiredHype:int=@n3) -> (Result)
   )
   n4 = GetAmountEarnedWithinDivision!(self:Fort Showdown Division Event View Object Reference) -> (ReturnValue:int=0)
   n5 = GetRequiredAmountToAdvanceDivision!(self:Fort Showdown Division Event View Object Reference) -> (ReturnValue:int=0)
   > SetPercent!(
    execute
    self:Progress Bar Object Reference=Get Bar_DivisionProgress() -> (Bar_DivisionProgress:Progress Bar Object Reference)
    InPercent:float=float / float(A="To Float (Integer)"(InInt=@n4), B="To Float (Integer)"(InInt=@n5)) {is=title K2Node_CallFunction}
   ) -> (then)
   n6 = Get DivsionRank() -> (DivsionRank=0)
   > n7 = Branch(Condition="Equal (Integer)"(A=int + int(A=@n6, B=1) -> (ReturnValue=0), B=Get TotalNumberDivisions() -> (TotalNumberDivisions:int=0)) {is=title K2Node_CallFunction})
   n8 = Get WidgetSwitcher_Progress() -> (WidgetSwitcher_Progress:Widget Switcher Object Reference)
   > SetActiveWidget(self=@n8, Widget=Get HorizontalBoxJustHYpe() -> (HorizontalBoxJustHYpe:Horizontal Box Object Reference))
   n9 = Get EarnPinContent() -> (EarnPinContent:Vertical Box Object Reference)
   > SetVisibility(self=@n9, InVisibility=Hidden)
   SetActiveWidget(execute=@n7.else, self=@n8, Widget=Get HorizontalBox_DivisionProgress() -> (HorizontalBox_DivisionProgress:Horizontal Box Object Reference))
   > SetVisibility(self=@n9, InVisibility=SelfHitTestInvisible)
  ]}
 UpdateColors {nodes [UpdateColors() {is=title K2Node_FunctionEntry}]}
]
variables [
 HighlightColor {type=SlateColor, default {colorUseRule=UseColor_Specified, specifiedColor=<rgba 1 0 1 1>}}
 PrimaryColor {type=SlateColor, default {colorUseRule=UseColor_Specified, specifiedColor=<rgba 1 0 1 1>}}
 ShadowColor {type=SlateColor, default {colorUseRule=UseColor_Specified, specifiedColor=<rgba 1 0 1 1>}}
]
widgets {
 info {inheritedWidgetCount=10, parentClass=/Script/FortniteUI.FortDivisionLobbyPanel, rootWidgetClass=&SizeBox, widgetCount=59}
 widgets [
  SizeBox_0 {widgetClassPath=&SizeBox, widget=SizeBox_0, properties {bOverride_WidthOverride=true, widthOverride=384}}
  GridPanel_0 {
   widgetClassPath=&GridPanel
   parent=SizeBox_0
   widget=GridPanel_0
   properties {columnFill [1], rowFill [1]}
   slot {class=SizeBoxSlot}
  }
  BackgroundBlur_0 {
   widgetClassPath=&BackgroundBlur
   parent=GridPanel_0
   widget=BackgroundBlur_0
   properties {blurStrength=25, horizontalAlignment=HAlign_Left}
   slot {class=GridSlot}
  }
  CommonBorder_0 {
   widgetClassPath=&CommonBorder
   parent=GridPanel_0
   widget=CommonBorder_0
   properties {
    background {
     bIsDynamicallyLoaded=false
     drawAs=Image
     imageType=NoImage
     mirroring=NoMirror
     resourceName=None
     resourceObject=null
     tiling=NoTile
     imageSize=<32 32>
     margin=<ltrb 0 0 0 0>
     outlineSettings {
      bUseBrushTransparency=false
      roundingType=HalfHeightRadius
      width=0
      color {colorUseRule=UseColor_Specified, specifiedColor=<rgba 0 0 0 0>}
      cornerRadii=<0 0 0 0>
     }
     tintColor {colorUseRule=UseColor_Specified, specifiedColor=<rgba 1 1 1 1>}
     uVRegion {bIsValid=false, max=<0 0>, min=<0 0>}
    }
    brushColor=<rgba 0.0052 0.0052 0.0056 0.34>
    padding=<ltrb 0 0 0 0>
   }
   slot {class=GridSlot, properties {layer=10}}
  }
  GridPanel_1 {
   widgetClassPath=&GridPanel
   parent=CommonBorder_0
   widget=GridPanel_1
   properties {columnFill [1], rowFill [1]}
   slot {class=BorderSlot, properties {padding=<ltrb 0 0 0 0>}}
  }
  Image_PanelBG {
   widgetClassPath=&Image
   bIsVariable=true
   parent=GridPanel_1
   widget=Image_PanelBG
   properties {brush {
     bIsDynamicallyLoaded=false
     drawAs=Box
     imageType=NoImage
     mirroring=NoMirror
     resourceName=None
     resourceObject=/FortniteCoreUI/Frontend/Showdown/Materials/league/M_LeaguePanelBG_Inst
     tiling=NoTile
     imageSize=<32 256>
     margin=<ltrb 0 0.17 0 0.17>
     outlineSettings {
      bUseBrushTransparency=false
      roundingType=HalfHeightRadius
      width=0
      color {colorUseRule=UseColor_Specified, specifiedColor=<rgba 0 0 0 0>}
      cornerRadii=<0 0 0 0>
     }
     tintColor {colorUseRule=UseColor_Specified, specifiedColor=<rgba 1 1 1 1>}
     uVRegion {bIsValid=false, max=<0 0>, min=<0 0>}
    }}
   slot {class=GridSlot, properties {padding=<ltrb -10 -13 -10 -8>}}
  }
  Image_PanelBorder {
   widgetClassPath=&Image
   bIsVariable=true
   parent=GridPanel_1
   widget=Image_PanelBorder
   properties {brush {
     bIsDynamicallyLoaded=false
     drawAs=Box
     imageType=NoImage
     mirroring=NoMirror
     resourceName=None
     resourceObject=/Game/UI/Frontend/Showdown/Materials/M_UI_SkewBoxBorder_TournamentPanelBorder
     tiling=NoTile
     imageSize=<32 256>
     margin=<ltrb 0 0.17 0 0.17>
     outlineSettings {
      bUseBrushTransparency=false
      roundingType=HalfHeightRadius
      width=0
      color {colorUseRule=UseColor_Specified, specifiedColor=<rgba 0 0 0 0>}
      cornerRadii=<0 0 0 0>
     }
     tintColor {colorUseRule=UseColor_Specified, specifiedColor=<rgba 1 1 1 1>}
     uVRegion {bIsValid=false, max=<0 0>, min=<0 0>}
    }}
   slot {class=GridSlot, properties {layer=10, padding=<ltrb -10 -13 -10 -8>}}
  }
  VerticalBox_0 {widgetClassPath=&VerticalBox, parent=GridPanel_1, widget=VerticalBox_0, slot {class=GridSlot, properties {layer=20}}}
  GP_LeagueHeader {
   widgetClassPath=&GridPanel
   bIsVariable=true
   parent=VerticalBox_0
   widget=GP_LeagueHeader
   properties {columnFill [1], rowFill [1]}
   slot {class=VerticalBoxSlot, properties {verticalAlignment=VAlign_Top, padding=<ltrb 24 -48 24 24>}}
  }
  Image_CurrentDivisionFancy {
   widgetClassPath=&FortLazyImage
   bInherited=true
   bIsVariable=true
   parent=GP_LeagueHeader
   widget=Image_CurrentDivisionFancy
   properties {brush {
     bIsDynamicallyLoaded=false
     drawAs=Image
     imageType=NoImage
     mirroring=NoMirror
     resourceName=None
     resourceObject=/FortniteCoreUI/Frontend/Showdown/Materials/league/M_UI_LeagueDivision_Badge1
     tiling=NoTile
     imageSize=<256 256>
     margin=<ltrb 0 0 0 0>
     outlineSettings {
      bUseBrushTransparency=false
      roundingType=HalfHeightRadius
      width=0
      color {colorUseRule=UseColor_Specified, specifiedColor=<rgba 0 0 0 0>}
      cornerRadii=<0 0 0 0>
     }
     tintColor {colorUseRule=UseColor_Specified, specifiedColor=<rgba 1 1 1 1>}
     uVRegion {bIsValid=false, max=<0 0>, min=<0 0>}
    }}
   slot {class=GridSlot, properties {horizontalAlignment=HAlign_Center, verticalAlignment=VAlign_Top, padding=<ltrb 0 -16 0 0>}}
  }
  VerticalBox_2 {
   widgetClassPath=&VerticalBox
   parent=GP_LeagueHeader
   widget=VerticalBox_2
   slot {class=GridSlot, properties {layer=10, verticalAlignment=VAlign_Bottom, padding=<ltrb 0 232 0 0>}}
  }
  RichTextBlock_TournamentTitle {
   widgetClassPath=&CommonRichTextBlock
   bIsVariable=true
   parent=VerticalBox_2
   widget=RichTextBlock_TournamentTitle
   properties {
    bOverrideDefaultStyle=true
    justification=Center
    scrollStyle=&TextScrollStyle-Base_C
    text=Open League
    defaultTextStyleOverride {
     overflowPolicy=Clip
     transformPolicy=ToUpper
     colorAndOpacity {colorUseRule=UseColor_Specified, specifiedColor=<rgba 1 1 1 1>}
     font {
      bForceMonospaced=false
      bMaterialIsStencil=false
      fontMaterial=null
      fontObject=&HeadingNow
      letterSpacing=-40
      monospacedWidth=1
      size=32.25
      skewAmount=0
      typefaceFontName=86 Bold
      outlineSettings {
       bApplyOutlineToDropShadows=false
       bMiteredCorners=false
       bSeparateFillAlpha=false
       outlineMaterial=null
       outlineSize=0
       outlineColor=<rgba 0 0 0 1>
      }
     }
     highlightColor {colorUseRule=UseColor_Specified, specifiedColor=<rgba 0 0 0 1>}
     highlightShape {
      bIsDynamicallyLoaded=false
      drawAs=Image
      imageType=NoImage
      mirroring=NoMirror
      resourceName=None
      resourceObject=null
      tiling=NoTile
      imageSize=<32 32>
      margin=<ltrb 0 0 0 0>
      outlineSettings {
       bUseBrushTransparency=false
       roundingType=HalfHeightRadius
       width=0
       color {colorUseRule=UseColor_Specified, specifiedColor=<rgba 0 0 0 0>}
       cornerRadii=<0 0 0 0>
      }
      tintColor {colorUseRule=UseColor_Specified, specifiedColor=<rgba 1 1 1 1>}
      uVRegion {bIsValid=false, max=<0 0>, min=<0 0>}
     }
     selectedBackgroundColor {colorUseRule=UseColor_Foreground, specifiedColor=<rgba 1 0 1 1>}
     shadowColorAndOpacity=<rgba 0.0273 0.0012 0.1221 1>
     shadowOffset=<1 1>
     strikeBrush {
      bIsDynamicallyLoaded=false
      drawAs=Image
      imageType=NoImage
      mirroring=NoMirror
      resourceName=None
      resourceObject=null
      tiling=NoTile
      imageSize=<32 32>
      margin=<ltrb 0 0 0 0>
      outlineSettings {
       bUseBrushTransparency=false
       roundingType=HalfHeightRadius
       width=0
       color {colorUseRule=UseColor_Specified, specifiedColor=<rgba 0 0 0 0>}
       cornerRadii=<0 0 0 0>
      }
      tintColor {colorUseRule=UseColor_Specified, specifiedColor=<rgba 1 1 1 1>}
      uVRegion {bIsValid=false, max=<0 0>, min=<0 0>}
     }
     underlineBrush {
      bIsDynamicallyLoaded=false
      drawAs=Image
      imageType=NoImage
      mirroring=NoMirror
      resourceName=None
      resourceObject=null
      tiling=NoTile
      imageSize=<32 32>
      margin=<ltrb 0 0 0 0>
      outlineSettings {
       bUseBrushTransparency=false
       roundingType=HalfHeightRadius
       width=0
       color {colorUseRule=UseColor_Specified, specifiedColor=<rgba 0 0 0 0>}
       cornerRadii=<0 0 0 0>
      }
      tintColor {colorUseRule=UseColor_Specified, specifiedColor=<rgba 1 1 1 1>}
      uVRegion {bIsValid=false, max=<0 0>, min=<0 0>}
     }
    }
   }
   slot {class=VerticalBoxSlot, properties {verticalAlignment=VAlign_Top}}
  }
  Text_DivsionRank {
   widgetClassPath=&CommonTextBlock
   bIsVariable=true
   parent=VerticalBox_2
   widget=Text_DivsionRank
   properties {
    justification=Center
    scrollStyle=&TextScrollStyle-Base_C
    text=DIVISION 1
    textTransformPolicy=ToUpper
    colorAndOpacity {colorUseRule=UseColor_Specified, specifiedColor=<rgba 0 1 0.9823 1>}
    font {
     bForceMonospaced=false
     bMaterialIsStencil=false
     fontMaterial=null
     fontObject=&BurbankSmall
     letterSpacing=0
     monospacedWidth=1
     size=18
     skewAmount=0
     typefaceFontName=Black
     outlineSettings {
      bApplyOutlineToDropShadows=false
      bMiteredCorners=false
      bSeparateFillAlpha=false
      outlineMaterial=null
      outlineSize=0
      outlineColor=<rgba 0 0 0 1>
     }
    }
   }
   slot {class=VerticalBoxSlot, properties {verticalAlignment=VAlign_Top}}
  }
  WidgetSwitcher_Progress {
   widgetClassPath=&WidgetSwitcher
   bIsVariable=true
   parent=VerticalBox_0
   widget=WidgetSwitcher_Progress
   slot {class=VerticalBoxSlot, properties {padding=<ltrb 8 0 8 16>}}
  }
  HorizontalBox_DivisionProgress {
   widgetClassPath=&HorizontalBox
   bIsVariable=true
   parent=WidgetSwitcher_Progress
   widget=HorizontalBox_DivisionProgress
   slot {class=WidgetSwitcherSlot}
  }
  Image_CurrentDivisionPlain {
   widgetClassPath=&FortLazyImage
   bInherited=true
   bIsVariable=true
   parent=HorizontalBox_DivisionProgress
   widget=Image_CurrentDivisionPlain
   properties {brush {
     bIsDynamicallyLoaded=false
     drawAs=Image
     imageType=NoImage
     mirroring=NoMirror
     resourceName=None
     resourceObject=/FortniteCoreUI/Frontend/Showdown/Materials/league/M_UI_LeagueDivision_Icon1
     tiling=NoTile
     imageSize=<64 64>
     margin=<ltrb 0 0 0 0>
     outlineSettings {
      bUseBrushTransparency=false
      roundingType=HalfHeightRadius
      width=0
      color {colorUseRule=UseColor_Specified, specifiedColor=<rgba 0 0 0 0>}
      cornerRadii=<0 0 0 0>
     }
     tintColor {colorUseRule=UseColor_Specified, specifiedColor=<rgba 1 1 1 1>}
     uVRegion {bIsValid=false, max=<0 0>, min=<0 0>}
    }}
   slot {class=HorizontalBoxSlot}
  }
  Overlay_0 {
   widgetClassPath=&Overlay
   parent=HorizontalBox_DivisionProgress
   widget=Overlay_0
   slot {class=HorizontalBoxSlot, properties {verticalAlignment=VAlign_Center, size {sizeRule=Fill, value=1}}}
  }
  Bar_DivisionProgress {
   widgetClassPath=/Script/UMG.ProgressBar
   bIsVariable=true
   parent=Overlay_0
   widget=Bar_DivisionProgress
   properties {
    percent=0.6
    widgetStyle {
     enableFillAnimation=false
     backgroundImage {
      bIsDynamicallyLoaded=false
      drawAs=Image
      imageType=FullColor
      mirroring=NoMirror
      resourceName=../../../Engine/Content/Slate/Common/ProgressBar_Background.png
      resourceObject=/Game/Global/Textures/Default/Blanks/T_WhiteUserInterface
      tiling=NoTile
      imageSize=<16 16>
      margin=<ltrb 0.4167 0.4167 0.4167 0.4167>
      outlineSettings {
       bUseBrushTransparency=false
       roundingType=HalfHeightRadius
       width=0
       color {colorUseRule=UseColor_Specified, specifiedColor=<rgba 0 0 0 0>}
       cornerRadii=<0 0 0 0>
      }
      tintColor {colorUseRule=UseColor_Specified, specifiedColor=<rgba 0.0052 0.0052 0.0056 0.5>}
      uVRegion {bIsValid=false, max=<0 0>, min=<0 0>}
     }
     fillImage {
      bIsDynamicallyLoaded=false
      drawAs=Image
      imageType=FullColor
      mirroring=NoMirror
      resourceName=../../../Engine/Content/Slate/Common/ProgressBar_Fill.png
      resourceObject=&T_WhiteUserInterface
      tiling=NoTile
      imageSize=<16 16>
      margin=<ltrb 0.4167 0.4167 0.4167 0.4167>
      outlineSettings {
       bUseBrushTransparency=false
       roundingType=HalfHeightRadius
       width=0
       color {colorUseRule=UseColor_Specified, specifiedColor=<rgba 0 0 0 0>}
       cornerRadii=<0 0 0 0>
      }
      tintColor {colorUseRule=UseColor_Specified, specifiedColor=<rgba 0 1 0.9823 1>}
      uVRegion {bIsValid=false, max=<0 0>, min=<0 0>}
     }
     marqueeImage {
      bIsDynamicallyLoaded=false
      drawAs=Image
      imageType=FullColor
      mirroring=NoMirror
      resourceName=../../../Engine/Content/Slate/Common/ProgressBar_Marquee.png
      resourceObject=null
      tiling=Horizontal
      imageSize=<20 12>
      margin=<ltrb 0 0 0 0>
      outlineSettings {
       bUseBrushTransparency=false
       roundingType=HalfHeightRadius
       width=0
       color {colorUseRule=UseColor_Specified, specifiedColor=<rgba 0 0 0 0>}
       cornerRadii=<0 0 0 0>
      }
      tintColor {colorUseRule=UseColor_Specified, specifiedColor=<rgba 1 1 1 1>}
      uVRegion {bIsValid=false, max=<0 0>, min=<0 0>}
     }
    }
   }
   slot {class=OverlaySlot, properties {horizontalAlignment=HAlign_Fill}}
  }
  HorizontalBox_2 {
   widgetClassPath=&HorizontalBox
   parent=Overlay_0
   widget=HorizontalBox_2
   properties {renderTransform {angle=0, scale=<1 1>, shear=<0 0>, translation=<0 4>}}
   slot {class=OverlaySlot, properties {horizontalAlignment=HAlign_Center, padding=<ltrb 0 12 0 -96>}}
  }
  Image_4 {
   widgetClassPath=&Image
   bIsVariable=true
   parent=HorizontalBox_2
   widget=Image_4
   properties {
    brush {
     bIsDynamicallyLoaded=false
     drawAs=Image
     imageType=NoImage
     mirroring=NoMirror
     resourceName=None
     resourceObject=/FortniteCoreUI/Frontend/Showdown/Materials/league/M_UI_LeagueDivision_PlayerRatingIcon
     tiling=NoTile
     imageSize=<24 24>
     margin=<ltrb 0 0 0 0>
     outlineSettings {
      bUseBrushTransparency=false
      roundingType=HalfHeightRadius
      width=0
      color {colorUseRule=UseColor_Specified, specifiedColor=<rgba 0 0 0 0>}
      cornerRadii=<0 0 0 0>
     }
     tintColor {colorUseRule=UseColor_Specified, specifiedColor=<rgba 1 1 1 1>}
     uVRegion {bIsValid=false, max=<0 0>, min=<0 0>}
    }
    colorAndOpacity=<rgba 0 1 0.9823 1>
   }
   slot {class=HorizontalBoxSlot, properties {horizontalAlignment=HAlign_Left, verticalAlignment=VAlign_Center, padding=<ltrb 0 0 4 0>}}
  }
  Text_DivsionProgress {
   widgetClassPath=&CommonTextBlock
   bIsVariable=true
   parent=HorizontalBox_2
   widget=Text_DivsionProgress
   properties {
    justification=Center
    text=155/200
    colorAndOpacity {colorUseRule=UseColor_Specified, specifiedColor=<rgba 0 1 0.9823 1>}
    font {
     bForceMonospaced=false
     bMaterialIsStencil=false
     fontMaterial=null
     fontObject=&BurbankSmall
     letterSpacing=0
     monospacedWidth=1
     size=18
     skewAmount=0
     typefaceFontName=Black
     outlineSettings {
      bApplyOutlineToDropShadows=false
      bMiteredCorners=false
      bSeparateFillAlpha=false
      outlineMaterial=null
      outlineSize=0
      outlineColor=<rgba 0 0 0 1>
     }
    }
   }
   slot {class=HorizontalBoxSlot, properties {horizontalAlignment=HAlign_Center, verticalAlignment=VAlign_Center}}
  }
  Image_NextDivision {
   widgetClassPath=&FortLazyImage
   bInherited=true
   bIsVariable=true
   parent=HorizontalBox_DivisionProgress
   widget=Image_NextDivision
   properties {brush {
     bIsDynamicallyLoaded=false
     drawAs=Image
     imageType=NoImage
     mirroring=NoMirror
     resourceName=None
     resourceObject=/FortniteCoreUI/Frontend/Showdown/Materials/league/M_UI_LeagueDivision_Icon6
     tiling=NoTile
     imageSize=<64 64>
     margin=<ltrb 0 0 0 0>
     outlineSettings {
      bUseBrushTransparency=false
      roundingType=HalfHeightRadius
      width=0
      color {colorUseRule=UseColor_Specified, specifiedColor=<rgba 0 0 0 0>}
      cornerRadii=<0 0 0 0>
     }
     tintColor {colorUseRule=UseColor_Specified, specifiedColor=<rgba 1 1 1 1>}
     uVRegion {bIsValid=false, max=<0 0>, min=<0 0>}
    }}
   slot {class=HorizontalBoxSlot, properties {horizontalAlignment=HAlign_Center, verticalAlignment=VAlign_Center}}
  }
  HorizontalBoxJustHYpe {
   widgetClassPath=&HorizontalBox
   bIsVariable=true
   parent=WidgetSwitcher_Progress
   widget=HorizontalBoxJustHYpe
   slot {class=WidgetSwitcherSlot, properties {horizontalAlignment=HAlign_Center, verticalAlignment=VAlign_Top, padding=<ltrb 0 12 0 -100>}}
  }
  Image_HypeIcon {
   widgetClassPath=&Image
   parent=HorizontalBoxJustHYpe
   widget=Image_HypeIcon
   properties {
    brush {
     bIsDynamicallyLoaded=false
     drawAs=Image
     imageType=NoImage
     mirroring=NoMirror
     resourceName=None
     resourceObject=&M_UI_LeagueDivision_PlayerRatingIcon
     tiling=NoTile
     imageSize=<42 42>
     margin=<ltrb 0 0 0 0>
     outlineSettings {
      bUseBrushTransparency=false
      roundingType=HalfHeightRadius
      width=0
      color {colorUseRule=UseColor_Specified, specifiedColor=<rgba 0 0 0 0>}
      cornerRadii=<0 0 0 0>
     }
     tintColor {colorUseRule=UseColor_Specified, specifiedColor=<rgba 1 1 1 1>}
     uVRegion {bIsValid=false, max=<0 0>, min=<0 0>}
    }
    colorAndOpacity=<rgba 0 1 0.9823 1>
   }
   slot {class=HorizontalBoxSlot, properties {verticalAlignment=VAlign_Center, padding=<ltrb -5 0 0 0>}}
  }
  CommonTextBlock_1 {
   widgetClassPath=&CommonTextBlock
   bIsVariable=true
   parent=HorizontalBoxJustHYpe
   widget=CommonTextBlock_1
   properties {
    justification=Center
    text=200
    colorAndOpacity {colorUseRule=UseColor_Specified, specifiedColor=<rgba 0 1 0.9823 1>}
    font {
     bForceMonospaced=false
     bMaterialIsStencil=false
     fontMaterial=null
     fontObject=&BurbankSmall
     letterSpacing=0
     monospacedWidth=1
     size=34
     skewAmount=0
     typefaceFontName=Black
     outlineSettings {
      bApplyOutlineToDropShadows=false
      bMiteredCorners=false
      bSeparateFillAlpha=false
      outlineMaterial=null
      outlineSize=0
      outlineColor=<rgba 0 0 0 1>
     }
    }
   }
   slot {class=HorizontalBoxSlot, properties {horizontalAlignment=HAlign_Center, verticalAlignment=VAlign_Center}}
  }
  CentralContentSwitcher {
   widgetClassPath=&CommonWidgetSwitcherLegacy
   bIsVariable=true
   parent=VerticalBox_0
   widget=CentralContentSwitcher
   properties {visibility=Collapsed}
   slot {class=VerticalBoxSlot, properties {verticalAlignment=VAlign_Top, padding=<ltrb 0 24 0 0>}}
  }
  WarningMessage {
   widgetClassPath=&Overlay
   bIsVariable=true
   parent=CentralContentSwitcher
   widget=WarningMessage
   slot {class=WidgetSwitcherSlot, properties {verticalAlignment=VAlign_Top, padding=<ltrb -60 0 -60 0>}}
  }
  Image_ErrorBG {
   widgetClassPath=&Image
   bIsVariable=true
   parent=WarningMessage
   widget=Image_ErrorBG
   properties {
    visibility=SelfHitTestInvisible
    brush {
     bIsDynamicallyLoaded=false
     drawAs=Image
     imageType=NoImage
     mirroring=NoMirror
     resourceName=None
     resourceObject=/FortniteCoreUI/Frontend/Showdown/Materials/M_UI_SkewBox_TournamentPanelErrorBG
     tiling=NoTile
     imageSize=<32 60.8663>
     margin=<ltrb 0 0 0 0>
     outlineSettings {
      bUseBrushTransparency=false
      roundingType=HalfHeightRadius
      width=0
      color {colorUseRule=UseColor_Specified, specifiedColor=<rgba 0 0 0 0>}
      cornerRadii=<0 0 0 0>
     }
     tintColor {colorUseRule=UseColor_Specified, specifiedColor=<rgba 1 1 1 1>}
     uVRegion {bIsValid=false, max=<0 0>, min=<0 0>}
    }
    colorAndOpacity=<rgba 0 1 0.9823 1>
    renderTransformPivot=<0 1>
   }
   slot {class=OverlaySlot, properties {horizontalAlignment=HAlign_Fill, verticalAlignment=VAlign_Fill, padding=<ltrb 0 -15 -15 -5>}}
  }
  CommonWidgetSwitcher_TournamentWarnings {
   widgetClassPath=&CommonWidgetSwitcherLegacy
   bIsVariable=true
   parent=WarningMessage
   widget=CommonWidgetSwitcher_TournamentWarnings
   slot {class=OverlaySlot, properties {horizontalAlignment=HAlign_Fill, padding=<ltrb 40 10 40 25>}}
  }
  Text_CrossplayWarning {
   widgetClassPath=&CommonTextBlock
   bIsVariable=true
   parent=CommonWidgetSwitcher_TournamentWarnings
   widget=Text_CrossplayWarning
   properties {
    autoWrapText=true
    justification=Center
    lineHeightPercentage=0.9
    text=Opt-in for crossplay to play in competitive events
    wrapTextAt=366
    colorAndOpacity {colorUseRule=UseColor_Specified, specifiedColor=<rgba 0.0044 0.0091 0.0497 1>}
    font {
     bForceMonospaced=false
     bMaterialIsStencil=false
     fontMaterial=null
     fontObject=&Burbank
     letterSpacing=0
     monospacedWidth=1
     size=22
     skewAmount=0
     typefaceFontName=BurbankRegular-Black
     outlineSettings {
      bApplyOutlineToDropShadows=false
      bMiteredCorners=false
      bSeparateFillAlpha=false
      outlineMaterial=null
      outlineSize=0
      outlineColor=<rgba 0 0 0 1>
     }
    }
    renderTransform {angle=0, scale=<1 1>, shear=<-7 0>, translation=<0 0>}
   }
   slot {class=WidgetSwitcherSlot}
  }
  CommonTextBlock_0 {
   widgetClassPath=&CommonTextBlock
   parent=CommonWidgetSwitcher_TournamentWarnings
   widget=CommonTextBlock_0
   properties {
    autoWrapText=true
    justification=Center
    lineHeightPercentage=0.9
    text="You need more \r\nplayers to compete"
    wrapTextAt=366
    colorAndOpacity {colorUseRule=UseColor_Specified, specifiedColor=<rgba 0.0044 0.0091 0.0497 1>}
    font {
     bForceMonospaced=false
     bMaterialIsStencil=false
     fontMaterial=null
     fontObject=&Burbank
     letterSpacing=0
     monospacedWidth=1
     size=22
     skewAmount=0
     typefaceFontName=BurbankRegular-Black
     outlineSettings {
      bApplyOutlineToDropShadows=false
      bMiteredCorners=false
      bSeparateFillAlpha=false
      outlineMaterial=null
      outlineSize=0
      outlineColor=<rgba 0 0 0 1>
     }
    }
    renderTransform {angle=0, scale=<1 1>, shear=<-7 0>, translation=<0 0>}
   }
   slot {class=WidgetSwitcherSlot}
  }
  Text_MatchCapWarning {
   widgetClassPath=&CommonTextBlock
   bIsVariable=true
   parent=CommonWidgetSwitcher_TournamentWarnings
   widget=Text_MatchCapWarning
   properties {
    autoWrapText=true
    justification=Center
    lineHeightPercentage=0.9
    text=5 matches remaining
    wrapTextAt=366
    colorAndOpacity {colorUseRule=UseColor_Specified, specifiedColor=<rgba 0.0044 0.0091 0.0497 1>}
    font {
     bForceMonospaced=false
     bMaterialIsStencil=false
     fontMaterial=null
     fontObject=&Burbank
     letterSpacing=0
     monospacedWidth=1
     size=22
     skewAmount=0
     typefaceFontName=BurbankRegular-Black
     outlineSettings {
      bApplyOutlineToDropShadows=false
      bMiteredCorners=false
      bSeparateFillAlpha=false
      outlineMaterial=null
      outlineSize=0
      outlineColor=<rgba 0 0 0 1>
     }
    }
    renderTransform {angle=0, scale=<1 1>, shear=<-7 0>, translation=<0 0>}
   }
   slot {class=WidgetSwitcherSlot}
  }
  MFA_WarningText {
   widgetClassPath=&CommonTextBlock
   bIsVariable=true
   parent=CommonWidgetSwitcher_TournamentWarnings
   widget=MFA_WarningText
   properties {
    autoWrapText=true
    justification=Center
    lineHeightPercentage=0.9
    text="Two-Factor Authentication (2FA) is required to compete for cash prizes! See Event Details"
    wrapTextAt=366
    colorAndOpacity {colorUseRule=UseColor_Specified, specifiedColor=<rgba 0.0044 0.0091 0.0497 1>}
    font {
     bForceMonospaced=false
     bMaterialIsStencil=false
     fontMaterial=null
     fontObject=&Burbank
     letterSpacing=0
     monospacedWidth=1
     size=22
     skewAmount=0
     typefaceFontName=BurbankRegular-Black
     outlineSettings {
      bApplyOutlineToDropShadows=false
      bMiteredCorners=false
      bSeparateFillAlpha=false
      outlineMaterial=null
      outlineSize=0
      outlineColor=<rgba 0 0 0 1>
     }
    }
    renderTransform {angle=0, scale=<1 1>, shear=<-7 0>, translation=<0 0>}
   }
   slot {class=WidgetSwitcherSlot}
  }
  NoMessage {widgetClassPath=&Overlay, bIsVariable=true, parent=CentralContentSwitcher, widget=NoMessage, slot {class=WidgetSwitcherSlot}}
  Spacer_78 {
   widgetClassPath=&Spacer
   parent=NoMessage
   widget=Spacer_78
   properties {size=<0 100>}
   slot {class=OverlaySlot, properties {horizontalAlignment=HAlign_Fill, verticalAlignment=VAlign_Fill}}
  }
  EarnPinMessage {
   widgetClassPath=&Overlay
   bIsVariable=true
   parent=VerticalBox_0
   widget=EarnPinMessage
   slot {class=VerticalBoxSlot, properties {verticalAlignment=VAlign_Bottom, padding=<ltrb 0 24 0 0>, size {sizeRule=Fill, value=1}}}
  }
  EarnPinContent {
   widgetClassPath=&VerticalBox
   bIsVariable=true
   parent=EarnPinMessage
   widget=EarnPinContent
   slot {class=OverlaySlot, properties {horizontalAlignment=HAlign_Fill, verticalAlignment=VAlign_Fill}}
  }
  CommonTextBlock_EarnAPin {
   widgetClassPath=&CommonTextBlock
   bIsVariable=true
   parent=EarnPinContent
   widget=CommonTextBlock_EarnAPin
   properties {
    justification=Center
    scrollStyle=&TextScrollStyle-Base_C
    text=DIVISION RANK UP IN
    colorAndOpacity {colorUseRule=UseColor_Specified, specifiedColor=<rgba 0 1 0.9823 1>}
    font {
     bForceMonospaced=false
     bMaterialIsStencil=false
     fontMaterial=null
     fontObject=&BurbankSmall
     letterSpacing=0
     monospacedWidth=1
     size=18
     skewAmount=0
     typefaceFontName=Black
     outlineSettings {
      bApplyOutlineToDropShadows=false
      bMiteredCorners=false
      bSeparateFillAlpha=false
      outlineMaterial=null
      outlineSize=0
      outlineColor=<rgba 0 0 0 1>
     }
    }
   }
   slot {class=VerticalBoxSlot, properties {padding=<ltrb 10 0 10 0>}}
  }
  HorizontalBox_4 {widgetClassPath=&HorizontalBox, parent=EarnPinContent, widget=HorizontalBox_4, slot {class=VerticalBoxSlot}}
  Image_Zag {
   widgetClassPath=&Image
   parent=HorizontalBox_4
   widget=Image_Zag
   properties {
    brush {
     bIsDynamicallyLoaded=false
     drawAs=Box
     imageType=NoImage
     mirroring=NoMirror
     resourceName=None
     resourceObject=/FortniteCoreUI/Frontend/Showdown/Materials/M_UI_HZag_Inst
     tiling=NoTile
     imageSize=<128 12>
     margin=<ltrb 0 0 0.4771 0>
     outlineSettings {
      bUseBrushTransparency=false
      roundingType=HalfHeightRadius
      width=0
      color {colorUseRule=UseColor_Specified, specifiedColor=<rgba 0 0 0 0>}
      cornerRadii=<0 0 0 0>
     }
     tintColor {colorUseRule=UseColor_Specified, specifiedColor=<rgba 1 1 1 1>}
     uVRegion {bIsValid=false, max=<0 0>, min=<0 0>}
    }
    renderTransform {angle=0, scale=<-1 1>, shear=<0 0>, translation=<0 0>}
   }
   slot {class=HorizontalBoxSlot, properties {verticalAlignment=VAlign_Center, padding=<ltrb 0 -5 0 0>, size {sizeRule=Fill, value=1}}}
  }
  HypeIcon {
   widgetClassPath=&Image
   bIsVariable=true
   parent=HorizontalBox_4
   widget=HypeIcon
   properties {brush {
     bIsDynamicallyLoaded=false
     drawAs=Image
     imageType=NoImage
     mirroring=NoMirror
     resourceName=None
     resourceObject=&M_UI_LeagueDivision_PlayerRatingIcon
     tiling=NoTile
     imageSize=<40 40>
     margin=<ltrb 0 0 0 0>
     outlineSettings {
      bUseBrushTransparency=false
      roundingType=HalfHeightRadius
      width=0
      color {colorUseRule=UseColor_Specified, specifiedColor=<rgba 0 0 0 0>}
      cornerRadii=<0 0 0 0>
     }
     tintColor {colorUseRule=UseColor_Specified, specifiedColor=<rgba 1 1 1 1>}
     uVRegion {bIsValid=false, max=<0 0>, min=<0 0>}
    }}
   slot {class=HorizontalBoxSlot, properties {horizontalAlignment=HAlign_Left, verticalAlignment=VAlign_Center, padding=<ltrb 16 0 0 0>}}
  }
  Text_CurrentHype {
   widgetClassPath=&CommonTextBlock
   bIsVariable=true
   parent=HorizontalBox_4
   widget=Text_CurrentHype
   properties {
    justification=Center
    text=40
    font {
     bForceMonospaced=false
     bMaterialIsStencil=false
     fontMaterial=null
     fontObject=&BurbankSmall
     letterSpacing=0
     monospacedWidth=1
     size=34
     skewAmount=0
     typefaceFontName=Black
     outlineSettings {
      bApplyOutlineToDropShadows=false
      bMiteredCorners=false
      bSeparateFillAlpha=false
      outlineMaterial=null
      outlineSize=0
      outlineColor=<rgba 0 0 0 1>
     }
    }
   }
   slot {class=HorizontalBoxSlot, properties {horizontalAlignment=HAlign_Center, padding=<ltrb 8 0 16 0>}}
  }
  Image_Zag2 {
   widgetClassPath=&Image
   parent=HorizontalBox_4
   widget=Image_Zag2
   properties {brush {
     bIsDynamicallyLoaded=false
     drawAs=Box
     imageType=NoImage
     mirroring=NoMirror
     resourceName=None
     resourceObject=&M_UI_HZag_Inst
     tiling=NoTile
     imageSize=<128 12>
     margin=<ltrb 0 0 1 0>
     outlineSettings {
      bUseBrushTransparency=false
      roundingType=HalfHeightRadius
      width=0
      color {colorUseRule=UseColor_Specified, specifiedColor=<rgba 0 0 0 0>}
      cornerRadii=<0 0 0 0>
     }
     tintColor {colorUseRule=UseColor_Specified, specifiedColor=<rgba 1 1 1 1>}
     uVRegion {bIsValid=false, max=<0 0>, min=<0 0>}
    }}
   slot {class=HorizontalBoxSlot, properties {verticalAlignment=VAlign_Center, padding=<ltrb 0 -5 0 0>, size {sizeRule=Fill, value=1}}}
  }
  VerticalBox_1 {
   widgetClassPath=&VerticalBox
   parent=VerticalBox_0
   widget=VerticalBox_1
   properties {clipping=ClipToBounds}
   slot {class=VerticalBoxSlot, properties {padding=<ltrb 0 24 0 0>}}
  }
  BorderTimeRemaining {
   widgetClassPath=&Border
   bIsVariable=true
   parent=VerticalBox_1
   widget=BorderTimeRemaining
   properties {
    background {
     bIsDynamicallyLoaded=false
     drawAs=Image
     imageType=NoImage
     mirroring=NoMirror
     resourceName=None
     resourceObject=/FortniteCoreUI/Frontend/Showdown/Materials/M_UI_SkewBox_TournamentPanelTimerBG
     tiling=NoTile
     imageSize=<32 32>
     margin=<ltrb 0 0 0 0>
     outlineSettings {
      bUseBrushTransparency=false
      roundingType=HalfHeightRadius
      width=0
      color {colorUseRule=UseColor_Specified, specifiedColor=<rgba 0 0 0 0>}
      cornerRadii=<0 0 0 0>
     }
     tintColor {colorUseRule=UseColor_Specified, specifiedColor=<rgba 1 1 1 1>}
     uVRegion {bIsValid=false, max=<0 0>, min=<0 0>}
    }
    brushColor=<rgba 0.1022 0 1 1>
    padding=<ltrb 16 16 16 16>
   }
   slot {class=VerticalBoxSlot}
  }
  VB_Content_TimeRemaining {
   widgetClassPath=&VerticalBox
   parent=BorderTimeRemaining
   widget=VB_Content_TimeRemaining
   properties {clipping=ClipToBounds}
   slot {class=BorderSlot, properties {padding=<ltrb 16 16 16 16>}}
  }
  HorizontalBox_0 {
   widgetClassPath=&HorizontalBox
   parent=VB_Content_TimeRemaining
   widget=HorizontalBox_0
   slot {class=VerticalBoxSlot, properties {horizontalAlignment=HAlign_Center}}
  }
  Image_35 {
   widgetClassPath=&Image
   bIsVariable=true
   parent=HorizontalBox_0
   widget=Image_35
   properties {brush {
     bIsDynamicallyLoaded=false
     drawAs=Image
     imageType=NoImage
     mirroring=NoMirror
     resourceName=None
     resourceObject=/Game/UI/Foundation/Textures/Icons/HUD/T-Icon-Clock-128
     tiling=NoTile
     imageSize=<40 40>
     margin=<ltrb 0 0 0 0>
     outlineSettings {
      bUseBrushTransparency=false
      roundingType=HalfHeightRadius
      width=0
      color {colorUseRule=UseColor_Specified, specifiedColor=<rgba 0 0 0 0>}
      cornerRadii=<0 0 0 0>
     }
     tintColor {colorUseRule=UseColor_Specified, specifiedColor=<rgba 1 1 1 1>}
     uVRegion {bIsValid=false, max=<0 0>, min=<0 0>}
    }}
   slot {class=HorizontalBoxSlot, properties {horizontalAlignment=HAlign_Left, verticalAlignment=VAlign_Center}}
  }
  Text_TimeRemaining {
   widgetClassPath=&CommonTextBlock
   bInherited=true
   bIsVariable=true
   parent=HorizontalBox_0
   widget=Text_TimeRemaining
   properties {
    justification=Center
    scrollStyle=&TextScrollStyle-Base_C
    text=1H 2M
    font {
     bForceMonospaced=false
     bMaterialIsStencil=false
     fontMaterial=null
     fontObject=&BurbankSmall
     letterSpacing=0
     monospacedWidth=1
     size=30
     skewAmount=0
     typefaceFontName=Black
     outlineSettings {
      bApplyOutlineToDropShadows=false
      bMiteredCorners=false
      bSeparateFillAlpha=false
      outlineMaterial=null
      outlineSize=0
      outlineColor=<rgba 0 0 0 1>
     }
    }
   }
   slot {class=HorizontalBoxSlot, properties {verticalAlignment=VAlign_Center, padding=<ltrb 8 0 0 0>, size {sizeRule=Fill, value=1}}}
  }
  Text_Remaining {
   widgetClassPath=&CommonTextBlock
   bIsVariable=true
   parent=VB_Content_TimeRemaining
   widget=Text_Remaining
   properties {
    justification=Center
    text=Until Reset
    textTransformPolicy=ToUpper
    font {
     bForceMonospaced=false
     bMaterialIsStencil=false
     fontMaterial=null
     fontObject=&BurbankSmall
     letterSpacing=0
     monospacedWidth=1
     size=26
     skewAmount=0
     typefaceFontName=NotoSans-Regular
     outlineSettings {
      bApplyOutlineToDropShadows=false
      bMiteredCorners=false
      bSeparateFillAlpha=false
      outlineMaterial=null
      outlineSize=0
      outlineColor=<rgba 0 0 0 1>
     }
    }
   }
   slot {class=VerticalBoxSlot, properties {verticalAlignment=VAlign_Center, padding=<ltrb 0 4 0 0>}}
  }
  O_EventDetails {
   widgetClassPath=&Overlay
   parent=VerticalBox_1
   widget=O_EventDetails
   slot {class=VerticalBoxSlot, properties {verticalAlignment=VAlign_Bottom}}
  }
  Image_DetailsButtonBG {
   widgetClassPath=&Image
   bIsVariable=true
   parent=O_EventDetails
   widget=Image_DetailsButtonBG
   properties {
    visibility=SelfHitTestInvisible
    brush {
     bIsDynamicallyLoaded=false
     drawAs=Image
     imageType=NoImage
     mirroring=NoMirror
     resourceName=None
     resourceObject=/Game/UI/Frontend/Showdown/Materials/M_UI_SkewBox_TournamentPanelDetailsBG
     tiling=NoTile
     imageSize=<32 32>
     margin=<ltrb 0 0 0 0>
     outlineSettings {
      bUseBrushTransparency=false
      roundingType=HalfHeightRadius
      width=0
      color {colorUseRule=UseColor_Specified, specifiedColor=<rgba 0 0 0 0>}
      cornerRadii=<0 0 0 0>
     }
     tintColor {colorUseRule=UseColor_Specified, specifiedColor=<rgba 1 1 1 1>}
     uVRegion {bIsValid=false, max=<0 0>, min=<0 0>}
    }
    renderTransformPivot=<0 1>
   }
   slot {class=OverlaySlot, properties {horizontalAlignment=HAlign_Fill, verticalAlignment=VAlign_Fill, padding=<ltrb 0 -5 0 0>}}
  }
  Button_EventDetails {
   widgetClassPath=&IconTextButton
   bInherited=true
   bIsVariable=true
   parent=O_EventDetails
   widget=Button_EventDetails
   properties {
    buttonText=""
    contentAlignment=HAlign_Right
    onButtonBaseClicked=(&self:WidgetTree.Button_EventDetails.HandleOnButtonClicked)
    onButtonBaseDoubleClicked=(&self:WidgetTree.Button_EventDetails.HandleOnButtonDoubleClicked)
    onButtonBaseHovered=(&self:WidgetTree.Button_EventDetails.HandleOnButtonHovered)
    onButtonBaseUnhovered=(&self:WidgetTree.Button_EventDetails.HandleOnButtonUnhovered)
    onSelectedChangedBase=(&self:WidgetTree.Button_EventDetails.HandleOnSelectedChanged)
    style=/FortUILibrary/Foundation/Buttons/ButtonStyles/ButtonStyle-TextOnlyBase_Empty.ButtonStyle-TextOnlyBase_Empty_C
   }
   slot {class=OverlaySlot, properties {horizontalAlignment=HAlign_Fill, verticalAlignment=VAlign_Fill}}
  }
  Text_DetailsButton {
   widgetClassPath=&CommonTextBlock
   bIsVariable=true
   parent=O_EventDetails
   widget=Text_DetailsButton
   properties {
    justification=Right
    text=League Details
    textTransformPolicy=ToUpper
    colorAndOpacity {colorUseRule=UseColor_Specified, specifiedColor=<rgba 0.0116 0.016 0.0395 1>}
    font {
     bForceMonospaced=false
     bMaterialIsStencil=false
     fontMaterial=null
     fontObject=&BurbankSmall
     letterSpacing=0
     monospacedWidth=1
     size=20
     skewAmount=0
     typefaceFontName=Black
     outlineSettings {
      bApplyOutlineToDropShadows=false
      bMiteredCorners=false
      bSeparateFillAlpha=false
      outlineMaterial=null
      outlineSize=0
      outlineColor=<rgba 0 0 0 1>
     }
    }
   }
   slot {class=OverlaySlot, properties {horizontalAlignment=HAlign_Fill, verticalAlignment=VAlign_Center, padding=<ltrb 16 0 48 0>}}
  }
  Text_DebugId {
   widgetClassPath=&TextBlock
   bInherited=true
   parent=GridPanel_0
   widget=Text_DebugId
   properties {
    autoWrapText=true
    justification=Center
    text=EventId: Event EventWindowId: EventWindow
    visibility=SelfHitTestInvisible
    wrapTextAt=200
    wrappingPolicy=AllowPerCharacterWrapping
    font {
     bForceMonospaced=false
     bMaterialIsStencil=false
     fontMaterial=null
     fontObject=&Roboto
     letterSpacing=0
     monospacedWidth=1
     size=20
     skewAmount=0
     typefaceFontName=Bold
     outlineSettings {
      bApplyOutlineToDropShadows=false
      bMiteredCorners=false
      bSeparateFillAlpha=false
      outlineMaterial=null
      outlineSize=2
      outlineColor=<rgba 0 0 0 1>
     }
    }
   }
   slot {class=GridSlot, properties {horizontalAlignment=HAlign_Left, layer=20, verticalAlignment=VAlign_Bottom}}
  }
  Image_PosterFront {widgetClassPath=/Script/EpicCMSUIFramework.EpicCMSImage, bInherited=true, widget=null}
  Image_PosterBack {widgetClassPath=&EpicCMSImage, bInherited=true, widget=null}
  Image_LoadingScreen {widgetClassPath=&EpicCMSImage, bInherited=true, widget=null}
  Image_PlaylistTile {widgetClassPath=&EpicCMSImage, bInherited=true, widget=null}
 ]
}
