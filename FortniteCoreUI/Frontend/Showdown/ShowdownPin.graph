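# Tournament pin widget for Showdown screens, showing the party's score on either a silhouette badge (none or locked) or the unlocked pin.
# Refresh gets event or tournament-best state, switches content, and displays a main or alternative leaderboard score, optionally abbreviated.
# Main score relative to unlock threshold selects one of four silhouette materials, optionally tinted with tournament colors through dynamic materials.
# Animations drive material parameters for reveal, focus and hover; PreConstruct previews dev values at design time.
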WidgetBlueprint /FortniteCoreUI/Frontend/Showdown/ShowdownPin
animations [
 Reveal {
  binding_count=2
  display_label=""
  length_seconds=2.0377
  object=&self:Reveal
  sequence {bindings [
    NoPinIcon {
     id=#1
     possessed_class=CommonLazyImage
     tracks [{
       type=MovieSceneWidgetMaterialTrack
       sections [{
         type=MovieSceneParameterSection
         channels [
          BaseReveal {type=MovieSceneScriptingFloatChannel, keys [[0,0],[25.1773,1]]}
          BracketReveal {type=MovieSceneScriptingFloatChannel, keys [[4.422,0],[29.5993,1]]}
          LaurelReveal {type=MovieSceneScriptingFloatChannel, keys [[9.613,0],[34.7903,1]]}
          WingStarReveal {type=MovieSceneScriptingFloatChannel, keys [[14.9003,0],[40.0777,1]]}
          WingStarBump {type=MovieSceneScriptingFloatChannel, keys [[22.438,1],[24.8437,0.4],[40.753,1]]}
          BracketBump {type=MovieSceneScriptingFloatChannel, keys [[19.175,1],[21.348,0.4],[30.3573,1]]}
          LaurelBump {type=MovieSceneScriptingFloatChannel, keys [[22.3403,1],[24.5133,0.4],[33.3033,1]]}
          BaseBump {type=MovieSceneScriptingFloatChannel, keys [[13.7357,1],[15.9087,0.4],[27.767,1]]}
         ]
        }]
      }]
    }
    VerticalBox_PointValuesProgress {
     id=#2
     possessed_class=VerticalBox
     tracks [{
       type=MovieSceneFloatTrack
       display_name=Render Opacity
       sections [{type=MovieSceneFloatSection, channels [None {type=MovieSceneScriptingFloatChannel, keys [[0,0],[7.5773,0],[17.2337,1]]}]}]
      }]
    }
   ]}
 }
 Focus {
  binding_count=3
  display_label=""
  length_seconds=0.3048
  object=&self:Focus
  sequence {bindings [
    NoPinIcon {
     id=#3
     possessed_class=CommonLazyImage
     tracks [{
       type=MovieSceneWidgetMaterialTrack
       sections [{
         type=MovieSceneParameterSection
         end_frame=6
         start_frame=0
         channels [Focus {type=MovieSceneScriptingFloatChannel, keys [[0,0],[6.0957,1]]}]
        }]
      }]
    }
    Image_PinShadow {
     id=#4
     possessed_class=Image
     tracks [
      {
       type=MovieScene2DTransformTrack
       sections [{
         type=MovieScene2DTransformSection
         end_frame=6
         start_frame=0
         channels [
          Translation.X {type=MovieSceneScriptingFloatChannel, keys [[0,3],[6.0957,8]]}
          Translation.Y {type=MovieSceneScriptingFloatChannel, keys [[0,3],[6.0957,8]]}
          Angle {type=MovieSceneScriptingFloatChannel}
          Scale.X {type=MovieSceneScriptingFloatChannel}
          Scale.Y {type=MovieSceneScriptingFloatChannel}
          Shear.X {type=MovieSceneScriptingFloatChannel}
          Shear.Y {type=MovieSceneScriptingFloatChannel}
         ]
        }]
      }
      {
       type=MovieSceneFloatTrack
       display_name=Render Opacity
       sections [{type=MovieSceneFloatSection, end_frame=6, start_frame=0, channels [None {type=MovieSceneScriptingFloatChannel, keys [[0,1],[6.0957,0.3]]}]}]
      }
     ]
    }
    NoPinShadow {
     id=#5
     possessed_class=CommonLazyImage
     tracks [{
       type=MovieSceneFloatTrack
       display_name=Render Opacity
       sections [{type=MovieSceneFloatSection, end_frame=6, start_frame=0, channels [None {type=MovieSceneScriptingFloatChannel, keys [[0,0],[6.0957,1]]}]}]
      }]
    }
   ]}
 }
 Hover {
  binding_count=1
  display_label=""
  length_seconds=0.3048
  object=&self:Hover
  sequence {bindings [NoPinIcon {
     id=#3
     possessed_class=CommonLazyImage
     tracks [{
       type=MovieSceneWidgetMaterialTrack
       sections [{
         type=MovieSceneParameterSection
         end_frame=6
         start_frame=0
         channels [Hover {type=MovieSceneScriptingFloatChannel, keys [[0,0],[6.0957,1]]}]
        }]
      }]
    }]}
 }
]
graphs [
 ChooseScoreToDisplay {nodes [
   ChooseScoreToDisplay() {is=title K2Node_FunctionEntry}
   n1 = NOT Boolean(A=Get bRepresentsTournamentBestPin() -> (bRepresentsTournamentBestPin:bool))
   n2 = Get bTryUsingAlternativeScore() -> (bTryUsingAlternativeScore:bool)
   > n3 = Branch(Condition=AND Boolean!(A=@n1, B=@n2, C:bool=ShouldUseRankGroupAlternativeScore!(self:Fort Showdown Detail View Object Reference) -> (ReturnValue:bool)) -> (ReturnValue))
   > Return Node!(
    execute
    ScoreToDisplay:int=GetCurrentPartyRankGroupAlternativeLeaderboardPoints!(self:Fort Showdown Detail View Object Reference) -> (ReturnValue:int=0)
   )
   n4 = Branch(
    execute=@n3.else
    Condition=AND Boolean(A=@n1, B=@n2, C=ShouldUseRelevantAlternativeScore!(self:Fort Showdown Detail View Object Reference) -> (ReturnValue:bool))
   )
   > Return Node(ScoreToDisplay=GetCurrentPartyRelevantAlternativeLeaderboardPoints!(self:Fort Showdown Detail View Object Reference) -> (ReturnValue:int=0))
   n5 = GetMainLeaderboardScoreToDisplay!(execute=@n4.else, self:Fort Showdown Pin Object Reference, bUseBestPin:bool=Get bRepresentsTournamentBestPin()) -> (then, ReturnValue:int=0)
   > Return Node(ScoreToDisplay=@n5)
  ]}
 EventGraph {nodes [
   n1 = EventSetTier!() -> (OutputDelegate, then, Score:int) {is=title K2Node_CustomEvent}
   n2 = float / float(
    A="To Float (Integer)"(InInt=@n1.Score)
    B="To Float (Integer)"(InInt=GetTournamentPinUnlockScore!(self:Fort Showdown Detail View Object Reference) -> (ReturnValue:int=0))
   ) {is=title K2Node_CallFunction}
   n3 = "SetText (Text)"(
    self=Get CommonTextBlock_NoPinPts() -> (CommonTextBlock_NoPinPts:Common Text Object Reference)
    InText="To Text (Float)"(Value=@n2, MinimumIntegralDigits=1, MaximumIntegralDigits=324, MinimumFractionalDigits=0, MaximumFractionalDigits=3)
   )
   n4 = "float < float"!(A:double=@n2, B:double=0.05) -> (ReturnValue:bool) {is=title K2Node_CallFunction}
   n5 = Select(Option 0:int=Select(Option 0=3, Option 1=2, Index="float < float"(A=@n2, B=0.7)), Option 1:int=1, Index="float < float"(A=@n2, B=0.4)) -> (ReturnValue:int)
   SetBrushFromMaterial(
    execute=[@n16.then, @n3.then]
    self=[Get NoPinShadow() -> (NoPinShadow:Common Lazy Image Object Reference), Get NoPinIcon() -> (NoPinIcon:Common Lazy Image Object Reference)]
    Material=Select!(
     Option 0:Material Interface Object Reference=/FortniteCoreUI/Frontend/Showdown/Materials/M_UI_Badge_Silhouette_Tier1
     Option 1:Material Interface Object Reference=/FortniteCoreUI/Frontend/Showdown/Materials/M_UI_Badge_Silhouette_Tier2
     Option 2:Material Interface Object Reference=/FortniteCoreUI/Frontend/Showdown/Materials/M_UI_Badge_Silhouette_Tier3
     Option 3:Material Interface Object Reference=/FortniteCoreUI/Frontend/Showdown/Materials/M_UI_Badge_Silhouette_Tier4
     Index:int=Select!(Option 0:int=@n5, Option 1:int=0, Index:bool=@n4) -> (ReturnValue:int)
    ) -> (ReturnValue:Material Interface Object Reference)
   )
   n6 = EventFixPts!() -> (OutputDelegate, then, Points:int) {is=title K2Node_CustomEvent}
   > "SetText (Text)"(
    self=[Get CommonTextBlock_NoPinPts(), Get CommonTextBlock_PinPts() -> (CommonTextBlock_PinPts:Common Text Object Reference)]
    InText=Format Text!(Format="LOCTEXT(\"{0}|plural(one=PT,other=PTS)\")", 0:int=@n6.Points) -> (Result)
   )
   EventColorize() {is=title K2Node_CustomEvent}
   > Branch(Condition=Get colorSilhouette() -> (colorSilhouette:bool))
   > n7 = GetDynamicMaterial(self=Get NoPinIcon())
   n8 = Break Fort Tournament Display Info!(FortTournamentDisplayInfo:"Fort Tournament Display Info Structure (by ref)"=Get TournamentDisplayInfo() -> (TournamentDisplayInfo:Fort Tournament Display Info Structure)) -> (
    TitleLine1:Text
    TitleLine2:Text
    ScheduleInfo:Text
    FlavorDescription:Text
    DetailsDescription:Text
    ShortFormatTitle:Text
    LongFormatTitle:Text
    BackgroundTitle:Text
    PinScoreRequirement:int
    PinEarnedText:Text
    BaseColor:Linear Color Structure
    PrimaryColor:Linear Color Structure
    SecondaryColor:Linear Color Structure
    HighlightColor:Linear Color Structure
    TitleColor:Linear Color Structure
    ShadowColor:Linear Color Structure
    BackgroundLeftColor:Linear Color Structure
    BackgroundRightColor:Linear Color Structure
    BackgroundTextColor:Linear Color Structure
    PosterFadeColor:Linear Color Structure
    AlertText:Text
    SeriesPointLeaderboardName:Text
    AlertType:EFortTournamentAlertType Enum
    RoundDisplayNames:Array of Texts
   ) {is=title K2Node_BreakStruct}
   > SetVectorParameterValue(self=@n7, ParameterName=OutlineColor, Value=@n8.PrimaryColor)
   > SetVectorParameterValue(self=@n7, ParameterName=ShapeColor, Value=@n8.ShadowColor)
   > SetVectorParameterValue(self=@n7, ParameterName=OutlineColorHover, Value=(R=1,G=1,B=1,A=1))
   > SetVectorParameterValue(self=@n7, ParameterName=OutlineColorFocus, Value=(R=1,G=1,B=1,A=1))
   > n9 = GetDynamicMaterial(self=Get PinIcon() -> (PinIcon:Image Object Reference))
   > SetVectorParameterValue(self=@n9, ParameterName=Color1, Value=@n8.PrimaryColor)
   > SetVectorParameterValue(self=@n9, ParameterName=Color2, Value=@n8.SecondaryColor)
   n10 = Break Linear Color!(LinearColor:"Linear Color Structure (by ref)"=@n8.ShadowColor) -> (R:float=~, G:float=~, B:float=~, A:float=~) {is=title K2Node_BreakStruct}
   > SetColorAndOpacity(
    self:Image Object Reference=Get Image_PinShadow() -> (Image_PinShadow:Image Object Reference)
    InColorAndOpacity:Linear Color Structure=MakeColor!(R:float=@n10, G:float=@n10.G, B:float=@n10.B, A:float=0.6) -> (ReturnValue:Linear Color Structure)
   )
   EventRevealAnim() {is=title K2Node_CustomEvent}
   > Anim Transition!(
    Exec:Exec
    In:bool=true
    Animation:Widget Animation Object Reference=Get Reveal() -> (Reveal:Widget Animation Object Reference)
    PlaySpeed:double=1
    Widget:User Widget Object Reference
   ) -> (ExecOut:Exec) {is=title K2Node_MacroInstance}
   n11 = Event PreConstruct()
   > n12 = Branch(Condition=@n11.IsDesignTime)
   n13 = Get DEVShowPin() -> (DEVShowPin:bool)
   > SetActiveWidgetIndex(
    self=Get PinInfoSwitcher() -> (PinInfoSwitcher:Common Widget Switcher Legacy Object Reference)
    Index=Select(Option 0=0, Option 1=1, Index=@n13)
   )
   n14 = Get DEVPinScore() -> (DEVPinScore:int=0)
   > EventSetTier!(execute, self, Score:int=@n14) -> (then) {is=title K2Node_CallFunction}
   > n15 = "SetText (Text)"(
    self=[
     Get PinScoreText() -> (PinScoreText:Common Text Object Reference)
     Get NoPinScoreText() -> (NoPinScoreText:Common Numeric Text Block Object Reference)
    ]
    InText="To Text (Integer)"(Value=@n14)
   )
   SetVisibility(
    execute=[@n15.then, @n12.else]
    self=[
     Get VerticalBox_PointValuesProgress() -> (VerticalBox_PointValuesProgress:Vertical Box Object Reference)
     Get VerticalBoxPointValues() -> (VerticalBoxPointValues:Vertical Box Object Reference)
    ]
    InVisibility=Select(Option 0:ESlateVisibility Enum=HitTestInvisible, Option 1:ESlateVisibility Enum=Collapsed, Index=Get HidePoints() -> (HidePoints:bool)) -> (ReturnValue:ESlateVisibility Enum)
   )
   > SetVisibility(
    self=[Get NoPinShadow(), Get NoPinIcon()]
    InVisibility=Select(Option 0=SelfHitTestInvisible, Option 1=Collapsed, Index=Get hideBackplate() -> (hideBackplate:bool))
   )
   Event Refresh Data() {is=title K2Node_Event}
   > Refresh()
   OR Boolean(
    A=@n13
    B="integer >= integer"!(A:int=Get DEVPinScore() -> (DEVPinScore=0), B:int=GetTournamentPinUnlockScore() -> (ReturnValue=0)) -> (ReturnValue:bool) {is=title K2Node_CallFunction}
   )
   n16 = EventFixPts!(execute=@n1.then, self, Points:int=@n1.Score) -> (then) {is=title K2Node_CallFunction}
  ]}
 Refresh {nodes [
   Refresh() {is=title K2Node_FunctionEntry}
   > Set PinState(PinState:EFortShowdownPinState Enum=Select(
     Option 0:EFortShowdownPinState Enum=GetEventPinState!(self:Fort Showdown Detail View Object Reference) -> (ReturnValue:EFortShowdownPinState Enum=None)
     Option 1:EFortShowdownPinState Enum=GetTournamentBestPinState!(self:Fort Showdown Detail View Object Reference) -> (ReturnValue:EFortShowdownPinState Enum=None)
     Index=Get bRepresentsTournamentBestPin()
    ) -> (ReturnValue:EFortShowdownPinState Enum)) -> (Output_Get:EFortShowdownPinState Enum=None)
   n1 = Get NoPinContent() -> (NoPinContent:Overlay Object Reference)
   > SetActiveWidget(
    self:Widget Switcher Object Reference=Get PinInfoSwitcher()
    Widget=Select!(
     None:Widget Object Reference=@n1
     Locked:Widget Object Reference=@n1
     Unlocked:Widget Object Reference=Get PinContent() -> (PinContent:Overlay Object Reference)
     Index:EFortShowdownPinState Enum=Get PinState() -> (PinState:EFortShowdownPinState Enum=None)
    ) -> (ReturnValue:Widget Object Reference)
   )
   > n2 = ChooseScoreToDisplay!(execute, self) -> (then, ScoreToDisplay:int=0) {is=title K2Node_CallFunction}
   > n3 = Branch(Condition=Get bUseAbbreviatedScore() -> (bUseAbbreviatedScore:bool))
   > n4 = Switch on EFortShowdownPinState!(execute, Selection:EFortShowdownPinState Enum=Get PinState() -> (PinState=None)) -> (None:Exec, Locked:Exec, Unlocked:Exec) {is=title K2Node_SwitchEnum}
   "SetText (Text)"(
    execute=[@n4.Locked, @n4.None]
    self=Get NoPinScoreText()
    InText=GetShorthandQuantityWithLimitedDigits!(Score:int=@n2, MaxDigits:int=3) -> (ReturnValue:Text)
   )
   > n5 = Branch(Condition="integer >= integer"(A=@n2, B=1000) {is=title K2Node_PromotableOperator})
   n6 = Get NoPinPlusSign() -> (NoPinPlusSign:Common Text Object Reference)
   > n7 = SetVisibility(self=@n6, InVisibility=SelfHitTestInvisible)
   n8 = GetMainLeaderboardScoreToDisplay(execute=[@n9.then, @n7.then, @n10.then, @n12.then], bUseBestPin=Get bRepresentsTournamentBestPin()) -> (ReturnValue=0)
   > EventSetTier(Score=@n8)
   > EventFixPts(Points=@n2)
   > EventColorize() {is=title K2Node_CallFunction}
   n9 = SetVisibility(execute=@n5.else, self=@n6, InVisibility=Collapsed)
   n10 = "SetText (Text)"(execute=@n4.Unlocked, self=Get PinScoreText(), InText=GetShorthandQuantityWithLimitedDigits(Score=@n2, MaxDigits=3))
   n11 = Get NoPinScoreText()
   "SetText (Text)"(
    execute=@n3.else
    self=Select(
     None:Text Object Reference=@n11
     Locked:Text Object Reference=@n11
     Unlocked:Text Object Reference=Get PinScoreText()
     Index=Get PinState() -> (PinState=None)
    ) -> (ReturnValue:Text Object Reference)
    InText="To Text (Integer)"(Value=@n2)
   )
   > n12 = InterpolateToValue!(
    execute
    self:Common Numeric Text Block Object Reference=Get NoPinScoreText()
    TargetValue:float="To Float (Integer)"(InInt=@n2)
    MaximumInterpolationDuration:float=0.3
    MinimumChangeRate:float=1
    OutroOffset:float=0
   ) -> (then)
  ]}
]
variables [
 bRepresentsTournamentBestPin {type=boolean}
 colorSilhouette {type=boolean}
 DEVPinScore {type=integer, default=5}
 DEVShowPin {type=boolean}
 hideBackplate {type=boolean}
 HidePoints {type=boolean}
]
widgets {
 info {
  inheritedWidgetCount=6
  parentClass=/Script/FortniteUI.FortShowdownPin
  rootWidgetClass=&CommonWidgetSwitcherLegacy
  widgetCount=23
 }
 widgets [
  PinInfoSwitcher {widgetClassPath=&CommonWidgetSwitcherLegacy, bIsVariable=true, widget=PinInfoSwitcher}
  NoPinContent {
   widgetClassPath=&Overlay
   bIsVariable=true
   parent=PinInfoSwitcher
   widget=NoPinContent
   slot {class=WidgetSwitcherSlot, properties {horizontalAlignment=HAlign_Center, verticalAlignment=VAlign_Center}}
  }
  NoPinShadow {
   widgetClassPath=/Script/CommonUI.CommonLazyImage
   bIsVariable=true
   parent=NoPinContent
   widget=NoPinShadow
   properties {
    renderOpacity=0
    brush {
     bIsDynamicallyLoaded=false
     drawAs=Image
     imageType=FullColor
     mirroring=NoMirror
     resourceName=None
     resourceObject=&M_UI_Badge_Silhouette_Tier2
     tiling=NoTile
     imageSize=<256 256>
     margin=<ltrb 0 0 0 0>
     outlineSettings {
      bUseBrushTransparency=false
      roundingType=HalfHeightRadius
      width=0
      color {colorUseRule=UseColor_Specified, specifiedColor=<rgba 0 0 0 0>}
      cornerRadii=<0 0 0 0>
     }
     tintColor {colorUseRule=UseColor_Specified, specifiedColor=<rgba 1 1 1 1>}
     uVRegion {bIsValid=false, max=<0 0>, min=<0 0>}
    }
    colorAndOpacity=<rgba 0.0001 0.0001 0.0056 0.15>
   }
   slot {class=OverlaySlot, properties {horizontalAlignment=HAlign_Center, verticalAlignment=VAlign_Center}}
  }
  NoPinIcon {
   widgetClassPath=&CommonLazyImage
   bIsVariable=true
   parent=NoPinContent
   widget=NoPinIcon
   properties {brush {
     bIsDynamicallyLoaded=false
     drawAs=Image
     imageType=FullColor
     mirroring=NoMirror
     resourceName=None
     resourceObject=&M_UI_Badge_Silhouette_Tier2
     tiling=NoTile
     imageSize=<256 256>
     margin=<ltrb 0 0 0 0>
     outlineSettings {
      bUseBrushTransparency=false
      roundingType=HalfHeightRadius
      width=0
      color {colorUseRule=UseColor_Specified, specifiedColor=<rgba 0 0 0 0>}
      cornerRadii=<0 0 0 0>
     }
     tintColor {colorUseRule=UseColor_Specified, specifiedColor=<rgba 1 1 1 1>}
     uVRegion {bIsValid=false, max=<0 0>, min=<0 0>}
    }}
   slot {class=OverlaySlot, properties {horizontalAlignment=HAlign_Center, verticalAlignment=VAlign_Center}}
  }
  VerticalBox_PointValuesProgress {
   widgetClassPath=&VerticalBox
   bIsVariable=true
   parent=NoPinContent
   widget=VerticalBox_PointValuesProgress
   properties {renderTransform {angle=0, scale=<1 1>, shear=<0 0>, translation=<0 -24>}}
   slot {class=OverlaySlot, properties {horizontalAlignment=HAlign_Center, verticalAlignment=VAlign_Center, padding=<ltrb 0 8 0 0>}}
  }
  HorizontalBox_0 {
   widgetClassPath=&HorizontalBox
   parent=VerticalBox_PointValuesProgress
   widget=HorizontalBox_0
   slot {class=VerticalBoxSlot, properties {verticalAlignment=VAlign_Center}}
  }
  NoPinScoreText {
   widgetClassPath=/Script/CommonUI.CommonNumericTextBlock
   bIsVariable=true
   parent=HorizontalBox_0
   widget=NoPinScoreText
   properties {
    currentNumericValue=32
    justification=Center
    text=32
    font {
     bForceMonospaced=false
     bMaterialIsStencil=false
     fontMaterial=null
     fontObject=&HeadingNow
     letterSpacing=0
     monospacedWidth=1
     size=51.75
     skewAmount=0
     typefaceFontName=86 Bold
     outlineSettings {
      bApplyOutlineToDropShadows=false
      bMiteredCorners=false
      bSeparateFillAlpha=false
      outlineMaterial=null
      outlineSize=4
      outlineColor=<rgba 0 0.1315 0.401 1>
     }
    }
   }
   slot {class=HorizontalBoxSlot, properties {horizontalAlignment=HAlign_Center, verticalAlignment=VAlign_Center}}
  }
  NoPinPlusSign {
   widgetClassPath=&CommonTextBlock
   bIsVariable=true
   parent=HorizontalBox_0
   widget=NoPinPlusSign
   properties {
    justification=Center
    text=+
    visibility=Collapsed
    font {
     bForceMonospaced=false
     bMaterialIsStencil=false
     fontMaterial=null
     fontObject=&HeadingNow
     letterSpacing=0
     monospacedWidth=1
     size=51.75
     skewAmount=0
     typefaceFontName=86 Bold
     outlineSettings {
      bApplyOutlineToDropShadows=false
      bMiteredCorners=false
      bSeparateFillAlpha=false
      outlineMaterial=null
      outlineSize=4
      outlineColor=<rgba 0 0.1315 0.401 1>
     }
    }
   }
   slot {class=HorizontalBoxSlot, properties {horizontalAlignment=HAlign_Center, verticalAlignment=VAlign_Center, padding=<ltrb -4 0 0 0>}}
  }
  CommonTextBlock_NoPinPts {
   widgetClassPath=&CommonTextBlock
   bIsVariable=true
   parent=VerticalBox_PointValuesProgress
   widget=CommonTextBlock_NoPinPts
   properties {
    text=PTS
    font {
     bForceMonospaced=false
     bMaterialIsStencil=false
     fontMaterial=null
     fontObject=&Burbank
     letterSpacing=0
     monospacedWidth=1
     size=16
     skewAmount=0
     typefaceFontName=BurbankRegular-Black
     outlineSettings {
      bApplyOutlineToDropShadows=false
      bMiteredCorners=false
      bSeparateFillAlpha=false
      outlineMaterial=null
      outlineSize=2
      outlineColor=<rgba 0 0.1315 0.401 1>
     }
    }
   }
   slot {class=VerticalBoxSlot, properties {horizontalAlignment=HAlign_Center, verticalAlignment=VAlign_Top, padding=<ltrb 0 -8 0 0>}}
  }
  PinContent {
   widgetClassPath=&Overlay
   bIsVariable=true
   parent=PinInfoSwitcher
   widget=PinContent
   slot {class=WidgetSwitcherSlot, properties {horizontalAlignment=HAlign_Center, verticalAlignment=VAlign_Center}}
  }
  Image_PinShadow {
   widgetClassPath=&Image
   bIsVariable=true
   parent=PinContent
   widget=Image_PinShadow
   properties {
    brush {
     bIsDynamicallyLoaded=false
     drawAs=Image
     imageType=NoImage
     mirroring=NoMirror
     resourceName=None
     resourceObject=/FortniteCoreUI/Frontend/Showdown/Materials/M_UI_QualificationPin
     tiling=NoTile
     imageSize=<256 256>
     margin=<ltrb 0 0 0 0>
     outlineSettings {
      bUseBrushTransparency=false
      roundingType=HalfHeightRadius
      width=0
      color {colorUseRule=UseColor_Specified, specifiedColor=<rgba 0 0 0 0>}
      cornerRadii=<0 0 0 0>
     }
     tintColor {colorUseRule=UseColor_Specified, specifiedColor=<rgba 0 0 0 0.7>}
     uVRegion {bIsValid=false, max=<0 0>, min=<0 0>}
    }
    renderTransform {angle=0, scale=<1 1>, shear=<0 0>, translation=<3 3>}
   }
   slot {class=OverlaySlot}
  }
  PinIcon {
   widgetClassPath=&Image
   bIsVariable=true
   parent=PinContent
   widget=PinIcon
   properties {brush {
     bIsDynamicallyLoaded=false
     drawAs=Image
     imageType=NoImage
     mirroring=NoMirror
     resourceName=None
     resourceObject=&M_UI_QualificationPin
     tiling=NoTile
     imageSize=<256 256>
     margin=<ltrb 0 0 0 0>
     outlineSettings {
      bUseBrushTransparency=false
      roundingType=HalfHeightRadius
      width=0
      color {colorUseRule=UseColor_Specified, specifiedColor=<rgba 0 0 0 0>}
      cornerRadii=<0 0 0 0>
     }
     tintColor {colorUseRule=UseColor_Specified, specifiedColor=<rgba 1 1 1 1>}
     uVRegion {bIsValid=false, max=<0 0>, min=<0 0>}
    }}
   slot {class=OverlaySlot}
  }
  VerticalBoxPointValues {
   widgetClassPath=&VerticalBox
   bIsVariable=true
   parent=PinContent
   widget=VerticalBoxPointValues
   properties {renderTransform {angle=0, scale=<1 1>, shear=<0 0>, translation=<4 101>}}
   slot {class=OverlaySlot, properties {horizontalAlignment=HAlign_Center, verticalAlignment=VAlign_Center}}
  }
  SizeBox_0 {
   widgetClassPath=&SizeBox
   parent=VerticalBoxPointValues
   widget=SizeBox_0
   properties {bOverride_MaxDesiredWidth=true, maxDesiredWidth=200}
   slot {class=VerticalBoxSlot}
  }
  ScaleBox_0 {widgetClassPath=&ScaleBox, parent=SizeBox_0, widget=ScaleBox_0, slot {class=SizeBoxSlot}}
  PinScoreText {
   widgetClassPath=&CommonTextBlock
   bIsVariable=true
   parent=ScaleBox_0
   widget=PinScoreText
   properties {
    text=32
    colorAndOpacity {colorUseRule=UseColor_Specified, specifiedColor=<rgba 1 0.9473 0.0529 1>}
    font {
     bForceMonospaced=false
     bMaterialIsStencil=false
     fontMaterial=null
     fontObject=&Burbank
     letterSpacing=0
     monospacedWidth=1
     size=52
     skewAmount=0
     typefaceFontName=Burbank-Black
     outlineSettings {
      bApplyOutlineToDropShadows=false
      bMiteredCorners=false
      bSeparateFillAlpha=false
      outlineMaterial=null
      outlineSize=2
      outlineColor=<rgba 1 0.3916 0 1>
     }
    }
    margin=<ltrb 0 0 0 -5>
    shadowColorAndOpacity=<rgba 0 0.0194 0.2874 0>
    shadowOffset=<8 8>
   }
   slot {class=ScaleBoxSlot}
  }
  CommonTextBlock_PinPts {
   widgetClassPath=&CommonTextBlock
   bIsVariable=true
   parent=VerticalBoxPointValues
   widget=CommonTextBlock_PinPts
   properties {
    justification=Center
    text=PTS
    visibility=Collapsed
    colorAndOpacity {colorUseRule=UseColor_Specified, specifiedColor=<rgba 1 0.9473 0.0529 1>}
    font {
     bForceMonospaced=false
     bMaterialIsStencil=false
     fontMaterial=null
     fontObject=&Burbank
     letterSpacing=0
     monospacedWidth=1
     size=16
     skewAmount=0
     typefaceFontName=BurbankRegular-Black
     outlineSettings {
      bApplyOutlineToDropShadows=false
      bMiteredCorners=false
      bSeparateFillAlpha=false
      outlineMaterial=null
      outlineSize=2
      outlineColor=<rgba 1 0.3916 0 1>
     }
    }
    shadowColorAndOpacity=<rgba 0 0.0194 0.2874 1>
   }
   slot {class=VerticalBoxSlot, properties {horizontalAlignment=HAlign_Center, verticalAlignment=VAlign_Top, padding=<ltrb 0 -30 0 -30>}}
  }
  Text_TimeRemaining {widgetClassPath=&CommonTextBlock, bInherited=true, widget=null}
  Text_DebugId {widgetClassPath=&TextBlock, bInherited=true, widget=null, properties=~}
  Image_PosterFront {widgetClassPath=/Script/EpicCMSUIFramework.EpicCMSImage, bInherited=true, widget=null}
  Image_PosterBack {widgetClassPath=&EpicCMSImage, bInherited=true, widget=null}
  Image_LoadingScreen {widgetClassPath=&EpicCMSImage, bInherited=true, widget=null}
  Image_PlaylistTile {widgetClassPath=&EpicCMSImage, bInherited=true, widget=null}
 ]
}
