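# Front face of a Showdown tournament poster, showing CMS artwork with title, schedule, lock and alert overlays.
# Refresh Data fills and colors text from display info, picks standard or World Cup qualifier layout, shows the lock by eligibility and alerts outside live states.
# Reset clears these and toggles scale-box stretch off and back.
# Construct optionally hides it until the image loads, which also triggers a debug print; the pin setup chain is unwired.
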WidgetBlueprint /FortniteCoreUI/Frontend/Showdown/ShowdownTournamentPosterFront
graphs [
 EventGraph {nodes [
   "On Image Loading Complete (Image_PosterFront)"() {is=title K2Node_ComponentBoundEvent}
   > SetVisibility(InVisibility=HitTestInvisible)
   n1 = Get Brush!(self:Image Object Reference=Get Image_PosterFront() -> (Image_PosterFront:Epic CMSImage Object Reference)) -> (Brush:Slate Brush Structure)
   > PrintString!(
    execute
    InString=GetDisplayName(Object=Break Slate Brush!(SlateBrush:"Slate Brush Structure (by ref)"=@n1) -> (
      TintColor:Slate Color Structure
      DrawAs:ESlateBrushDrawType Enum
      Tiling:ESlateBrushTileType Enum
      Mirroring:ESlateBrushMirrorType Enum
      ImageSize:Vector 2D Structure
      Margin:Margin Structure
      ResourceObject:Object Reference
      OutlineSettings:Slate Brush Outline Settings Structure
     ) {is=title K2Node_BreakStruct}.ResourceObject)
    bPrintToScreen
    bPrintToLog
    TextColor
    Duration
    Key
   ) -> (then)
   Event Construct()
   > SetVisibility(InVisibility=Select(Option 0=Visible, Option 1=Hidden, Index=Get HiddenWhileLoading() -> (HiddenWhileLoading:bool)))
   Event BP_ResetDetailView() {is=title K2Node_Event}
   > "SetText (Text)"(self=[
     Get TournamentTitle1() -> (TournamentTitle1:Common Text Object Reference)
     Get TournamentTitle1WorldCup() -> (TournamentTitle1WorldCup:Common Text Object Reference)
     Get TournamentTitle2WorldCup() -> (TournamentTitle2WorldCup:Common Text Object Reference)
     Get Text_EventAlert() -> (Text_EventAlert:Common Text Object Reference)
     Get TournamentDateText() -> (TournamentDateText:Common Text Object Reference)
     Get TournamentDateTextWorldCup() -> (TournamentDateTextWorldCup:Common Text Object Reference)
     Get TournamentTitle2() -> (TournamentTitle2:Common Text Object Reference)
    ])
   > SetVisibility(
    self=[
     Get OverlayEventLocked() -> (OverlayEventLocked:Overlay Object Reference)
     Get OverlayEventLockedWorldCup() -> (OverlayEventLockedWorldCup:Overlay Object Reference)
     Get Border_Alert() -> (Border_Alert:Border Object Reference)
    ]
    InVisibility=Hidden
   )
   > SetColorAndOpacity!(
    execute
    self=[Get TournamentTitle1(), Get TournamentTitle2(), Get TournamentTitle1WorldCup(), Get TournamentTitle2WorldCup()]
    InColorAndOpacity_SpecifiedColor:Linear Color Structure=(R=1,G=1,B=1,A=1)
    InColorAndOpacity_ColorUseRule:ESlateColorStylingMode Enum
   ) -> (then)
   > SetColorAndOpacity(self=[Get TournamentDateTextWorldCup(), Get TournamentDateText()], InColorAndOpacity_SpecifiedColor=(R=1,G=1,B=1,A=1))
   > SetActiveWidgetIndex(self=Get Switcher_PosterTextLayout() -> (Switcher_PosterTextLayout:Common Widget Switcher Legacy Object Reference), Index=0)
   > SetAutoWrapText!(execute, self:Text Object Reference=[Get TournamentDateTextWorldCup(), Get TournamentDateText()], InAutoTextWrap:bool=false) -> (then)
   > Set TournamentScaleboxStretch(TournamentScaleboxStretch:EStretch Enum=Get Stretch!(self:Scale Box Object Reference=Get ScaleBox_0() -> (ScaleBox_0:Scale Box Object Reference)) -> (Stretch:EStretch Enum=None)) -> (Output_Get:EStretch Enum=None)
   > SetStretch!(
    execute
    self:Scale Box Object Reference=[Get ScaleBox_1() -> (ScaleBox_1:Scale Box Object Reference), Get ScaleBox_0()]
    InStretch:EStretch Enum=None
   ) -> (then)
   > SetStretch(self=[Get ScaleBox_0(), Get ScaleBox_1()], InStretch=Get TournamentScaleboxStretch() -> (TournamentScaleboxStretch:EStretch Enum=None))
   Event Refresh Data() {is=title K2Node_Event}
   > n2 = Sequence()
   SetColorAndOpacity!(
    execute=@n2.then_1
    self=[Get TournamentTitle2(), Get TournamentTitle1(), Get TournamentTitle1WorldCup(), Get TournamentTitle2WorldCup()]
    InColorAndOpacity:Slate Color Structure=Get TitleColor() -> (TitleColor:Slate Color Structure)
   ) -> (then)
   > SetColorAndOpacity(self=[Get TournamentDateText(), Get TournamentDateTextWorldCup()], InColorAndOpacity=Get SubColor() -> (SubColor:Slate Color Structure))
   > SetActiveWidget(
    self:Widget Switcher Object Reference=Get Switcher_PosterTextLayout()
    Widget=Select(
     Option 0:Widget Object Reference=Get StandardLayout() -> (StandardLayout:Vertical Box Object Reference)
     Option 1:Widget Object Reference=Get WorldCupLayout() -> (WorldCupLayout:Vertical Box Object Reference)
     Index=IsWorldCupQualifierEvent!(self:Fort Showdown Detail View Object Reference) -> (ReturnValue:bool)
    ) -> (ReturnValue:Widget Object Reference)
   )
   n3 = Sequence(execute=@n2.then_0)
   n4 = Break Fort Tournament Display Info!(FortTournamentDisplayInfo:"Fort Tournament Display Info Structure (by ref)"=Get TournamentDisplayInfo() -> (TournamentDisplayInfo:Fort Tournament Display Info Structure)) -> (
    TitleLine1:Text
    TitleLine2:Text
    ScheduleInfo:Text
    FlavorDescription:Text
    DetailsDescription:Text
    ShortFormatTitle:Text
    LongFormatTitle:Text
    BackgroundTitle:Text
    PinScoreRequirement:int
    PinEarnedText:Text
    BaseColor:Linear Color Structure
    PrimaryColor:Linear Color Structure
    SecondaryColor:Linear Color Structure
    HighlightColor:Linear Color Structure
    TitleColor:Linear Color Structure
    ShadowColor:Linear Color Structure
    BackgroundLeftColor:Linear Color Structure
    BackgroundRightColor:Linear Color Structure
    BackgroundTextColor:Linear Color Structure
    PosterFadeColor:Linear Color Structure
    AlertText:Text
    SeriesPointLeaderboardName:Text
    AlertType:EFortTournamentAlertType Enum
    RoundDisplayNames:Array of Texts
   ) {is=title K2Node_BreakStruct}
   > "SetText (Text)"(self=[Get TournamentTitle1(), Get TournamentTitle1WorldCup()], InText=TextToUpper!(InText:"Text (by ref)"=@n4) -> (ReturnValue:Text))
   > "SetText (Text)"(self=[Get TournamentTitle2(), Get TournamentTitle2WorldCup()], InText=TextToUpper(InText=@n4.TitleLine2))
   > "SetText (Text)"(self=[Get TournamentDateText(), Get TournamentDateTextWorldCup()], InText=TextToUpper(InText=@n4.ScheduleInfo))
   > Set TitleColor(TitleColor:Slate Color Structure=Make Slate Color(SpecifiedColor=@n4.TitleColor)) -> (Output_Get:Slate Color Structure)
   > Set SubColor(SubColor:Slate Color Structure=Make Slate Color(SpecifiedColor=@n4.BaseColor)) -> (Output_Get:Slate Color Structure)
   > SetVisibility(
    self=[Get OverlayEventLocked(), Get OverlayEventLockedWorldCup()]
    InVisibility=Select!(
     Unknown:ESlateVisibility Enum=Hidden
     Public:ESlateVisibility Enum=Hidden
     Private:ESlateVisibility Enum=Hidden
     Locked:ESlateVisibility Enum=SelfHitTestInvisible
     Index:EFortEventWindowEligibility Enum=GetTournamentEligibility!(self:Fort Showdown Detail View Object Reference) -> (ReturnValue:EFortEventWindowEligibility Enum=Unknown)
    ) -> (ReturnValue:ESlateVisibility Enum)
   )
   SetVisibility(execute=@n3.then_1, self=Get Border_Alert(), InVisibility=Collapsed)
   > n5 = Branch(Condition=TextIsEmpty(InText=@n4.AlertText))
   n6 = Switch on EFortShowdownEventState!(
    execute=@n5.else
    Selection:EFortShowdownEventState Enum=GetTournamentTimeState!(self:Fort Showdown Detail View Object Reference) -> (ReturnValue:EFortShowdownEventState Enum=Unknown)
   ) -> (
    Unknown:Exec
    FutureTBD:Exec
    FutureScheduled:Exec
    FutureNext:Exec
    Live:Exec
    LiveParticipating:Exec
    LiveNotParticipating:Exec
    Completed:Exec
    CompletedParticipated:Exec
    CompletedNotPartipated:Exec
    Cancelled:Exec
   ) {is=title K2Node_SwitchEnum}
   Switch on EFortTournamentAlertType!(
    execute=[@n6.Cancelled, @n6.Completed, @n6.CompletedNotPartipated, @n6.CompletedParticipated, @n6.FutureNext, @n6.FutureScheduled, @n6.FutureTBD]
    Selection:EFortTournamentAlertType Enum=@n4.AlertType
   ) -> (Warning:Exec, Info:Exec) {is=title K2Node_SwitchEnum}
   > "SetText (Text)"(self=Get Text_EventAlert(), InText=TextToUpper(InText=@n4.AlertText))
   > SetVisibility(self=Get Border_Alert(), InVisibility=HitTestInvisible)
   n7 = Get ShowdownPin() -> (ShowdownPin:Showdown Pin Object Reference)
   Set bRepresentsTournamentBestPin!(execute, bRepresentsTournamentBestPin:bool=true, self:Showdown Pin Object Reference=@n7) -> (then, Output_Get:bool)
   > SetTournament!(
    execute
    self:Fort Showdown Detail View Object Reference=@n7
    TournamentId:String=GetTournamentId!(self:Fort Showdown Detail View Object Reference) -> (ReturnValue:String)
    EventId:String
    bForceRefreshUI:bool=true
   ) -> (then)
   n8 = "integer >= integer"!(A:int=GetTournamentBestScore!(self:Fort Showdown Detail View Object Reference) -> (ReturnValue:int=0), B:int=0) -> (ReturnValue:bool) {is=title K2Node_CallFunction}
   n9 = NOT Boolean(A=Get HidePin() -> (HidePin:bool))
   > SetVisibility(
    self=Get ShowdownPin()
    InVisibility=Select!(Option 0:ESlateVisibility Enum=Collapsed, Option 1:ESlateVisibility Enum=HitTestInvisible, Index:bool=AND Boolean(A=@n8, B=@n9)) -> (ReturnValue)
   )
   Event PreConstruct()
  ]}
 GetTournamentDateText {nodes [
   GetTournamentDateText() {is=title K2Node_FunctionEntry}
   n1 = BreakDateTime!(InDateTime:Date Time Structure=GetTournamentEndTime!(self:Fort Showdown Detail View Object Reference) -> (ReturnValue:Date Time Structure)) -> (Year:int=0, Month:int=0, Day:int=0, Hour:int=0, Minute:int=0, Second:int=0, Millisecond:int=0)
   n2 = "Equal (Integer)"(
    A=BreakDateTime(InDateTime=GetTournamentBeginTime!(self:Fort Showdown Detail View Object Reference) -> (ReturnValue:Date Time Structure)) -> (Year=0, Month=0, Day=0, Hour=0, Minute=0, Second=0, Millisecond=0).Month
    B=BreakDateTime(InDateTime=GetTournamentEndTime()) -> (Year=0, Month=0, Day=0, Hour=0, Minute=0, Second=0, Millisecond=0).Month
   ) {is=title K2Node_CallFunction}
   n3 = BreakDateTime(InDateTime=GetTournamentBeginTime()) -> (Year=0, Month=0, Day=0, Hour=0, Minute=0, Second=0, Millisecond=0)
   n4 = "Equal (Integer)"(A=@n3, B=@n1)
   n5 = "To Text (Integer)"(Value=@n3, bUseGrouping=false)
   n6 = Select!(
    Option 0:Text
    Option 1:Text="LOCTEXT(\"January\")"
    Option 2:Text="LOCTEXT(\"February\")"
    Option 3:Text="LOCTEXT(\"March\")"
    Option 4:Text="LOCTEXT(\"April\")"
    Option 5:Text="LOCTEXT(\"May\")"
    Option 6:Text="LOCTEXT(\"June\")"
    Option 7:Text="LOCTEXT(\"July\")"
    Option 8:Text="LOCTEXT(\"August\")"
    Option 9:Text="LOCTEXT(\"September\")"
    Option 10:Text="LOCTEXT(\"October\")"
    Option 11:Text="LOCTEXT(\"November\")"
    Option 12:Text="LOCTEXT(\"December\")"
    Index:int=BreakDateTime(InDateTime=GetTournamentBeginTime()) -> (Year=0, Month=0, Day=0, Hour=0, Minute=0, Second=0, Millisecond=0).Month
   ) -> (ReturnValue:Text)
   n7 = BreakDateTime(InDateTime=GetTournamentEndTime()) -> (Year=0, Month=0, Day=0, Hour=0, Minute=0, Second=0, Millisecond=0)
   > Return Node!(
    execute
    TournamentDate:Text=Format Text!(
     Format="LOCTEXT(\"{StartMonth} {StartYear}{Separator}{EndMonth} {EndYear}\")"
     StartMonth:Text=Select!(Option 0=@n6, Option 1, Index:bool=@n2) -> (ReturnValue)
     StartYear:Text=Select(Option 0=@n5, Index=@n4)
     Separator:Text=Select(Option 0="LOCTEXT(\" - \")", Index=@n2)
     EndMonth:Text=Select!(
      Option 0
      Option 1="LOCTEXT(\"January\")"
      Option 2:Text="LOCTEXT(\"February\")"
      Option 3:Text="LOCTEXT(\"March\")"
      Option 4:Text="LOCTEXT(\"April\")"
      Option 5:Text="LOCTEXT(\"May\")"
      Option 6:Text="LOCTEXT(\"June\")"
      Option 7:Text="LOCTEXT(\"July\")"
      Option 8:Text="LOCTEXT(\"August\")"
      Option 9:Text="LOCTEXT(\"September\")"
      Option 10:Text="LOCTEXT(\"October\")"
      Option 11:Text="LOCTEXT(\"November\")"
      Option 12:Text="LOCTEXT(\"December\")"
      Index:int=@n7.Month
     ) -> (ReturnValue)
     EndYear:Text="To Text (Integer)"(Value=@n1, bUseGrouping=false)
    ) -> (Result)
   )
  ]}
]
variables [
 HiddenWhileLoading {type=boolean}
 HidePin {type=boolean}
 SubColor {type=SlateColor, default {colorUseRule=UseColor_Specified, specifiedColor=<rgba 1 0 1 1>}}
 TitleColor {type=SlateColor, default {colorUseRule=UseColor_Specified, specifiedColor=<rgba 0.9911 0.9131 0.0409 1>}}
 TournamentScaleboxStretch {type=EStretch, default=ScaleToFit}
]
widgets {
 info {inheritedWidgetCount=6, parentClass=/Script/FortniteUI.FortShowdownDetailView, rootWidgetClass=&SizeBox, widgetCount=40}
 widgets [
  SizeBox_0 {
   widgetClassPath=&SizeBox
   widget=SizeBox_0
   properties {bOverride_HeightOverride=true, bOverride_WidthOverride=true, heightOverride=1080, widthOverride=750}
  }
  CommonBorder_0 {
   widgetClassPath=&CommonBorder
   parent=SizeBox_0
   widget=CommonBorder_0
   properties {padding=<ltrb 0 0 0 0>}
   slot {class=SizeBoxSlot}
  }
  Overlay_0 {widgetClassPath=&Overlay, parent=CommonBorder_0, widget=Overlay_0, slot {class=BorderSlot, properties {padding=<ltrb 0 0 0 0>}}}
  Image_PosterShadow {
   widgetClassPath=&Image
   bIsVariable=true
   parent=Overlay_0
   widget=Image_PosterShadow
   properties {colorAndOpacity=<rgba 0.0075 0.0091 0.0242 0.3>, renderTransform {angle=0, scale=<1 1>, shear=<0 0>, translation=<15 15>}}
   slot {class=OverlaySlot, properties {horizontalAlignment=HAlign_Fill, verticalAlignment=VAlign_Fill}}
  }
  Image_PosterFront {
   widgetClassPath=/Script/EpicCMSUIFramework.EpicCMSImage
   bInherited=true
   bIsVariable=true
   parent=Overlay_0
   widget=Image_PosterFront
   properties {loadingBackgroundBrush {
     bIsDynamicallyLoaded=false
     drawAs=Image
     imageType=NoImage
     mirroring=NoMirror
     resourceName=None
     resourceObject=null
     tiling=NoTile
     imageSize=<32 32>
     margin=<ltrb 0 0 0 0>
     outlineSettings {
      bUseBrushTransparency=false
      roundingType=HalfHeightRadius
      width=0
      color {colorUseRule=UseColor_Specified, specifiedColor=<rgba 0 0 0 0>}
      cornerRadii=<0 0 0 0>
     }
     tintColor {colorUseRule=UseColor_Specified, specifiedColor=<rgba 0 0.0182 0.144 1>}
     uVRegion {bIsValid=false, max=<0 0>, min=<0 0>}
    }}
   slot {class=OverlaySlot, properties {horizontalAlignment=HAlign_Fill, verticalAlignment=VAlign_Fill}}
  }
  Switcher_PosterTextLayout {
   widgetClassPath=&CommonWidgetSwitcherLegacy
   bIsVariable=true
   parent=Overlay_0
   widget=Switcher_PosterTextLayout
   slot {class=OverlaySlot, properties {horizontalAlignment=HAlign_Fill, verticalAlignment=VAlign_Bottom}}
  }
  StandardLayout {
   widgetClassPath=&VerticalBox
   bIsVariable=true
   parent=Switcher_PosterTextLayout
   widget=StandardLayout
   slot {class=WidgetSwitcherSlot, properties {padding=<ltrb 0 0 0 50>}}
  }
  OverlayEventLocked {
   widgetClassPath=&Overlay
   bIsVariable=true
   parent=StandardLayout
   widget=OverlayEventLocked
   properties {visibility=Hidden}
   slot {class=VerticalBoxSlot, properties {horizontalAlignment=HAlign_Center, verticalAlignment=VAlign_Bottom, size {sizeRule=Fill, value=1}}}
  }
  Image_LockBG {
   widgetClassPath=&Image
   parent=OverlayEventLocked
   widget=Image_LockBG
   properties {colorAndOpacity=<rgba 0.0052 0.0052 0.0056 0.8>, renderTransform {angle=45, scale=<1 1>, shear=<0 0>, translation=<0 0>}}
   slot {class=OverlaySlot, properties {horizontalAlignment=HAlign_Fill, verticalAlignment=VAlign_Fill}}
  }
  Image_LockIcon {
   widgetClassPath=&Image
   bIsVariable=true
   parent=OverlayEventLocked
   widget=Image_LockIcon
   properties {
    brush {
     bIsDynamicallyLoaded=false
     drawAs=Image
     imageType=NoImage
     mirroring=NoMirror
     resourceName=None
     resourceObject=/Game/UI/SocialPanel/Art/Materials/M_UI_Icon_Lock_Inst
     tiling=NoTile
     imageSize=<256 256>
     margin=<ltrb 0 0 0 0>
     outlineSettings {
      bUseBrushTransparency=false
      roundingType=HalfHeightRadius
      width=0
      color {colorUseRule=UseColor_Specified, specifiedColor=<rgba 0 0 0 0>}
      cornerRadii=<0 0 0 0>
     }
     tintColor {colorUseRule=UseColor_Specified, specifiedColor=<rgba 1 1 1 1>}
     uVRegion {bIsValid=false, max=<0 0>, min=<0 0>}
    }
    renderTransform {angle=0, scale=<1 1>, shear=<0 0>, translation=<0 -8>}
   }
   slot {class=OverlaySlot}
  }
  InvalidationBox_318 {
   widgetClassPath=/Script/UMG.InvalidationBox
   bIsVariable=true
   parent=StandardLayout
   widget=InvalidationBox_318
   slot {class=VerticalBoxSlot, properties {verticalAlignment=VAlign_Bottom, size {sizeRule=Fill, value=1}}}
  }
  VerticalBox_1 {widgetClassPath=&VerticalBox, parent=InvalidationBox_318, widget=VerticalBox_1, slot {class=PanelSlot}}
  ScaleBox_0 {
   widgetClassPath=&ScaleBox
   bIsVariable=true
   parent=VerticalBox_1
   widget=ScaleBox_0
   properties {stretchDirection=DownOnly}
   slot {class=VerticalBoxSlot}
  }
  TournamentTitle1 {
   widgetClassPath=&CommonTextBlock
   bIsVariable=true
   parent=ScaleBox_0
   widget=TournamentTitle1
   properties {
    justification=Center
    lineHeightPercentage=0.75
    text=TOURNAMENT
    font {
     bForceMonospaced=false
     bMaterialIsStencil=false
     fontMaterial=null
     fontObject=&Burbank
     letterSpacing=0
     monospacedWidth=1
     size=100
     skewAmount=0
     typefaceFontName=BurbankRegular-Black
     outlineSettings {
      bApplyOutlineToDropShadows=false
      bMiteredCorners=false
      bSeparateFillAlpha=false
      outlineMaterial=null
      outlineSize=0
      outlineColor=<rgba 0 0 0 1>
     }
    }
    renderTransform {angle=0, scale=<0.9 1>, shear=<-7 0>, translation=<0 0>}
   }
   slot {class=ScaleBoxSlot}
  }
  ScaleBox_1 {
   widgetClassPath=&ScaleBox
   bIsVariable=true
   parent=VerticalBox_1
   widget=ScaleBox_1
   properties {stretchDirection=DownOnly}
   slot {class=VerticalBoxSlot}
  }
  TournamentTitle2 {
   widgetClassPath=&CommonTextBlock
   bIsVariable=true
   parent=ScaleBox_1
   widget=TournamentTitle2
   properties {
    justification=Center
    lineHeightPercentage=0.75
    text=TITLE
    font {
     bForceMonospaced=false
     bMaterialIsStencil=false
     fontMaterial=null
     fontObject=&Burbank
     letterSpacing=0
     monospacedWidth=1
     size=100
     skewAmount=0
     typefaceFontName=BurbankRegular-Black
     outlineSettings {
      bApplyOutlineToDropShadows=false
      bMiteredCorners=false
      bSeparateFillAlpha=false
      outlineMaterial=null
      outlineSize=0
      outlineColor=<rgba 0 0 0 1>
     }
    }
    renderTransform {angle=0, scale=<0.9 1>, shear=<-7 0>, translation=<0 0>}
   }
   slot {class=ScaleBoxSlot}
  }
  InvalidationBox_551 {
   widgetClassPath=&InvalidationBox
   bIsVariable=true
   parent=StandardLayout
   widget=InvalidationBox_551
   slot {class=VerticalBoxSlot, properties {horizontalAlignment=HAlign_Center, padding=<ltrb 0 50 0 50>}}
  }
  TournamentDateText {
   widgetClassPath=&CommonTextBlock
   bIsVariable=true
   parent=InvalidationBox_551
   widget=TournamentDateText
   properties {
    autoWrapText=true
    justification=Center
    text="{Start Date}"
    font {
     bForceMonospaced=false
     bMaterialIsStencil=false
     fontMaterial=null
     fontObject=&Burbank
     letterSpacing=0
     monospacedWidth=1
     size=32
     skewAmount=0
     typefaceFontName=BurbankRegular-Black
     outlineSettings {
      bApplyOutlineToDropShadows=false
      bMiteredCorners=false
      bSeparateFillAlpha=false
      outlineMaterial=null
      outlineSize=0
      outlineColor=<rgba 0 0 0 1>
     }
    }
   }
   slot {class=PanelSlot}
  }
  WorldCupLayout {
   widgetClassPath=&VerticalBox
   bIsVariable=true
   parent=Switcher_PosterTextLayout
   widget=WorldCupLayout
   slot {class=WidgetSwitcherSlot}
  }
  OverlayEventLockedWorldCup {
   widgetClassPath=&Overlay
   bIsVariable=true
   parent=WorldCupLayout
   widget=OverlayEventLockedWorldCup
   properties {visibility=Hidden}
   slot {class=VerticalBoxSlot, properties {horizontalAlignment=HAlign_Center, verticalAlignment=VAlign_Bottom, size {sizeRule=Fill, value=1}}}
  }
  Image_0 {
   widgetClassPath=&Image
   parent=OverlayEventLockedWorldCup
   widget=Image_0
   properties {colorAndOpacity=<rgba 0.0052 0.0052 0.0056 0.8>, renderTransform {angle=45, scale=<1 1>, shear=<0 0>, translation=<0 0>}}
   slot {class=OverlaySlot, properties {horizontalAlignment=HAlign_Fill, verticalAlignment=VAlign_Fill}}
  }
  Image_LockIconWorldCup {
   widgetClassPath=&Image
   bIsVariable=true
   parent=OverlayEventLockedWorldCup
   widget=Image_LockIconWorldCup
   properties {
    brush {
     bIsDynamicallyLoaded=false
     drawAs=Image
     imageType=NoImage
     mirroring=NoMirror
     resourceName=None
     resourceObject=&M_UI_Icon_Lock_Inst
     tiling=NoTile
     imageSize=<256 256>
     margin=<ltrb 0 0 0 0>
     outlineSettings {
      bUseBrushTransparency=false
      roundingType=HalfHeightRadius
      width=0
      color {colorUseRule=UseColor_Specified, specifiedColor=<rgba 0 0 0 0>}
      cornerRadii=<0 0 0 0>
     }
     tintColor {colorUseRule=UseColor_Specified, specifiedColor=<rgba 1 1 1 1>}
     uVRegion {bIsValid=false, max=<0 0>, min=<0 0>}
    }
    renderTransform {angle=0, scale=<1 1>, shear=<0 0>, translation=<0 -8>}
   }
   slot {class=OverlaySlot}
  }
  InvalidationBox_431 {
   widgetClassPath=&InvalidationBox
   bIsVariable=true
   parent=WorldCupLayout
   widget=InvalidationBox_431
   slot {class=VerticalBoxSlot, properties {verticalAlignment=VAlign_Bottom, padding=<ltrb 0 100 0 0>, size {sizeRule=Fill, value=1}}}
  }
  VerticalBox_2 {widgetClassPath=&VerticalBox, parent=InvalidationBox_431, widget=VerticalBox_2, slot {class=PanelSlot}}
  ScaleBox_2 {
   widgetClassPath=&ScaleBox
   parent=VerticalBox_2
   widget=ScaleBox_2
   properties {stretchDirection=DownOnly}
   slot {class=VerticalBoxSlot}
  }
  TournamentTitle1WorldCup {
   widgetClassPath=&CommonTextBlock
   bIsVariable=true
   parent=ScaleBox_2
   widget=TournamentTitle1WorldCup
   properties {
    justification=Center
    lineHeightPercentage=0.75
    text=TOURNAMENT
    font {
     bForceMonospaced=false
     bMaterialIsStencil=false
     fontMaterial=null
     fontObject=&Burbank
     letterSpacing=0
     monospacedWidth=1
     size=100
     skewAmount=0
     typefaceFontName=BurbankRegular-Black
     outlineSettings {
      bApplyOutlineToDropShadows=false
      bMiteredCorners=false
      bSeparateFillAlpha=false
      outlineMaterial=null
      outlineSize=0
      outlineColor=<rgba 0 0 0 1>
     }
    }
    renderTransform {angle=0, scale=<0.9 1>, shear=<-7 0>, translation=<0 0>}
   }
   slot {class=ScaleBoxSlot}
  }
  ScaleBox_3 {
   widgetClassPath=&ScaleBox
   parent=VerticalBox_2
   widget=ScaleBox_3
   properties {stretchDirection=DownOnly}
   slot {class=VerticalBoxSlot}
  }
  TournamentTitle2WorldCup {
   widgetClassPath=&CommonTextBlock
   bIsVariable=true
   parent=ScaleBox_3
   widget=TournamentTitle2WorldCup
   properties {
    justification=Center
    lineHeightPercentage=0.75
    text=TITLE
    font {
     bForceMonospaced=false
     bMaterialIsStencil=false
     fontMaterial=null
     fontObject=&Burbank
     letterSpacing=0
     monospacedWidth=1
     size=100
     skewAmount=0
     typefaceFontName=BurbankRegular-Black
     outlineSettings {
      bApplyOutlineToDropShadows=false
      bMiteredCorners=false
      bSeparateFillAlpha=false
      outlineMaterial=null
      outlineSize=0
      outlineColor=<rgba 0 0 0 1>
     }
    }
    renderTransform {angle=0, scale=<0.9 1>, shear=<-7 0>, translation=<0 0>}
   }
   slot {class=ScaleBoxSlot}
  }
  InvalidationBox_676 {
   widgetClassPath=&InvalidationBox
   bIsVariable=true
   parent=WorldCupLayout
   widget=InvalidationBox_676
   slot {class=VerticalBoxSlot, properties {horizontalAlignment=HAlign_Center, padding=<ltrb 0 20 0 30>}}
  }
  TournamentDateTextWorldCup {
   widgetClassPath=&CommonTextBlock
   bIsVariable=true
   parent=InvalidationBox_676
   widget=TournamentDateTextWorldCup
   properties {
    autoWrapText=true
    justification=Center
    text="{Start Date}"
    font {
     bForceMonospaced=false
     bMaterialIsStencil=false
     fontMaterial=null
     fontObject=&Burbank
     letterSpacing=0
     monospacedWidth=1
     size=32
     skewAmount=0
     typefaceFontName=BurbankRegular-Black
     outlineSettings {
      bApplyOutlineToDropShadows=false
      bMiteredCorners=false
      bSeparateFillAlpha=false
      outlineMaterial=null
      outlineSize=0
      outlineColor=<rgba 0 0 0 1>
     }
    }
   }
   slot {class=PanelSlot}
  }
  ScaleBoxPin {
   widgetClassPath=&ScaleBox
   parent=Overlay_0
   widget=ScaleBoxPin
   properties {stretch=UserSpecified, userSpecifiedScale=0.619}
   slot {class=OverlaySlot, properties {padding=<ltrb 8 8 0 0>}}
  }
  ShowdownPin {
   widgetClassPath=&/ShowdownPin
   bIsVariable=true
   parent=ScaleBoxPin
   widget=ShowdownPin
   properties {bRepresentsTournamentBestPin=true, colorSilhouette=true, dEVPinScore=12, dEVShowPin=true, visibility=Collapsed}
   slot {class=ScaleBoxSlot}
  }
  Text_DebugId {
   widgetClassPath=&TextBlock
   bInherited=true
   parent=Overlay_0
   widget=Text_DebugId
   properties {
    autoWrapText=true
    justification=Center
    text=EventId: Event EventWindowId: EventWindow
    wrapTextAt=700
    wrappingPolicy=AllowPerCharacterWrapping
    font {
     bForceMonospaced=false
     bMaterialIsStencil=false
     fontMaterial=null
     fontObject=&Roboto
     letterSpacing=0
     monospacedWidth=1
     size=36
     skewAmount=0
     typefaceFontName=Bold
     outlineSettings {
      bApplyOutlineToDropShadows=false
      bMiteredCorners=false
      bSeparateFillAlpha=false
      outlineMaterial=null
      outlineSize=4
      outlineColor=<rgba 0 0 0 1>
     }
    }
   }
   slot {class=OverlaySlot, properties {horizontalAlignment=HAlign_Fill, verticalAlignment=VAlign_Center}}
  }
  Border_Alert {
   widgetClassPath=&Border
   bIsVariable=true
   parent=Overlay_0
   widget=Border_Alert
   properties {
    visibility=Collapsed
    background {
     bIsDynamicallyLoaded=false
     drawAs=Image
     imageType=NoImage
     mirroring=NoMirror
     resourceName=None
     resourceObject=/Game/UI/Login/SubGameSelect/Materials/M_Notification_Right
     tiling=NoTile
     imageSize=<96 32>
     margin=<ltrb 0 0 0 0>
     outlineSettings {
      bUseBrushTransparency=false
      roundingType=HalfHeightRadius
      width=0
      color {colorUseRule=UseColor_Specified, specifiedColor=<rgba 0 0 0 0>}
      cornerRadii=<0 0 0 0>
     }
     tintColor {colorUseRule=UseColor_Specified, specifiedColor=<rgba 1 1 1 1>}
     uVRegion {bIsValid=false, max=<0 0>, min=<0 0>}
    }
    padding=<ltrb 0 0 0 0>
    renderTransform {angle=-20, scale=<1 1>, shear=<0 0>, translation=<0 0>}
   }
   slot {class=OverlaySlot, properties {horizontalAlignment=HAlign_Center, verticalAlignment=VAlign_Center, padding=<ltrb 0 0 0 200>}}
  }
  SizeBox_78 {
   widgetClassPath=&SizeBox
   parent=Border_Alert
   widget=SizeBox_78
   properties {bOverride_MaxDesiredWidth=true, maxDesiredWidth=700, widthOverride=540}
   slot {class=BorderSlot, properties {padding=<ltrb 0 0 0 0>}}
  }
  Text_EventAlert {
   widgetClassPath=&CommonTextBlock
   bIsVariable=true
   parent=SizeBox_78
   widget=Text_EventAlert
   properties {
    scrollStyle=&TextScrollStyle-Base_C
    text=CANCELLED
    font {
     bForceMonospaced=false
     bMaterialIsStencil=false
     fontMaterial=null
     fontObject=&Burbank
     letterSpacing=0
     monospacedWidth=1
     size=72
     skewAmount=0
     typefaceFontName=Burbank-Black
     outlineSettings {
      bApplyOutlineToDropShadows=false
      bMiteredCorners=false
      bSeparateFillAlpha=false
      outlineMaterial=null
      outlineSize=0
      outlineColor=<rgba 0 0 0 1>
     }
    }
    shadowOffset=<0 0>
   }
   slot {class=SizeBoxSlot, properties {padding=<ltrb 70 30 120 30>}}
  }
  Text_TimeRemaining {widgetClassPath=&CommonTextBlock, bInherited=true, widget=null}
  Image_PosterBack {widgetClassPath=&EpicCMSImage, bInherited=true, widget=null}
  Image_LoadingScreen {widgetClassPath=&EpicCMSImage, bInherited=true, widget=null}
  Image_PlaylistTile {widgetClassPath=&EpicCMSImage, bInherited=true, widget=null}
 ]
}
