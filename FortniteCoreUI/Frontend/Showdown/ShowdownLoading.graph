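# Showdown tournament loading screen widget, derived from FortShowdownCurrentEventView, with a details panel on the left and a poster on the right over a full-screen background.
# On Refresh Data, it passes the series and event IDs to both child widgets via SetTournament with forced UI refresh.
# It then tints the background's dynamic material using the left and right background colors from the tournament display info.
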WidgetBlueprint /FortniteCoreUI/Frontend/Showdown/ShowdownLoading
graphs [EventGraph {nodes [
   Event Refresh Data() {is=title K2Node_Event}
   n1 = GetTournamentSeriesId!(self:Fort Showdown Detail View Object Reference) -> (ReturnValue:String)
   n2 = GetEventId!(self:Fort Showdown Detail View Object Reference) -> (ReturnValue:String)
   > SetTournament!(
    execute
    self:Fort Showdown Detail View Object Reference=Get TournamentLoadingScreenDetails() -> (TournamentLoadingScreenDetails:Tournament Loading Screen Details Object Reference)
    TournamentId:String=@n1
    EventId:String=@n2
    bForceRefreshUI:bool=true
   ) -> (then)
   > SetTournament(
    self=Get ShowdownTournamentPosterFront() -> (ShowdownTournamentPosterFront:Showdown Tournament Poster Front Object Reference)
    TournamentId=@n1
    EventId=@n2
    bForceRefreshUI=true
   )
   > n3 = GetDynamicMaterial(self=Get Image_Background() -> (Image_Background:Image Object Reference))
   n4 = Break Fort Tournament Display Info!(FortTournamentDisplayInfo:"Fort Tournament Display Info Structure (by ref)"=Get TournamentDisplayInfo() -> (TournamentDisplayInfo:Fort Tournament Display Info Structure)) -> (
    TitleLine1:Text
    TitleLine2:Text
    ScheduleInfo:Text
    FlavorDescription:Text
    DetailsDescription:Text
    ShortFormatTitle:Text
    LongFormatTitle:Text
    BackgroundTitle:Text
    PinScoreRequirement:int
    PinEarnedText:Text
    BaseColor:Linear Color Structure
    PrimaryColor:Linear Color Structure
    SecondaryColor:Linear Color Structure
    HighlightColor:Linear Color Structure
    TitleColor:Linear Color Structure
    ShadowColor:Linear Color Structure
    BackgroundLeftColor:Linear Color Structure
    BackgroundRightColor:Linear Color Structure
    BackgroundTextColor:Linear Color Structure
    PosterFadeColor:Linear Color Structure
    AlertText:Text
    SeriesPointLeaderboardName:Text
    AlertType:EFortTournamentAlertType Enum
    RoundDisplayNames:Array of Texts
   ) {is=title K2Node_BreakStruct}
   > SetVectorParameterValue(self=@n3, ParameterName=Color1, Value=@n4.BackgroundLeftColor)
   > SetVectorParameterValue(self=@n3, ParameterName=Color2, Value=@n4.BackgroundRightColor)
  ]}]
variables []
widgets {
 info {inheritedWidgetCount=6, parentClass=/Script/FortniteUI.FortShowdownCurrentEventView, rootWidgetClass=&Overlay, widgetCount=18}
 widgets [
  Overlay_0 {widgetClassPath=&Overlay, widget=Overlay_0}
  Image_Background {
   widgetClassPath=&Image
   bIsVariable=true
   parent=Overlay_0
   widget=Image_Background
   properties {brush {
     bIsDynamicallyLoaded=false
     drawAs=Image
     imageType=NoImage
     mirroring=NoMirror
     resourceName=None
     resourceObject=/FortniteCoreUI/Frontend/Showdown/Materials/M_UI_TournamentDetailsBG_Inst
     tiling=NoTile
     imageSize=<32 32>
     margin=<ltrb 0 0 0 0>
     outlineSettings {
      bUseBrushTransparency=false
      roundingType=HalfHeightRadius
      width=0
      color {colorUseRule=UseColor_Specified, specifiedColor=<rgba 0 0 0 0>}
      cornerRadii=<0 0 0 0>
     }
     tintColor {colorUseRule=UseColor_Specified, specifiedColor=<rgba 1 1 1 1>}
     uVRegion {bIsValid=false, max=<0 0>, min=<0 0>}
    }}
   slot {class=OverlaySlot, properties {horizontalAlignment=HAlign_Fill, verticalAlignment=VAlign_Fill}}
  }
  SafeZone_0 {
   widgetClassPath=&SafeZone
   bIsVariable=true
   parent=Overlay_0
   widget=SafeZone_0
   slot {class=OverlaySlot, properties {horizontalAlignment=HAlign_Fill, verticalAlignment=VAlign_Fill}}
  }
  HorizontalBox_0 {widgetClassPath=&HorizontalBox, parent=SafeZone_0, widget=HorizontalBox_0, slot {class=SafeZoneSlot}}
  SizeBox_0 {
   widgetClassPath=&SizeBox
   parent=HorizontalBox_0
   widget=SizeBox_0
   properties {bOverride_WidthOverride=true, widthOverride=1200}
   slot {class=HorizontalBoxSlot, properties {verticalAlignment=VAlign_Bottom, size {sizeRule=Fill, value=1}}}
  }
  ScaleBox_1 {widgetClassPath=&ScaleBox, parent=SizeBox_0, widget=ScaleBox_1, slot {class=SizeBoxSlot}}
  CommonBorder_0 {
   widgetClassPath=&CommonBorder
   parent=ScaleBox_1
   widget=CommonBorder_0
   properties {padding=<ltrb 80 50 80 100>}
   slot {class=ScaleBoxSlot, properties {horizontalAlignment=HAlign_Fill, verticalAlignment=VAlign_Fill}}
  }
  CommonBorder_1 {
   widgetClassPath=&CommonBorder
   parent=CommonBorder_0
   widget=CommonBorder_1
   properties {
    background {
     bIsDynamicallyLoaded=false
     drawAs=NoDrawType
     imageType=NoImage
     mirroring=NoMirror
     resourceName=None
     resourceObject=/Game/UI/Foundation/Textures/LoadingScreens/LTM/Materials/M_UI_SkewBox_Inst
     tiling=NoTile
     imageSize=<32 32>
     margin=<ltrb 0 0 0 0>
     outlineSettings {
      bUseBrushTransparency=false
      roundingType=HalfHeightRadius
      width=0
      color {colorUseRule=UseColor_Specified, specifiedColor=<rgba 0 0 0 0>}
      cornerRadii=<0 0 0 0>
     }
     tintColor {colorUseRule=UseColor_Specified, specifiedColor=<rgba 0.0052 0.0052 0.0056 0.771>}
     uVRegion {bIsValid=false, max=<0 0>, min=<0 0>}
    }
    padding=<ltrb 0 0 0 0>
   }
   slot {class=BorderSlot, properties {padding=<ltrb 80 50 80 100>}}
  }
  TournamentLoadingScreenDetails {
   widgetClassPath=&/TournamentLoadingScreenDetails
   bIsVariable=true
   parent=CommonBorder_1
   widget=TournamentLoadingScreenDetails
   slot {class=BorderSlot, properties {padding=<ltrb 0 0 0 0>}}
  }
  SizeBox_1 {
   widgetClassPath=&SizeBox
   parent=HorizontalBox_0
   widget=SizeBox_1
   properties {bOverride_WidthOverride=true, widthOverride=750}
   slot {class=HorizontalBoxSlot, properties {horizontalAlignment=HAlign_Right, verticalAlignment=VAlign_Bottom}}
  }
  ScaleBox_0 {widgetClassPath=&ScaleBox, parent=SizeBox_1, widget=ScaleBox_0, slot {class=SizeBoxSlot}}
  ShowdownTournamentPosterFront {
   widgetClassPath=&/ShowdownTournamentPosterFront
   bIsVariable=true
   parent=ScaleBox_0
   widget=ShowdownTournamentPosterFront
   properties {hiddenWhileLoading=true, hidePin=true}
   slot {class=ScaleBoxSlot}
  }
  Text_TimeRemaining {widgetClassPath=&CommonTextBlock, bInherited=true, widget=null}
  Text_DebugId {widgetClassPath=&TextBlock, bInherited=true, widget=null, properties=~}
  Image_PosterFront {widgetClassPath=/Script/EpicCMSUIFramework.EpicCMSImage, bInherited=true, widget=null}
  Image_PosterBack {widgetClassPath=&EpicCMSImage, bInherited=true, widget=null}
  Image_LoadingScreen {widgetClassPath=&EpicCMSImage, bInherited=true, widget=null}
  Image_PlaylistTile {widgetClassPath=&EpicCMSImage, bInherited=true, widget=null}
 ]
}
