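# Translucent UI material for the tournament score HUD badge, drawing a ShapeColor fill inside a gold rim.
# AAThreshold cuts antialiased fill, rim and inner-shade masks from badge silhouette gradients, with the outer shape from the Tier texture.
# A floored U coordinate shades the two halves differently for a bevel look, and radial gradients tint the gold.
# Static switches add either a color-dodged shine swipe on the rim or, without Shine, an optional scaled inner plate.
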Material /FortniteCoreUI/Frontend/Showdown/Materials/M_UI_TournamentScoreHUD
graph {
 properties {
  bUsedWithStaticMesh=false
  blendMode=BLEND_Translucent
  materialDomain=MD_UI
  pixelDepthOffsetMode=PDOM_Legacy
  usedShadingModels=Default Lit
 }
 nodes [
  n1 = ScalarParameter() {defaultValue=0.1, parameterName=EdgeBase}
  n2 = ScalarParameter() {defaultValue=0.4, parameterName=OutlineSize}
  n3 = TextureCoordinate()
  n4 = TextureSample(UVs=@n3) {texture=&/T_UI_Badge_Silhouette_1}
  n5 = Multiply(A=MaterialFunctionCall!(
    Threshold=Add(A=@n1, B=@n2)
    Gradient=@n4.R
    Filter Width Override=ScalarParameter() {defaultValue=0.05, parameterName=EdgeSoftness_Outline}
    Filter Width Fudge Factor
   ) {materialFunction=/Game/UI/Materials/AAThreshold}.Result)
  n6 = Constant() {r=1}
  n7 = MaterialFunctionCall(
   Threshold=@n1
   Gradient=TextureSampleParameter2D(UVs=@n3) {parameterName=Tier, texture=&T_UI_Badge_Silhouette_1}.R
   Filter Width Override=ScalarParameter() {defaultValue=0.025, parameterName=EdgeSoftness_Base}
  ) {materialFunction=&AAThreshold}
  n8 = DotProduct(A=OneMinus(None=@n5), B=Clamp(None=Multiply(A=@n7.Result)))
  Clamp(None=Add(A=Multiply(A=@n5, B=@n6), B=@n8))
  n9 = MaterialFunctionCall!(In_Uv) {materialFunction=/Game/UI/Foundation/Materials/M-Function-ShineSwipe}
  n10 = VectorParameter() {parameterName=ShapeColor, defaultValue=<rgba 0.1559 0.0561 0 1>}
  n11 = TextureCoordinate()
  n12 = Floor(None=Add(A=ComponentMask(None=@n11) {r=true}) {constB=0.5})
  n13 = LinearInterpolate(A=@n10.RGB, B=Multiply(A=@n10.RGB) {constB=0.8}, Alpha=@n12)
  n14 = ScalarParameter() {defaultValue=0.8, parameterName=OutlineSize_1}
  n15 = OneMinus(None=MaterialFunctionCall(
    Threshold=Add(A=@n1, B=@n14)
    Gradient=@n4.R
    Filter Width Override=ScalarParameter() {defaultValue=0.05, parameterName=EdgeSoftness_Outline_1}
   ) {materialFunction=&AAThreshold}.Result)
  n16 = OneMinus(None=Multiply(A=DotProduct(A=@n15, B=@n5)) {constB=0.4})
  n17 = TextureCoordinate() {desc=1.5, uTiling=0.5}
  n18 = Constant2Vector() {r=0.25}
  n19 = Constant() {r=0.75}
  n20 = Constant() {r=3}
  n21 = OneMinus(None=Multiply(A=MaterialFunctionCall(UVs=@n17, CenterPosition=@n18, Radius=@n19, Density=@n20) {materialFunction=&RadialGradientExponential}.RadialGradientExponential))
  n22 = LinearInterpolate(
   A=LinearInterpolate(A=Constant3Vector() {constant=<rgba 1 1 0.3916 0>}, B=Constant3Vector() {constant=<rgba 0.5395 0.1746 0.0185 0>}, Alpha=@n21)
   B=Constant3Vector() {constant=<rgba 1 0.6308 0.1022 0>}
   Alpha=@n21
  )
  n23 = LinearInterpolate(
   A=LinearInterpolate(A=Constant3Vector() {constant=<rgba 1 0.6584 0.0802 0>}, B=Constant3Vector() {constant=<rgba 0.2051 0.0452 0.006 0>}, Alpha=@n21)
   B=Constant3Vector() {constant=<rgba 0.5395 0.1746 0.0185 0>}
   Alpha=@n21
  )
  n24 = Add(A=LinearInterpolate(B=Multiply(A=@n13, B=@n16), Alpha=@n5), B=LinearInterpolate(B=LinearInterpolate(A=@n22, B=@n23, Alpha=@n12), Alpha=@n8))
  n25 = TextureSample(UVs=MaterialFunctionCall!(
    Texture Scale=ScalarParameter() {defaultValue=1, parameterName=PlateScale}
    UVs
    Offset XY Texturesize Z=VectorParameter() {parameterName=PlateSize, defaultValue=<rgba 64 40 128 1>}.RGB
   ) {materialFunction=/Engine/Functions/Engine_MaterialFunctions01/Texturing/ScaleUVsAroundPoint}.UVs) {texture=&/T_UI_Ingame_TournamentScorePlate_InnerPlate}
  StaticSwitchParameter(
   True=MaterialFunctionCall!(Blend=LinearInterpolate(B=@n9.Alpha, Alpha=@n8), Base=@n24) {materialFunction=/Engine/Functions/Engine_MaterialFunctions03/Blends/Blend_ColorDodge}.Result
   False=StaticSwitchParameter(True=LinearInterpolate(A=@n24, B=@n25.RGB, Alpha=@n25.RGB), False=@n24) {parameterName=Plate}
  ) {defaultValue=true, parameterName=Shine}
 ]
}
