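# Showdown tournament HUD alert, derived from FortHUDScoringAlert, that pops up when the player scores points, showing an icon, description and point count.
# OnPointScored reveals the blurred container, starts ScoreAnimation and plays a point sound cue.
# The animation slides and fades the content, punches the count's scale and sweeps its font-material shine.
# An animation-finished delegate, bound on Construct, collapses the container and calls HandlePointAnimationFinished on the parent class.
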WidgetBlueprint /FortniteCoreUI/Frontend/Showdown/ShowdownScoringHUD
animations [ScoreAnimation {
  binding_count=4
  display_label=""
  length_seconds=5
  object=&self:ScoreAnimation
  sequence {bindings [
    Text_ScoreDesc {
     id=#1
     possessed_class=CommonTextBlock
     tracks [{
       type=MovieScene2DTransformTrack
       sections [{
         type=MovieScene2DTransformSection
         end_frame=100
         start_frame=0
         channels [
          Translation.X {type=MovieSceneScriptingFloatChannel, keys [[0,-54.5397],[21,0]]}
          Translation.Y {type=MovieSceneScriptingFloatChannel}
          Angle {type=MovieSceneScriptingFloatChannel}
          Scale.X {type=MovieSceneScriptingFloatChannel}
          Scale.Y {type=MovieSceneScriptingFloatChannel}
          Shear.X {type=MovieSceneScriptingFloatChannel}
          Shear.Y {type=MovieSceneScriptingFloatChannel}
         ]
        }]
      }]
    }
    Overlay_0 {
     id=#2
     display_name=CommonBorder_PointsContainer
     possessed_class=Overlay
     tracks [{
       type=MovieSceneFloatTrack
       display_name=Render Opacity
       sections [{
         type=MovieSceneFloatSection
         end_frame=100
         start_frame=0
         channels [None {type=MovieSceneScriptingFloatChannel, keys [[0,0],[15.1393,1],[94.7893,1],[100,0]]}]
        }]
      }]
    }
    HorizontalBox {
     id=#3
     display_name=ScoringContent
     possessed_class=HorizontalBox
     tracks [
      {
       type=MovieSceneByteTrack
       display_name=Visibility
       sections [{type=MovieSceneByteSection, end_frame=100, start_frame=0, channels [None {type=MovieSceneScriptingByteChannel, keys [[0,3],[100,1]]}]}]
      }
      {
       type=MovieSceneColorTrack
       display_name=Content Color and Opacity
       sections [{
         type=MovieSceneColorSection
         end_frame=100
         start_frame=0
         channels [
          Color.R {type=MovieSceneScriptingFloatChannel, const [1,1]}
          Color.G {type=MovieSceneScriptingFloatChannel, const [1,1]}
          Color.B {type=MovieSceneScriptingFloatChannel, const [1,1]}
          Color.A {type=MovieSceneScriptingFloatChannel, keys [[0,0],[10,1],[94.7893,1],[100,0]]}
         ]
        }]
      }
      {
       type=MovieScene2DTransformTrack
       sections [{
         type=MovieScene2DTransformSection
         end_frame=100
         start_frame=0
         channels [
          Translation.X {type=MovieSceneScriptingFloatChannel, keys [[0,-282.8563],[10,0],[94.7893,0],[100,-100]]}
          Translation.Y {type=MovieSceneScriptingFloatChannel}
          Angle {type=MovieSceneScriptingFloatChannel}
          Scale.X {type=MovieSceneScriptingFloatChannel}
          Scale.Y {type=MovieSceneScriptingFloatChannel}
          Shear.X {type=MovieSceneScriptingFloatChannel}
          Shear.Y {type=MovieSceneScriptingFloatChannel}
         ]
        }]
      }
     ]
    }
    Text_PointCount {
     id=#4
     possessed_class=CommonTextBlock
     tracks [
      {
       type=MovieScene2DTransformTrack
       sections [{
         type=MovieScene2DTransformSection
         end_frame=100
         start_frame=0
         channels [
          Translation.X {type=MovieSceneScriptingFloatChannel}
          Translation.Y {type=MovieSceneScriptingFloatChannel}
          Angle {type=MovieSceneScriptingFloatChannel}
          Scale.X {type=MovieSceneScriptingFloatChannel, keys [[15.1393,1],[16,1.1982],[19,0.9804],[19.8487,1]]}
          Scale.Y {type=MovieSceneScriptingFloatChannel, keys [[15.1393,1],[16,1.1982],[19,0.9804],[19.8487,1]]}
          Shear.X {type=MovieSceneScriptingFloatChannel}
          Shear.Y {type=MovieSceneScriptingFloatChannel}
         ]
        }]
      }
      {
       type=MovieSceneWidgetMaterialTrack
       display_name=Font.FontMaterial
       sections [{
         type=MovieSceneParameterSection
         channels [TimeOffset {type=MovieSceneScriptingFloatChannel, keys [[16.5183,0.8425],[30.2603,-1.3558],[37.5823,0.8425],[51.3243,-1.3558]]}]
        }]
      }
     ]
    }
   ]}
 }]
graphs [EventGraph {nodes [
   n1 = ScoreFinished() {is=title K2Node_CustomEvent}
   > SetVisibility(self=Get CommonBorder_PointsContainer() -> (CommonBorder_PointsContainer:Common Border Object Reference), InVisibility=Collapsed)
   > HandlePointAnimationFinished(self:Fort HUDScoring Alert Object Reference)
   Event Destruct()
   > UnbindFromAnimationFinished!(
    execute
    self:User Widget Object Reference
    Animation:Widget Animation Object Reference=Get ScoreAnimation() -> (ScoreAnimation:Widget Animation Object Reference)
    Delegate:Delegate=Create Event()
   ) -> (then)
   Event Construct()
   > BindToAnimationFinished!(execute, self:Widget Animation Object Reference=Get ScoreAnimation(), Widget:User Widget Object Reference=Self-Reference(), Delegate=@n1) -> (then)
   Event OnPointScored() {is=title K2Node_Event}
   > PlayAnimation(InAnimation=Get ScoreAnimation())
   > SetVisibility(self=Get CommonBorder_PointsContainer(), InVisibility=HitTestInvisible)
   > PlaySound2D(Sound=/Game/Sounds/Fort_UI_Sounds/FrontEnd/Store/Fort_UI_Showdown_Point_Cue)
  ]}]
variables [ScoringRuleDisplayIcons {
  type="array<Texture2D>"
  default [
   /Game/UI/Foundation/Textures/Icons/Quest/T-Icon-Trophy-32
   /Game/Athena/HUD/Art/killsico_alt.KillsIco_Alt
   /FortniteCoreUI/UI/Events/Icons/T-Icon-Bus-Fare-Flat
  ]
 }]
widgets {
 info {inheritedWidgetCount=4, parentClass=/Script/FortniteUI.FortHUDScoringAlert, rootWidgetClass=&Overlay, widgetCount=16}
 widgets [
  Overlay_0 {widgetClassPath=&Overlay, widget=Overlay_0}
  CommonBorder_PointsContainer {
   widgetClassPath=&CommonBorder
   bIsVariable=true
   parent=Overlay_0
   widget=CommonBorder_PointsContainer
   properties {visibility=Collapsed, padding=<ltrb 0 0 0 0>}
   slot {class=OverlaySlot, properties {verticalAlignment=VAlign_Center}}
  }
  BackgroundBlurPoints {
   widgetClassPath=&BackgroundBlur
   parent=CommonBorder_PointsContainer
   widget=BackgroundBlurPoints
   properties {blurStrength=40}
   slot {class=BorderSlot, properties {padding=<ltrb 0 0 0 0>}}
  }
  Overlay_2 {
   widgetClassPath=&Overlay
   parent=BackgroundBlurPoints
   widget=Overlay_2
   slot {class=BackgroundBlurSlot, properties {padding=<ltrb 0 0 0 0>}}
  }
  Image_39 {
   widgetClassPath=&Image
   parent=Overlay_2
   widget=Image_39
   properties {
    visibility=SelfHitTestInvisible
    brush {
     bIsDynamicallyLoaded=false
     drawAs=Image
     imageType=NoImage
     mirroring=NoMirror
     resourceName=None
     resourceObject=null
     tiling=NoTile
     imageSize=<32 36.9007>
     margin=<ltrb 0 0 0 0>
     outlineSettings {
      bUseBrushTransparency=false
      roundingType=HalfHeightRadius
      width=0
      color {colorUseRule=UseColor_Specified, specifiedColor=<rgba 0 0 0 0>}
      cornerRadii=<0 0 0 0>
     }
     tintColor {colorUseRule=UseColor_Specified, specifiedColor=<rgba 1 1 1 1>}
     uVRegion {bIsValid=false, max=<0 0>, min=<0 0>}
    }
    colorAndOpacity=<rgba 1 0.3333 0 0.6>
   }
   slot {class=OverlaySlot, properties {horizontalAlignment=HAlign_Fill, verticalAlignment=VAlign_Bottom}}
  }
  ScoringContent {
   widgetClassPath=&CommonBorder
   parent=Overlay_2
   widget=ScoringContent
   properties {verticalAlignment=VAlign_Center, visibility=Collapsed, padding=<ltrb 4 2 10 2>}
   slot {class=OverlaySlot}
  }
  Overlay_4 {
   widgetClassPath=&Overlay
   parent=ScoringContent
   widget=Overlay_4
   slot {class=BorderSlot, properties {verticalAlignment=VAlign_Center, padding=<ltrb 4 2 10 2>}}
  }
  VerticalBox_0 {
   widgetClassPath=&VerticalBox
   parent=Overlay_4
   widget=VerticalBox_0
   slot {class=OverlaySlot, properties {padding=<ltrb 80 0 0 0>}}
  }
  Text_ScoreDesc {
   widgetClassPath=&CommonTextBlock
   bInherited=true
   bIsVariable=true
   parent=VerticalBox_0
   widget=Text_ScoreDesc
   properties {
    justification=Right
    text=4 ELIMINATIONS
    textTransformPolicy=ToUpper
    font {
     bForceMonospaced=false
     bMaterialIsStencil=false
     fontMaterial=null
     fontObject=&Burbank
     letterSpacing=0
     monospacedWidth=1
     size=18
     skewAmount=0
     typefaceFontName=Burbank-PlayerPower
     outlineSettings {
      bApplyOutlineToDropShadows=false
      bMiteredCorners=false
      bSeparateFillAlpha=false
      outlineMaterial=null
      outlineSize=0
      outlineColor=<rgba 0 0 0 1>
     }
    }
   }
   slot {
    class=VerticalBoxSlot
    properties {horizontalAlignment=HAlign_Right, verticalAlignment=VAlign_Center, padding=<ltrb 0 2 0 0>, size {sizeRule=Fill, value=1}}
   }
  }
  Text_PointCount {
   widgetClassPath=&CommonTextBlock
   bInherited=true
   bIsVariable=true
   parent=VerticalBox_0
   widget=Text_PointCount
   properties {
    justification=Right
    text=+1 POINT
    font {
     bForceMonospaced=false
     bMaterialIsStencil=false
     fontMaterial=/FortniteCoreUI/Athena/HUD/Materials/M_UI_PointShine
     fontObject=&Burbank
     letterSpacing=0
     monospacedWidth=1
     size=24
     skewAmount=0
     typefaceFontName=Burbank-Black
     outlineSettings {
      bApplyOutlineToDropShadows=false
      bMiteredCorners=false
      bSeparateFillAlpha=false
      outlineMaterial=null
      outlineSize=1
      outlineColor=<rgba 0.6654 0.1413 0 1>
     }
    }
    renderTransform {angle=0, scale=<1 1>, shear=<-7 0>, translation=<0 0>}
    renderTransformPivot=<0.9 0.8>
    shadowColorAndOpacity=<rgba 0.1812 0.0648 0 1>
    shadowOffset=<3 3>
   }
   slot {class=VerticalBoxSlot, properties {horizontalAlignment=HAlign_Right, verticalAlignment=VAlign_Center, size {sizeRule=Fill, value=1}}}
  }
  Overlay_1 {
   widgetClassPath=&Overlay
   parent=Overlay_4
   widget=Overlay_1
   slot {class=OverlaySlot, properties {verticalAlignment=VAlign_Center}}
  }
  Overlay_3 {
   widgetClassPath=&Overlay
   parent=Overlay_1
   widget=Overlay_3
   properties {renderTransform {angle=0, scale=<1 1>, shear=<0 0>, translation=<0 6>}}
   slot {class=OverlaySlot, properties {padding=<ltrb 0 -3 0 0>}}
  }
  Image_ScoreShadow {
   widgetClassPath=&Image
   parent=Overlay_3
   widget=Image_ScoreShadow
   properties {
    brush {
     bIsDynamicallyLoaded=false
     drawAs=Image
     imageType=NoImage
     mirroring=NoMirror
     resourceName=None
     resourceObject=/FortniteCoreUI/Frontend/Showdown/Materials/M_UI_TournamentScoreHUD
     tiling=NoTile
     imageSize=<76 76>
     margin=<ltrb 0 0 0 0>
     outlineSettings {
      bUseBrushTransparency=false
      roundingType=HalfHeightRadius
      width=0
      color {colorUseRule=UseColor_Specified, specifiedColor=<rgba 0 0 0 0>}
      cornerRadii=<0 0 0 0>
     }
     tintColor {colorUseRule=UseColor_Specified, specifiedColor=<rgba 1 1 1 1>}
     uVRegion {bIsValid=false, max=<0 0>, min=<0 0>}
    }
    colorAndOpacity=<rgba 0.0052 0.0052 0.0056 0.45>
    renderTransform {angle=0, scale=<1 1>, shear=<0 0>, translation=<2 4>}
   }
   slot {class=OverlaySlot}
  }
  Image_ScoreButton {
   widgetClassPath=&Image
   parent=Overlay_3
   widget=Image_ScoreButton
   properties {brush {
     bIsDynamicallyLoaded=false
     drawAs=Image
     imageType=NoImage
     mirroring=NoMirror
     resourceName=None
     resourceObject=&M_UI_TournamentScoreHUD
     tiling=NoTile
     imageSize=<76 76>
     margin=<ltrb 0 0 0 0>
     outlineSettings {
      bUseBrushTransparency=false
      roundingType=HalfHeightRadius
      width=0
      color {colorUseRule=UseColor_Specified, specifiedColor=<rgba 0 0 0 0>}
      cornerRadii=<0 0 0 0>
     }
     tintColor {colorUseRule=UseColor_Specified, specifiedColor=<rgba 1 1 1 1>}
     uVRegion {bIsValid=false, max=<0 0>, min=<0 0>}
    }}
   slot {class=OverlaySlot}
  }
  Icon_Scoring {
   widgetClassPath=/Script/CommonUI.CommonLazyImage
   bInherited=true
   bIsVariable=true
   parent=Overlay_1
   widget=Icon_Scoring
   properties {brush {
     bIsDynamicallyLoaded=false
     drawAs=Image
     imageType=NoImage
     mirroring=NoMirror
     resourceName=None
     resourceObject=&KillsIco_Alt
     tiling=NoTile
     imageSize=<32 32>
     margin=<ltrb 0 0 0 0>
     outlineSettings {
      bUseBrushTransparency=false
      roundingType=HalfHeightRadius
      width=0
      color {colorUseRule=UseColor_Specified, specifiedColor=<rgba 0 0 0 0>}
      cornerRadii=<0 0 0 0>
     }
     tintColor {colorUseRule=UseColor_Specified, specifiedColor=<rgba 1 1 1 1>}
     uVRegion {bIsValid=false, max=<0 0>, min=<0 0>}
    }}
   slot {class=OverlaySlot, properties {horizontalAlignment=HAlign_Center, verticalAlignment=VAlign_Center}}
  }
  TutorialHighlight {widgetClassPath=&TutorialHighlight}
 ]
}
