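# Showdown tournament HUD badge widget, derived from FortHUDPointCounter, that displays the score.
# A switcher shows a plain plate without a score, or a badge with the inherited score text inside width-capped Size and Scale boxes.
# OnPointTextSet picks the page from bHasScore, then saves, clears and restores the ScaleBox stretch, apparently forcing a refit of the new text.
# An inherited debug event-ID text is collapsed.
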WidgetBlueprint /FortniteCoreUI/Frontend/Showdown/ShowdownScoringHUD_Progress
graphs [EventGraph {nodes [
   n1 = Event OnPointTextSet!() -> (OutputDelegate, then, bHasScore:bool) {is=title K2Node_Event}
   > SetActiveWidgetIndex(
    self=Get CommonWidgetSwitcher_State() -> (CommonWidgetSwitcher_State:Common Widget Switcher Legacy Object Reference)
    Index=Select(Option 0:int=0, Option 1:int=1, Index=@n1.bHasScore) -> (ReturnValue:int)
   )
   > Set OldStretch(OldStretch:EStretch Enum=Get Stretch!(self:Scale Box Object Reference=Get ScaleBox_0() -> (ScaleBox_0:Scale Box Object Reference)) -> (Stretch:EStretch Enum=None)) -> (Output_Get:EStretch Enum=None)
   > SetStretch!(execute, self:Scale Box Object Reference=Get ScaleBox_0(), InStretch:EStretch Enum=None) -> (then)
   > SetStretch(self=Get ScaleBox_0(), InStretch=Get OldStretch() -> (OldStretch:EStretch Enum=None))
  ]}]
variables [OldStretch {type=EStretch, default=None}]
widgets {
 info {inheritedWidgetCount=2, parentClass=/Script/FortniteUI.FortHUDPointCounter, rootWidgetClass=&Overlay, widgetCount=9}
 widgets [
  Overlay_0 {widgetClassPath=&Overlay, widget=Overlay_0}
  CommonWidgetSwitcher_State {
   widgetClassPath=&CommonWidgetSwitcherLegacy
   bIsVariable=true
   parent=Overlay_0
   widget=CommonWidgetSwitcher_State
   slot {class=OverlaySlot, properties {horizontalAlignment=HAlign_Center, verticalAlignment=VAlign_Center}}
  }
  Image_Badge {
   widgetClassPath=&Image
   parent=CommonWidgetSwitcher_State
   widget=Image_Badge
   properties {
    visibility=SelfHitTestInvisible
    brush {
     bIsDynamicallyLoaded=false
     drawAs=Image
     imageType=NoImage
     mirroring=NoMirror
     resourceName=None
     resourceObject=/FortniteCoreUI/Frontend/Showdown/Materials/M_UI_TournamentScoreHUD_Plate
     tiling=NoTile
     imageSize=<75 75>
     margin=<ltrb 0 0 0 0>
     outlineSettings {
      bUseBrushTransparency=false
      roundingType=HalfHeightRadius
      width=0
      color {colorUseRule=UseColor_Specified, specifiedColor=<rgba 0 0 0 0>}
      cornerRadii=<0 0 0 0>
     }
     tintColor {colorUseRule=UseColor_Specified, specifiedColor=<rgba 1 1 1 1>}
     uVRegion {bIsValid=false, max=<0 0>, min=<0 0>}
    }
   }
   slot {class=WidgetSwitcherSlot}
  }
  Overlay_1 {widgetClassPath=&Overlay, parent=CommonWidgetSwitcher_State, widget=Overlay_1, slot {class=WidgetSwitcherSlot}}
  Image_OpenBadge {
   widgetClassPath=&Image
   parent=Overlay_1
   widget=Image_OpenBadge
   properties {
    visibility=SelfHitTestInvisible
    brush {
     bIsDynamicallyLoaded=false
     drawAs=Image
     imageType=NoImage
     mirroring=NoMirror
     resourceName=None
     resourceObject=/FortniteCoreUI/Frontend/Showdown/Materials/M_UI_TournamentScoreHUD
     tiling=NoTile
     imageSize=<75 75>
     margin=<ltrb 0 0 0 0>
     outlineSettings {
      bUseBrushTransparency=false
      roundingType=HalfHeightRadius
      width=0
      color {colorUseRule=UseColor_Specified, specifiedColor=<rgba 0 0 0 0>}
      cornerRadii=<0 0 0 0>
     }
     tintColor {colorUseRule=UseColor_Specified, specifiedColor=<rgba 1 1 1 1>}
     uVRegion {bIsValid=false, max=<0 0>, min=<0 0>}
    }
   }
   slot {class=OverlaySlot}
  }
  SizeBox_0 {
   widgetClassPath=&SizeBox
   parent=Overlay_1
   widget=SizeBox_0
   properties {bOverride_MaxDesiredWidth=true, maxDesiredWidth=50}
   slot {class=OverlaySlot, properties {horizontalAlignment=HAlign_Center, verticalAlignment=VAlign_Center, padding=<ltrb 0 -4 0 0>}}
  }
  ScaleBox_0 {widgetClassPath=&ScaleBox, bIsVariable=true, parent=SizeBox_0, widget=ScaleBox_0, slot {class=SizeBoxSlot}}
  Text_ScoreValue {
   widgetClassPath=&CommonTextBlock
   bInherited=true
   bIsVariable=true
   parent=ScaleBox_0
   widget=Text_ScoreValue
   properties {
    text=--
    font {
     bForceMonospaced=false
     bMaterialIsStencil=false
     fontMaterial=/FortniteCoreUI/Athena/HUD/Materials/M_UI_PointShine
     fontObject=&Burbank
     letterSpacing=0
     monospacedWidth=1
     size=30
     skewAmount=0
     typefaceFontName=Burbank-Black
     outlineSettings {
      bApplyOutlineToDropShadows=false
      bMiteredCorners=false
      bSeparateFillAlpha=false
      outlineMaterial=null
      outlineSize=1
      outlineColor=<rgba 0.6654 0.1413 0 1>
     }
    }
   }
   slot {class=ScaleBoxSlot}
  }
  Text_DebugId {
   widgetClassPath=&TextBlock
   bInherited=true
   parent=Overlay_0
   widget=Text_DebugId
   properties {
    autoWrapText=true
    justification=Center
    text=EventId: Event EventWindowId: EventWindow
    visibility=Collapsed
    wrapTextAt=300
    wrappingPolicy=AllowPerCharacterWrapping
    font {
     bForceMonospaced=false
     bMaterialIsStencil=false
     fontMaterial=null
     fontObject=&Roboto
     letterSpacing=0
     monospacedWidth=1
     size=12
     skewAmount=0
     typefaceFontName=Bold
     outlineSettings {
      bApplyOutlineToDropShadows=false
      bMiteredCorners=false
      bSeparateFillAlpha=false
      outlineMaterial=null
      outlineSize=2
      outlineColor=<rgba 0 0 0 1>
     }
    }
   }
   slot {class=OverlaySlot, properties {horizontalAlignment=HAlign_Center, verticalAlignment=VAlign_Center}}
  }
 ]
}
