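# Circular arrow button for the Showdown frontend UI, built on CommonButtonLegacy with an optional drop shadow and an input action icon.
# PreConstruct mirrors arrow and shadow through render scale when Flip is set, positions the icon, and applies softness and colors.
# UpdateArrowColor stores tail and hover colors and pushes them to the arrow's dynamic material.
# Hover animates the material's white fill, reversing on unhover; click briefly nudges the arrow diagonally.
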WidgetBlueprint /FortniteCoreUI/Frontend/Showdown/CircleArrowButton
animations [
 Hover {
  binding_count=1
  display_label=""
  length_seconds=0.14
  object=&self:Hover
  sequence {
   bindings [Image_Arrow {
     id=#1
     possessed_class=Image
     tracks [{
       type=MovieSceneWidgetMaterialTrack
       sections [{type=MovieSceneParameterSection, channels [AnimateWhiteFill {type=MovieSceneScriptingFloatChannel, keys [[0,0],[14,1]]}]}]
      }]
    }]
   display_rate {numerator=100}
  }
 }
 Click {
  binding_count=1
  display_label=""
  length_seconds=0.3119
  object=&self:Click
  sequence {bindings [Image_Arrow {
     id=#2
     possessed_class=Image
     tracks [{
       type=MovieScene2DTransformTrack
       sections [{
         type=MovieScene2DTransformSection
         end_frame=6
         start_frame=0
         channels [
          Translation.X {type=MovieSceneScriptingFloatChannel, keys [[0,0],[1.8057,6],[4.875,-1.1514],[6.238,0]]}
          Translation.Y {type=MovieSceneScriptingFloatChannel, keys [[0,0],[1.8057,6],[4.875,-1.1514],[6.238,0]]}
          Angle {type=MovieSceneScriptingFloatChannel}
          Scale.X {type=MovieSceneScriptingFloatChannel}
          Scale.Y {type=MovieSceneScriptingFloatChannel}
          Shear.X {type=MovieSceneScriptingFloatChannel}
          Shear.Y {type=MovieSceneScriptingFloatChannel}
         ]
        }]
      }]
    }]}
 }
]
graphs [
 EventGraph {nodes [
   Event PreConstruct()
   n1 = Get Flip() -> (Flip:bool)
   > SetHorizontalAlignment(
    self=SlotAsOverlaySlot(Widget=Get Sizebox_InputAction() -> (Sizebox_InputAction:Size Box Object Reference))
    InHorizontalAlignment=Select(
     Option 0:EHorizontalAlignment Enum=HAlign_Center
     Option 1:EHorizontalAlignment Enum=Select(Option 0=HAlign_Right, Option 1=HAlign_Left, Index=@n1)
     Index=Get InputActionOnSide() -> (InputActionOnSide:bool)
    ) -> (ReturnValue:EHorizontalAlignment Enum)
   )
   > SetVerticalAlignment(
    self=SlotAsOverlaySlot(Widget=Get Sizebox_InputAction())
    InVerticalAlignment=Select(Option 0:EVerticalAlignment Enum=VAlign_Top, Option 1:EVerticalAlignment Enum=VAlign_Center, Index=Get InputActionOnSide()) -> (ReturnValue:EVerticalAlignment Enum)
   )
   n2 = Get Image_Shadow() -> (Image_Shadow:Image Object Reference)
   > n3 = GetDynamicMaterial(self=@n2)
   n4 = Get EdgeSoftness() -> (EdgeSoftness:double)
   > SetScalarParameterValue(self=@n3, ParameterName=EdgeSoftness, Value=@n4)
   n5 = Select(Option 0:Vector 2D Structure=(X=1,Y=1), Option 1:Vector 2D Structure=(X=-1,Y=1), Index=Get Flip()) -> (ReturnValue:Vector 2D Structure)
   > SetRenderScale!(execute, self:Widget Object Reference=@n2, Scale:Vector 2D Structure=@n5) -> (then)
   > SetColorAndOpacity(self:Image Object Reference=@n2, InColorAndOpacity:Linear Color Structure=Get ShadowColor() -> (ShadowColor:Linear Color Structure))
   n6 = Get Image_Arrow() -> (Image_Arrow:Image Object Reference)
   > SetRenderScale(self=@n6, Scale=@n5)
   > n7 = GetDynamicMaterial(self=@n6)
   > SetScalarParameterValue(self=@n7, ParameterName=EdgeSoftness, Value=@n4)
   > UpdateArrowColor!(
    execute
    self
    Color:Linear Color Structure=Get ArrowColor() -> (ArrowColor:Linear Color Structure)
    HoverColor:Linear Color Structure=Get HoverColor() -> (HoverColor:Linear Color Structure)
   ) -> (then)
   > SetVisibility(
    self=Get Image_Shadow()
    InVisibility=Select(Option 0:ESlateVisibility Enum=Collapsed, Option 1:ESlateVisibility Enum=SelfHitTestInvisible, Index=Get UseShadow() -> (UseShadow:bool)) -> (ReturnValue:ESlateVisibility Enum)
   )
   Event On Hovered()
   n8 = Get Hover() -> (Hover:Widget Animation Object Reference)
   > PlayAnimation(InAnimation=@n8, PlaybackSpeed=2)
   Event On Unhovered()
   > PlayAnimation(InAnimation=@n8, PlayMode=Reverse)
   Event On Clicked()
   > PlayAnimation(InAnimation=Get Click() -> (Click:Widget Animation Object Reference))
   Event Tick()
  ]}
 UpdateArrowColor {nodes [
   n1 = UpdateArrowColor!() -> (then, Color:Linear Color Structure, HoverColor:Linear Color Structure) {is=title K2Node_FunctionEntry}
   > n2 = Set ArrowColor(ArrowColor:Linear Color Structure=@n1) -> (Output_Get:Linear Color Structure)
   > n3 = Set HoverColor(HoverColor:Linear Color Structure=@n1.HoverColor) -> (Output_Get:Linear Color Structure)
   > n4 = GetDynamicMaterial(self=Get Image_Arrow())
   > SetVectorParameterValue(self=@n4, ParameterName=TailColor, Value=@n2)
   > SetVectorParameterValue(self=@n4, ParameterName=HoverColor, Value=@n3)
  ]}
]
variables [
 ArrowColor {type=LinearColor, default=<rgba 0 0.9734 1 1>}
 EdgeSoftness {type=number, default=0.015}
 Flip {type=boolean}
 HoverColor {type=LinearColor, default=<rgba 1 1 1 1>}
 InputActionOnSide {type=boolean}
 ShadowColor {type=LinearColor, default=<rgba 0 0.0319 0.162 0.66>}
 UseShadow {type=boolean, default=true}
]
widgets {
 info {inheritedWidgetCount=1, parentClass=&CommonButtonLegacy, rootWidgetClass=&Overlay, widgetCount=6}
 widgets [
  Overlay_1 {widgetClassPath=&Overlay, widget=Overlay_1}
  Overlay_ButtonContainer {
   widgetClassPath=&Overlay
   bIsVariable=true
   parent=Overlay_1
   widget=Overlay_ButtonContainer
   slot {class=OverlaySlot, properties {horizontalAlignment=HAlign_Center, verticalAlignment=VAlign_Center}}
  }
  Image_Shadow {
   widgetClassPath=&Image
   bIsVariable=true
   parent=Overlay_ButtonContainer
   widget=Image_Shadow
   properties {
    brush {
     bIsDynamicallyLoaded=false
     drawAs=Image
     imageType=NoImage
     mirroring=NoMirror
     resourceName=None
     resourceObject=/FortniteCoreUI/Frontend/Showdown/Materials/M_UI_ArrowCircle
     tiling=NoTile
     imageSize=<128 128>
     margin=<ltrb 0 0 0 0>
     outlineSettings {
      bUseBrushTransparency=false
      roundingType=HalfHeightRadius
      width=0
      color {colorUseRule=UseColor_Specified, specifiedColor=<rgba 0 0 0 0>}
      cornerRadii=<0 0 0 0>
     }
     tintColor {colorUseRule=UseColor_Specified, specifiedColor=<rgba 1 1 1 1>}
     uVRegion {bIsValid=false, max=<0 0>, min=<0 0>}
    }
    colorAndOpacity=<rgba 0 0.0319 0.162 0.66>
    renderTransform {angle=0, scale=<1 1>, shear=<0 0>, translation=<8 8>}
   }
   slot {class=OverlaySlot}
  }
  Image_Arrow {
   widgetClassPath=&Image
   bIsVariable=true
   parent=Overlay_ButtonContainer
   widget=Image_Arrow
   properties {brush {
     bIsDynamicallyLoaded=false
     drawAs=Image
     imageType=NoImage
     mirroring=NoMirror
     resourceName=None
     resourceObject=&M_UI_ArrowCircle
     tiling=NoTile
     imageSize=<128 128>
     margin=<ltrb 0 0 0 0>
     outlineSettings {
      bUseBrushTransparency=false
      roundingType=HalfHeightRadius
      width=0
      color {colorUseRule=UseColor_Specified, specifiedColor=<rgba 0 0 0 0>}
      cornerRadii=<0 0 0 0>
     }
     tintColor {colorUseRule=UseColor_Specified, specifiedColor=<rgba 1 1 1 1>}
     uVRegion {bIsValid=false, max=<0 0>, min=<0 0>}
    }}
   slot {class=OverlaySlot}
  }
  Sizebox_InputAction {
   widgetClassPath=&SizeBox
   bIsVariable=true
   parent=Overlay_ButtonContainer
   widget=Sizebox_InputAction
   properties {bOverride_HeightOverride=true, bOverride_WidthOverride=true, heightOverride=32, widthOverride=32}
   slot {class=OverlaySlot, properties {horizontalAlignment=HAlign_Right, verticalAlignment=VAlign_Center, padding=<ltrb -42 -16 -42 -16>}}
  }
  InputActionWidget {
   widgetClassPath=&CommonActionWidget
   bInherited=true
   bIsVariable=true
   parent=Sizebox_InputAction
   widget=InputActionWidget
   properties {visibility=SelfHitTestInvisible}
   slot {class=SizeBoxSlot}
  }
 ]
}
