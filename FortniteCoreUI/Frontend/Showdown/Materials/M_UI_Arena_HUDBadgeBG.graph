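# UI material for the background of the Arena HUD badge in the Showdown frontend, giving a dark purple fill with a white outline.
# ScaleUVsByCenter, driven by the OutlineSize scalar parameter, supplies a 0-1 mask that blends between white and the purple.
# The outline is procedural, with no texture sampled in the graph.
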Material /FortniteCoreUI/Frontend/Showdown/Materials/M_UI_Arena_HUDBadgeBG
graph {
 properties {bUsedWithStaticMesh=false, materialDomain=MD_UI, pixelDepthOffsetMode=PDOM_Legacy, usedShadingModels=Default Lit}
 nodes [LinearInterpolate(
   A=Constant() {r=1}
   B=Constant3Vector() {constant=<rgba 0.0284 0.0003 0.117 1>}
   Alpha=MaterialFunctionCall(Texture Scale=ScalarParameter() {defaultValue=0.87, parameterName=OutlineSize}) {materialFunction=&ScaleUVsByCenter}.0-1 mask
  )]
}
