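# PrimarySkewButton is a CommonButtonLegacy widget for the Showdown frontend: a skewed yellow button with centred, sheared text over an offset shadow.
# On Construct and in design-time PreConstruct, it sets the label, optionally uppercased, and picks a normal or XL Burbank font.
# Hover brightens the top image to white through the Anim Transition macro.
# Mouse down/up and clicks drive the Pressed animation, nudging image and text toward the shadow.
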WidgetBlueprint /FortniteCoreUI/Frontend/Showdown/PrimarySkewButton
animations [
 Hover {
  binding_count=1
  display_label=""
  length_seconds=0.3493
  object=&self:Hover
  sequence {bindings [Image_ButtonTop {
     id=#1
     possessed_class=Image
     tracks [{
       type=MovieSceneColorTrack
       display_name=Color and Opacity
       sections [{
         type=MovieSceneColorSection
         end_frame=6
         start_frame=0
         channels [
          Color.R {type=MovieSceneScriptingFloatChannel, keys [[0,0.9216],[6.986,1]]}
          Color.G {type=MovieSceneScriptingFloatChannel, const [1,2]}
          Color.B {type=MovieSceneScriptingFloatChannel, keys [[0,0],[6.986,1]]}
          Color.A {type=MovieSceneScriptingFloatChannel, const [1,2]}
         ]
        }]
      }]
    }]}
 }
 Pressed {
  binding_count=2
  display_label=""
  length_seconds=0.45
  object=&self:Pressed
  sequence {bindings [
    Image_ButtonTop {
     id=#2
     possessed_class=Image
     tracks [{
       type=MovieScene2DTransformTrack
       sections [{
         type=MovieScene2DTransformSection
         end_frame=9
         start_frame=0
         channels [
          Translation.X {type=MovieSceneScriptingFloatChannel, keys [[0,0],[2,9.6925],[5,-0.7432],[9,0]]}
          Translation.Y {type=MovieSceneScriptingFloatChannel, keys [[0,0],[2,9.6925],[5,-0.7432],[9,0]]}
          Angle {type=MovieSceneScriptingFloatChannel}
          Scale.X {type=MovieSceneScriptingFloatChannel}
          Scale.Y {type=MovieSceneScriptingFloatChannel}
          Shear.X {type=MovieSceneScriptingFloatChannel}
          Shear.Y {type=MovieSceneScriptingFloatChannel}
         ]
        }]
      }]
    }
    CenterButtonTextWidget {
     id=#3
     possessed_class=CommonTextBlock
     tracks [{
       type=MovieScene2DTransformTrack
       sections [{
         type=MovieScene2DTransformSection
         end_frame=9
         start_frame=0
         channels [
          Translation.X {type=MovieSceneScriptingFloatChannel, keys [[0,0],[2,9.6925],[5,-0.7432],[9,0]]}
          Translation.Y {type=MovieSceneScriptingFloatChannel, keys [[0,0],[2,9.6925],[5,-0.7432],[9,0]]}
          Angle {type=MovieSceneScriptingFloatChannel}
          Scale.X {type=MovieSceneScriptingFloatChannel}
          Scale.Y {type=MovieSceneScriptingFloatChannel}
          Shear.X {type=MovieSceneScriptingFloatChannel}
          Shear.Y {type=MovieSceneScriptingFloatChannel}
         ]
        }]
      }]
    }
   ]}
 }
]
graphs [
 EventGraph {nodes [
   n1 = Event PreConstruct()
   > n2 = Branch(Condition=@n1.IsDesignTime)
   n3 = Get ButtonNameText() -> (ButtonNameText:Text)
   n4 = Get CenterButtonTextWidget() -> (CenterButtonTextWidget:Common Text Object Reference)
   "SetText (Text)"(
    execute=[@n7.then, @n2.then]
    self=@n4
    InText=Select(Option 0:Text=@n3, Option 1:Text=TextToUpper!(InText:"Text (by ref)"=@n3) -> (ReturnValue:Text), Index=Get toUpper() -> (toUpper:bool)) -> (ReturnValue:Text)
   )
   n5 = "To Float (Integer)"(InInt=MakeLiteralInt!(Value:int=40) -> (ReturnValue:int=0))
   n6 = "To Float (Integer)"(InInt=MakeLiteralInt(Value=80) -> (ReturnValue=0))
   > SetFont(
    self=@n4
    InFontInfo=Select(
     Option 0:Slate Font Info Structure=MakeSlateFontInfo!(
      FontObject:Object Reference=&Burbank
      FontMaterial:Object Reference
      OutlineSettings:Font Outline Settings Structure=(OutlineSize=0,bMiteredCorners=False,bSeparateFillAlpha=False,bApplyOutlineToDropShadows=False,OutlineMaterial=None,OutlineColor=(R=0,G=0,B=0,A=1))
      TypefaceFontName:Name=Burbank-RegularBlack
      Size:float=@n5
      LetterSpacing:int=0
      SkewAmount:float=0
      bForceMonospaced:bool=false
      bMaterialIsStencil:bool=false
      MonospacedWidth:float=1
     ) -> (ReturnValue:Slate Font Info Structure)
     Option 1:Slate Font Info Structure=MakeSlateFontInfo(
      FontObject=&Burbank
      OutlineSettings=(OutlineSize=0,bMiteredCorners=False,bSeparateFillAlpha=False,bApplyOutlineToDropShadows=False,OutlineMaterial=None,OutlineColor=(R=0,G=0,B=0,A=1))
      TypefaceFontName=Burbank-Black
      Size=@n6
      LetterSpacing=0
      SkewAmount=0
      bForceMonospaced=false
      bMaterialIsStencil=false
      MonospacedWidth=1
     )
     Index=Get XL() -> (XL:bool)
    ) -> (ReturnValue:Slate Font Info Structure)
   )
   n7 = Event Construct()
   Event On Hovered()
   n8 = Self-Reference()
   n9 = Get Hover() -> (Hover:Widget Animation Object Reference)
   > Anim Transition!(Exec:Exec, In:bool=true, Animation:Widget Animation Object Reference=@n9, PlaySpeed:double=3, Widget:User Widget Object Reference=@n8) -> (ExecOut:Exec) {is=title K2Node_MacroInstance}
   Event On Unhovered()
   > Anim Transition(Animation=@n9, PlaySpeed=1, Widget=@n8)
   Event On Clicked()
   > PlayAnimation(InAnimation=Get Pressed() -> (Pressed:Widget Animation Object Reference))
   SetVisibility(InVisibility=Select(
     Option 0:ESlateVisibility Enum=Collapsed
     Option 1:ESlateVisibility Enum=HitTestInvisible
     Index="On Input Method Changed (InputActionWidget)"!() -> (OutputDelegate, then, bUsingGamepad:bool) {is=title K2Node_ComponentBoundEvent}.bUsingGamepad
    ) -> (ReturnValue:ESlateVisibility Enum))
   Event Tick()
  ]}
 OnMouseButtonDown {nodes [
   OnMouseButtonDown!() -> (then, MyGeometry:Geometry Structure, MouseEvent:"Pointer Event Structure (by ref)") {is=title K2Node_FunctionEntry}
   > Anim Transition(In=true, Animation=Get Pressed(), PlaySpeed=2, Widget=Self-Reference())
   > Return Node!(execute, ReturnValue:Event Reply Structure=Unhandled())
  ]}
 OnMouseButtonUp {nodes [
   OnMouseButtonUp!() -> (then, MyGeometry:Geometry Structure, MouseEvent:"Pointer Event Structure (by ref)") {is=title K2Node_FunctionEntry}
   > Anim Transition(Animation=Get Pressed(), PlaySpeed=1, Widget=Self-Reference())
   > Return Node(ReturnValue=Unhandled())
  ]}
]
variables [
 ButtonNameText {type=string, default=""}
 toUpper {type=boolean, default=true}
 XL {type=boolean}
]
widgets {
 info {inheritedWidgetCount=1, parentClass=&CommonButtonLegacy, rootWidgetClass=&Overlay, widgetCount=7}
 widgets [
  Overlay_1 {widgetClassPath=&Overlay, widget=Overlay_1}
  Overlay_0 {
   widgetClassPath=&Overlay
   parent=Overlay_1
   widget=Overlay_0
   slot {class=OverlaySlot, properties {horizontalAlignment=HAlign_Fill, verticalAlignment=VAlign_Fill}}
  }
  Overlay_2 {
   widgetClassPath=&Overlay
   parent=Overlay_0
   widget=Overlay_2
   slot {class=OverlaySlot, properties {horizontalAlignment=HAlign_Fill, verticalAlignment=VAlign_Fill}}
  }
  Image_Shadow {
   widgetClassPath=&Image
   bIsVariable=true
   parent=Overlay_2
   widget=Image_Shadow
   properties {
    brush {
     bIsDynamicallyLoaded=false
     drawAs=Image
     imageType=NoImage
     mirroring=NoMirror
     resourceName=None
     resourceObject=/Game/UI/Frontend/Showdown/Materials/M_UI_SkewBox_LButton
     tiling=NoTile
     imageSize=<32 32>
     margin=<ltrb 0 0 0 0>
     outlineSettings {
      bUseBrushTransparency=false
      roundingType=HalfHeightRadius
      width=0
      color {colorUseRule=UseColor_Specified, specifiedColor=<rgba 0 0 0 0>}
      cornerRadii=<0 0 0 0>
     }
     tintColor {colorUseRule=UseColor_Specified, specifiedColor=<rgba 1 1 1 1>}
     uVRegion {bIsValid=false, max=<0 0>, min=<0 0>}
    }
    colorAndOpacity=<rgba 0.0075 0.0091 0.0242 0.28>
    renderTransform {angle=0, scale=<1 1>, shear=<0 0>, translation=<15 15>}
   }
   slot {class=OverlaySlot, properties {horizontalAlignment=HAlign_Fill, verticalAlignment=VAlign_Fill}}
  }
  Image_ButtonTop {
   widgetClassPath=&Image
   bIsVariable=true
   parent=Overlay_2
   widget=Image_ButtonTop
   properties {
    brush {
     bIsDynamicallyLoaded=false
     drawAs=Image
     imageType=NoImage
     mirroring=NoMirror
     resourceName=None
     resourceObject=&M_UI_SkewBox_LButton
     tiling=NoTile
     imageSize=<32 32>
     margin=<ltrb 0 0 0 0>
     outlineSettings {
      bUseBrushTransparency=false
      roundingType=HalfHeightRadius
      width=0
      color {colorUseRule=UseColor_Specified, specifiedColor=<rgba 0 0 0 0>}
      cornerRadii=<0 0 0 0>
     }
     tintColor {colorUseRule=UseColor_Specified, specifiedColor=<rgba 1 1 1 1>}
     uVRegion {bIsValid=false, max=<0 0>, min=<0 0>}
    }
    colorAndOpacity=<rgba 0.9216 1 0 1>
   }
   slot {class=OverlaySlot, properties {horizontalAlignment=HAlign_Fill, verticalAlignment=VAlign_Fill}}
  }
  CenterButtonTextWidget {
   widgetClassPath=&CommonTextBlock
   bIsVariable=true
   parent=Overlay_0
   widget=CenterButtonTextWidget
   properties {
    justification=Center
    text=Text Block
    colorAndOpacity {colorUseRule=UseColor_Specified, specifiedColor=<rgba 0 0 0 1>}
    font {
     bForceMonospaced=false
     bMaterialIsStencil=false
     fontMaterial=null
     fontObject=&Burbank
     letterSpacing=0
     monospacedWidth=1
     size=40
     skewAmount=0
     typefaceFontName=BurbankRegular-Black
     outlineSettings {
      bApplyOutlineToDropShadows=false
      bMiteredCorners=false
      bSeparateFillAlpha=false
      outlineMaterial=null
      outlineSize=0
      outlineColor=<rgba 0 0 0 1>
     }
    }
    renderTransform {angle=0, scale=<1 1>, shear=<-7 0>, translation=<0 0>}
   }
   slot {class=OverlaySlot, properties {horizontalAlignment=HAlign_Fill, verticalAlignment=VAlign_Fill, padding=<ltrb 75 11 75 10>}}
  }
  InputActionWidget {
   widgetClassPath=&CommonActionWidget
   bInherited=true
   bIsVariable=true
   parent=Overlay_0
   widget=InputActionWidget
   slot {class=OverlaySlot, properties {verticalAlignment=VAlign_Center, padding=<ltrb 25 0 0 0>}}
  }
 ]
}
