WidgetBlueprint /FortniteCoreUI/UI/Events/DivisionDetailsProgress
graphs [EventGraph {nodes [
   n1 = Event OnSetInProgress!() -> (OutputDelegate, then, NewPercent:float, bIsInProgress:bool) {is=title K2Node_Event}
   > n2 = Branch(Condition=@n1.bIsInProgress)
   n3 = Get Image_DivisionIcon() -> (Image_DivisionIcon:Fort Lazy Image Object Reference)
   > SetColorAndOpacity(self:Image Object Reference=@n3, InColorAndOpacity:Linear Color Structure=(R=0,G=1,B=0.9823,A=1))
   > Branch(Condition="float < float"!(A:double=@n1.NewPercent, B:double=1) -> (ReturnValue:bool) {is=title K2Node_CallFunction})
   n4 = Get CommonBorder_CalloutOffset() -> (CommonBorder_CalloutOffset:Common Border Object Reference)
   n5 = float - float(B=@n1.NewPercent) {is=title K2Node_CallFunction}
   > SetDesiredSizeScale!(execute, self:Border Object Reference=@n4, InScale:Vector 2D Structure=MakeVector2D(X="Clamp (Float)"(Value=@n5, Min=0, Max=1), Y=1)) -> (then)
   > SetVisibility(self=@n4, InVisibility=SelfHitTestInvisible)
   SetColorAndOpacity(execute=@n2.else, self=@n3, InColorAndOpacity=(R=1,G=1,B=1,A=1))
   > SetVisibility(self=@n4, InVisibility=Collapsed)
  ]}]
variables []
widgets {
 info {
  inheritedWidgetCount=2
  parentClass=/Script/FortniteUI.FortShowdownDivisionRankProgressView
  rootWidgetClass=&Overlay
  widgetCount=12
 }
 widgets [
  Overlay_0 {widgetClassPath=&Overlay, widget=Overlay_0}
  Overlay_2 {
   widgetClassPath=&Overlay
   parent=Overlay_0
   widget=Overlay_2
   slot {class=OverlaySlot, properties {horizontalAlignment=HAlign_Fill, verticalAlignment=VAlign_Fill}}
  }
  CommonBorder_CalloutOffset {
   widgetClassPath=&CommonBorder
   bIsVariable=true
   parent=Overlay_2
   widget=CommonBorder_CalloutOffset
   properties {horizontalAlignment=HAlign_Right, verticalAlignment=VAlign_Center, visibility=Collapsed, padding=<ltrb 0 0 0 0>}
   slot {class=OverlaySlot, properties {horizontalAlignment=HAlign_Right}}
  }
  SizeBox_1 {
   widgetClassPath=&SizeBox
   parent=CommonBorder_CalloutOffset
   widget=SizeBox_1
   properties {bOverride_WidthOverride=true, widthOverride=110, renderTransform {angle=0, scale=<1 1>, shear=<0 0>, translation=<-5 0>}}
   slot {class=BorderSlot, properties {horizontalAlignment=HAlign_Right, verticalAlignment=VAlign_Center, padding=<ltrb 0 0 0 0>}}
  }
  CommonBorder_1 {
   widgetClassPath=&CommonBorder
   parent=SizeBox_1
   widget=CommonBorder_1
   properties {horizontalAlignment=HAlign_Left, desiredSizeScale=<0 1>, padding=<ltrb 0 0 -600 0>}
   slot {class=SizeBoxSlot}
  }
  Overlay_1 {
   widgetClassPath=&Overlay
   parent=CommonBorder_1
   widget=Overlay_1
   properties {renderTransform {angle=0, scale=<1 1>, shear=<0 0>, translation=<0 -30>}}
   slot {class=BorderSlot, properties {horizontalAlignment=HAlign_Left, padding=<ltrb 0 0 -600 0>}}
  }
  Image_OutlineCallout {
   widgetClassPath=&Image
   parent=Overlay_1
   widget=Image_OutlineCallout
   properties {
    visibility=SelfHitTestInvisible
    brush {
     bIsDynamicallyLoaded=false
     drawAs=Box
     imageType=NoImage
     mirroring=NoMirror
     resourceName=None
     resourceObject=/FortniteCoreUI/Frontend/Showdown/Materials/league/M_UI_LeagueDivision_OutlineCallout_Inst
     tiling=NoTile
     imageSize=<60 60>
     margin=<ltrb 0.4 0.4 0.4 0.5>
     outlineSettings {
      bUseBrushTransparency=false
      roundingType=HalfHeightRadius
      width=0
      color {colorUseRule=UseColor_Specified, specifiedColor=<rgba 0 0 0 0>}
      cornerRadii=<0 0 0 0>
     }
     tintColor {colorUseRule=UseColor_Specified, specifiedColor=<rgba 1 1 1 1>}
     uVRegion {bIsValid=false, max=<0 0>, min=<0 0>}
    }
   }
   slot {class=OverlaySlot, properties {horizontalAlignment=HAlign_Fill, verticalAlignment=VAlign_Fill}}
  }
  CommonTextBlock_0 {
   widgetClassPath=&CommonTextBlock
   parent=Overlay_1
   widget=CommonTextBlock_0
   properties {
    text=You Are Here
    textTransformPolicy=ToUpper
    font {
     bForceMonospaced=false
     bMaterialIsStencil=false
     fontMaterial=null
     fontObject=&BurbankSmall
     letterSpacing=0
     monospacedWidth=1
     size=18
     skewAmount=0
     typefaceFontName=Black
     outlineSettings {
      bApplyOutlineToDropShadows=false
      bMiteredCorners=false
      bSeparateFillAlpha=false
      outlineMaterial=null
      outlineSize=0
      outlineColor=<rgba 0 0 0 1>
     }
    }
    shadowOffset=<0 0>
   }
   slot {class=OverlaySlot, properties {verticalAlignment=VAlign_Center, padding=<ltrb 30 18 30 25>}}
  }
  HorizontalBox_2 {widgetClassPath=&HorizontalBox, parent=Overlay_2, widget=HorizontalBox_2, slot {class=OverlaySlot}}
  Image_DivisionIcon {
   widgetClassPath=&FortLazyImage
   bInherited=true
   bIsVariable=true
   parent=HorizontalBox_2
   widget=Image_DivisionIcon
   properties {brush {
     bIsDynamicallyLoaded=false
     drawAs=Image
     imageType=NoImage
     mirroring=NoMirror
     resourceName=None
     resourceObject=/FortniteCoreUI/Frontend/Showdown/Materials/league/M_UI_LeagueDivision_Icon1
     tiling=NoTile
     imageSize=<80 80>
     margin=<ltrb 0 0 0 0>
     outlineSettings {
      bUseBrushTransparency=false
      roundingType=HalfHeightRadius
      width=0
      color {colorUseRule=UseColor_Specified, specifiedColor=<rgba 0 0 0 0>}
      cornerRadii=<0 0 0 0>
     }
     tintColor {colorUseRule=UseColor_Specified, specifiedColor=<rgba 1 1 1 1>}
     uVRegion {bIsValid=false, max=<0 0>, min=<0 0>}
    }}
   slot {class=HorizontalBoxSlot, properties {horizontalAlignment=HAlign_Left, verticalAlignment=VAlign_Top}}
  }
  ProgressBarContainer {
   widgetClassPath=&SizeBox
   bIsVariable=true
   parent=HorizontalBox_2
   widget=ProgressBarContainer
   properties {bOverride_MinDesiredWidth=true, minDesiredWidth=75}
   slot {class=HorizontalBoxSlot, properties {verticalAlignment=VAlign_Center}}
  }
  Bar_DivisionProgress {
   widgetClassPath=/Script/UMG.ProgressBar
   bInherited=true
   bIsVariable=true
   parent=ProgressBarContainer
   widget=Bar_DivisionProgress
   properties {widgetStyle {
     enableFillAnimation=false
     backgroundImage {
      bIsDynamicallyLoaded=false
      drawAs=Image
      imageType=FullColor
      mirroring=NoMirror
      resourceName=../../../Engine/Content/Slate/Common/ProgressBar_Background.png
      resourceObject=/Game/Global/Textures/Default/Blanks/T_WhiteUserInterface
      tiling=NoTile
      imageSize=<16 16>
      margin=<ltrb 0.4167 0.4167 0.4167 0.4167>
      outlineSettings {
       bUseBrushTransparency=false
       roundingType=HalfHeightRadius
       width=0
       color {colorUseRule=UseColor_Specified, specifiedColor=<rgba 0 0 0 0>}
       cornerRadii=<0 0 0 0>
      }
      tintColor {colorUseRule=UseColor_Specified, specifiedColor=<rgba 0.0052 0.0052 0.0056 0.5>}
      uVRegion {bIsValid=false, max=<0 0>, min=<0 0>}
     }
     fillImage {
      bIsDynamicallyLoaded=false
      drawAs=Image
      imageType=FullColor
      mirroring=NoMirror
      resourceName=../../../Engine/Content/Slate/Common/ProgressBar_Fill.png
      resourceObject=&T_WhiteUserInterface
      tiling=NoTile
      imageSize=<16 16>
      margin=<ltrb 0.4167 0.4167 0.4167 0.4167>
      outlineSettings {
       bUseBrushTransparency=false
       roundingType=HalfHeightRadius
       width=0
       color {colorUseRule=UseColor_Specified, specifiedColor=<rgba 0 0 0 0>}
       cornerRadii=<0 0 0 0>
      }
      tintColor {colorUseRule=UseColor_Specified, specifiedColor=<rgba 0 1 0.9823 1>}
      uVRegion {bIsValid=false, max=<0 0>, min=<0 0>}
     }
     marqueeImage {
      bIsDynamicallyLoaded=false
      drawAs=Image
      imageType=FullColor
      mirroring=NoMirror
      resourceName=../../../Engine/Content/Slate/Common/ProgressBar_Marquee.png
      resourceObject=null
      tiling=Horizontal
      imageSize=<20 12>
      margin=<ltrb 0 0 0 0>
      outlineSettings {
       bUseBrushTransparency=false
       roundingType=HalfHeightRadius
       width=0
       color {colorUseRule=UseColor_Specified, specifiedColor=<rgba 0 0 0 0>}
       cornerRadii=<0 0 0 0>
      }
      tintColor {colorUseRule=UseColor_Specified, specifiedColor=<rgba 1 1 1 1>}
      uVRegion {bIsValid=false, max=<0 0>, min=<0 0>}
     }
    }}
   slot {class=SizeBoxSlot}
  }
 ]
}
