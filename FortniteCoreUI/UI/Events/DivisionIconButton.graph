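# Selectable division rank icon button for Showdown league UI, derived from FortShowdownDivisionRankButton.
# An overlay layers a hidden skewed-border hover image with the division icon, and hover events toggle the border's visibility.
# On selection change, the icon turns cyan and plays a short overshooting scale-bounce animation, or returns to white when deselected.
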WidgetBlueprint /FortniteCoreUI/UI/Events/DivisionIconButton
animations [Selected {
  binding_count=1
  length_seconds=0.2092
  object=&self:Selected
  sequence {bindings [Image_DivisionIcon {
     id=#1
     possessed_class=FortLazyImage
     tracks [{
       type=MovieScene2DTransformTrack
       sections [{
         type=MovieScene2DTransformSection
         end_frame=4
         start_frame=0
         channels [
          Translation.X {type=MovieSceneScriptingFloatChannel}
          Translation.Y {type=MovieSceneScriptingFloatChannel}
          Angle {type=MovieSceneScriptingFloatChannel}
          Scale.X {type=MovieSceneScriptingFloatChannel, keys [[0,1],[0.788,1.3985],[2.8757,0.9409],[3.684,1.0452],[4.1827,1]]}
          Scale.Y {type=MovieSceneScriptingFloatChannel, keys [[0,1],[0.788,1.3985],[2.8757,0.9409],[3.684,1.0452],[4.1827,1]]}
          Shear.X {type=MovieSceneScriptingFloatChannel}
          Shear.Y {type=MovieSceneScriptingFloatChannel}
         ]
        }]
      }]
    }]}
 }]
graphs [EventGraph {nodes [
   n1 = Event OnRankButtonSelected!() -> (OutputDelegate, then, bIsSelected:bool) {is=title K2Node_Event}
   > n2 = Branch(Condition=@n1.bIsSelected)
   n3 = Get Image_DivisionIcon() -> (Image_DivisionIcon:Fort Lazy Image Object Reference)
   > SetColorAndOpacity(self:Image Object Reference=@n3, InColorAndOpacity:Linear Color Structure=(R=0,G=1,B=0.9823,A=1))
   > PlayAnimation(InAnimation=Get Selected() -> (Selected:Widget Animation Object Reference))
   SetColorAndOpacity(execute=@n2.else, self=@n3, InColorAndOpacity=(R=1,G=1,B=1,A=1))
   Event On Hovered()
   n4 = Get Image_Hover() -> (Image_Hover:Image Object Reference)
   > SetVisibility(self=@n4, InVisibility=HitTestInvisible)
   Event On Unhovered()
   > SetVisibility(self=@n4, InVisibility=Hidden)
  ]}]
variables []
widgets {
 info {inheritedWidgetCount=2, parentClass=/Script/FortniteUI.FortShowdownDivisionRankButton, rootWidgetClass=&Overlay, widgetCount=4}
 widgets [
  Overlay_0 {widgetClassPath=&Overlay, widget=Overlay_0}
  Image_Hover {
   widgetClassPath=&Image
   bIsVariable=true
   parent=Overlay_0
   widget=Image_Hover
   properties {
    visibility=Hidden
    brush {
     bIsDynamicallyLoaded=false
     drawAs=Image
     imageType=NoImage
     mirroring=NoMirror
     resourceName=None
     resourceObject=/FortniteCoreUI/Frontend/Showdown/Materials/M_UI_SkewBoxBorder_Hover
     tiling=NoTile
     imageSize=<78.8563 73.2642>
     margin=<ltrb 0 0 0 0>
     outlineSettings {
      bUseBrushTransparency=false
      roundingType=HalfHeightRadius
      width=0
      color {colorUseRule=UseColor_Specified, specifiedColor=<rgba 0 0 0 0>}
      cornerRadii=<0 0 0 0>
     }
     tintColor {colorUseRule=UseColor_Specified, specifiedColor=<rgba 1 1 1 1>}
     uVRegion {bIsValid=false, max=<0 0>, min=<0 0>}
    }
   }
   slot {class=OverlaySlot, properties {horizontalAlignment=HAlign_Center, verticalAlignment=VAlign_Center, padding=<ltrb -8 -8 -8 -8>}}
  }
  Image_DivisionIcon {
   widgetClassPath=&FortLazyImage
   bInherited=true
   bIsVariable=true
   parent=Overlay_0
   widget=Image_DivisionIcon
   properties {brush {
     bIsDynamicallyLoaded=false
     drawAs=Image
     imageType=NoImage
     mirroring=NoMirror
     resourceName=None
     resourceObject=/FortniteCoreUI/Frontend/Showdown/Materials/league/M_UI_LeagueDivision_Icon7
     tiling=NoTile
     imageSize=<67 67>
     margin=<ltrb 0 0 0 0>
     outlineSettings {
      bUseBrushTransparency=false
      roundingType=HalfHeightRadius
      width=0
      color {colorUseRule=UseColor_Specified, specifiedColor=<rgba 0 0 0 0>}
      cornerRadii=<0 0 0 0>
     }
     tintColor {colorUseRule=UseColor_Specified, specifiedColor=<rgba 1 1 1 1>}
     uVRegion {bIsValid=false, max=<0 0>, min=<0 0>}
    }}
   slot {class=OverlaySlot, properties {horizontalAlignment=HAlign_Center, verticalAlignment=VAlign_Center, padding=<ltrb 13 0 13 0>}}
  }
  InputActionWidget {widgetClassPath=&CommonActionWidget, bInherited=true, widget=null}
 ]
}
